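# Control Rig for the F_MED_TwinGinger female character body that procedurally corrects twist bones and clothing bones from the animated pose.
# Swing-twist extraction drives thigh, calf, upperarm and lowerarm twist bones, while weighted parent constraints position buckle bones.
# Functions layer weighted corrective offsets onto buckles, collar, lapels, shirt and straps.
# Those weights come from DeltaRotation joint-angle remaps, deformRig animation attributes and pose or state variables like ADS, skydiving and DBNO.
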ControlRigBlueprint /BRCosmetics/Characters/Player/Female/Medium/Bodies/F_MED_TwinGinger/Meshes/F_MED_TwinGinger_CtrlRig
graphs [
 CollapseNode_ContainedGraph {nodes [
   RigVMFunctionEntryNode()
   > RigVMFunction_Sequence = RigVMAggregateNode()
   RigVMFunctionReturnNode(ExecuteContext=@RigVMFunction_Sequence.E)
   RigVMFunction_MathQuaternionSwingTwist = &RigVMFunction_MathQuaternionSwingTwist(
    TwistAxis=(X=1,Y=0,Z=0)
    Input=&RigUnit_GetRelativeTransformForItem(Child=(Type=Bone,Name="thigh_l"), Parent=(Type=Null,Name="thigh_l_null"), bChildInitial=False, bParentInitial=False).RelativeTransform.Rotation
   )
   Inverse = /Script/RigVM.RigVMFunction_MathQuaternionInverse(Value=@RigVMFunction_MathQuaternionSwingTwist.Twist) {is=struct RigVMUnitNode, template="Inverse::Execute(in Value,out Result)"}
   &RigUnit_ModifyTransforms(
    ItemToModify=((Item=(Type=Bone,Name="thigh_twist_01_l"),Transform=(Rotation=(X=0,Y=0,Z=0,W=1),Translation=(X=0,Y=0,Z=0),Scale3D=(X=1,Y=1,Z=1))))
    Weight=1
    WeightMinimum=0
    ItemToModify.0.Transform.Rotation=@Inverse.Result
    ExecutePin=@RigVMFunction_Sequence.A
   )
   n1 = &RigUnit_GetRelativeTransformForItem(Child=(Type=Bone,Name="thigh_r"), Parent=(Type=Null,Name="thigh_r_null"), bChildInitial=False, bParentInitial=False)
   > &RigUnit_ModifyTransforms(
    ItemToModify=((Item=(Type=Bone,Name="thigh_twist_01_r"),Transform=(Rotation=(X=0,Y=0,Z=0,W=1),Translation=(X=0,Y=0,Z=0),Scale3D=(X=1,Y=1,Z=1))))
    Weight=1
    WeightMinimum=0
    ItemToModify.0.Transform.Rotation=&RigVMFunction_MathQuaternionInverse(Value=&RigVMFunction_MathQuaternionSwingTwist(TwistAxis=(X=1,Y=0,Z=0), Input=@n1.RelativeTransform.Rotation).Twist) {template="Inverse::Execute(in Value,out Result)"}.Result
   )
   n2 = &RigVMFunction_MathQuaternionInverse(Value=&RigVMFunction_MathQuaternionSwingTwist(
     TwistAxis=(X=1,Y=0,Z=0)
     Input=&RigUnit_GetRelativeTransformForItem(Child=(Type=Bone,Name="upperarm_l"), Parent=(Type=Null,Name="upperarm_l_null"), bChildInitial=False, bParentInitial=False).RelativeTransform.Rotation
    ).Twist) {template="Inverse::Execute(in Value,out Result)"}
   &RigUnit_ModifyTransforms(
    ItemToModify="((Item=(Type=Bone,Name=\"upperarm_twist_01_l\"),Transform=(Rotation=(X=0,Y=0,Z=0,W=1),Translation=(X=0,Y=0,Z=0),Scale3D=(X=1,Y=1,Z=1))),(Item=(Type=Bone,Name=\"upperarm_twist_02_l\"),Transform=(Rotation=(X=0,Y=0,Z=0,W=1),Translation=(X=0,Y=0,Z=0),Scale3D=(X=1, ...(truncated)"
    Weight=1
    WeightMinimum=0
    ItemToModify.1.Transform.Rotation=/Script/RigVM.RigVMFunction_MathQuaternionSlerp(B=(X=0,Y=0,Z=0,W=1), T=0.5, A=@n2.Result) {is=struct RigVMUnitNode, template="Interpolate::Execute(in A,in B,in T,out Result)"}.Result
    ItemToModify.0.Transform.Rotation=@n2.Result
    ExecutePin=@RigVMFunction_Sequence.B
   )
   n3 = &RigVMFunction_MathQuaternionInverse(Value=&RigVMFunction_MathQuaternionSwingTwist(
     TwistAxis=(X=1,Y=0,Z=0)
     Input=&RigUnit_GetRelativeTransformForItem(Child=(Type=Bone,Name="upperarm_r"), Parent=(Type=Null,Name="upperarm_r_null"), bChildInitial=False, bParentInitial=False).RelativeTransform.Rotation
    ).Twist) {template="Inverse::Execute(in Value,out Result)"}
   Interpolate = &RigVMFunction_MathQuaternionSlerp(B=(X=0,Y=0,Z=0,W=1), T=0.5, A=@n3.Result) {template="Interpolate::Execute(in A,in B,in T,out Result)"}
   > &RigUnit_ModifyTransforms(
    ItemToModify="((Item=(Type=Bone,Name=\"upperarm_twist_01_r\"),Transform=(Rotation=(X=0,Y=0,Z=0,W=1),Translation=(X=0,Y=0,Z=0),Scale3D=(X=1,Y=1,Z=1))),(Item=(Type=Bone,Name=\"upperarm_twist_02_r\"),Transform=(Rotation=(X=0,Y=0,Z=0,W=1),Translation=(X=0,Y=0,Z=0),Scale3D=(X=1, ...(truncated)"
    Weight=1
    WeightMinimum=0
    ItemToModify.1.Transform.Rotation=@Interpolate.Result
    ItemToModify.0.Transform.Rotation=@n3.Result
   )
   n4 = &RigVMFunction_MathQuaternionSwingTwist(
    TwistAxis=(X=1,Y=0,Z=0)
    Input=&RigUnit_GetRelativeTransformForItem(Child=(Type=Bone,Name="foot_l"), Parent=(Type=Null,Name="foot_l_null"), bChildInitial=False, bParentInitial=False).RelativeTransform.Rotation
   )
   &RigUnit_ModifyTransforms(
    ItemToModify="((Item=(Type=Bone,Name=\"calf_twist_01_l\"),Transform=(Rotation=(X=0,Y=0,Z=0,W=1),Translation=(X=0,Y=0,Z=0),Scale3D=(X=1,Y=1,Z=1))),(Item=(Type=Bone,Name=\"calf_twist_02_l\"),Transform=(Rotation=(X=0,Y=0,Z=0,W=1),Translation=(X=0,Y=0,Z=0),Scale3D=(X=1,Y=1,Z=1) ...(truncated)"
    Weight=1
    WeightMinimum=0
    ItemToModify.1.Transform.Rotation=&RigVMFunction_MathQuaternionSlerp(B=(X=0,Y=0,Z=0,W=1), T=0.5, A=@n4.Twist) {template="Interpolate::Execute(in A,in B,in T,out Result)"}.Result
    ItemToModify.0.Transform.Rotation=@n4.Twist
    ExecutePin=@RigVMFunction_Sequence.C
   )
   n5 = &RigVMFunction_MathQuaternionSwingTwist(
    TwistAxis=(X=1,Y=0,Z=0)
    Input=&RigUnit_GetRelativeTransformForItem(Child=(Type=Bone,Name="foot_r"), Parent=(Type=Null,Name="foot_r_null"), bChildInitial=False, bParentInitial=False).RelativeTransform.Rotation
   )
   > &RigUnit_ModifyTransforms(
    ItemToModify="((Item=(Type=Bone,Name=\"calf_twist_01_r\"),Transform=(Rotation=(X=0,Y=0,Z=0,W=1),Translation=(X=0,Y=0,Z=0),Scale3D=(X=1,Y=1,Z=1))),(Item=(Type=Bone,Name=\"calf_twist_02_r\"),Transform=(Rotation=(X=0,Y=0,Z=0,W=1),Translation=(X=0,Y=0,Z=0),Scale3D=(X=1,Y=1,Z=1) ...(truncated)"
    Weight=1
    WeightMinimum=0
    ItemToModify.1.Transform.Rotation=&RigVMFunction_MathQuaternionSlerp(B=(X=0,Y=0,Z=0,W=1), T=0.5, A=@n5.Twist) {template="Interpolate::Execute(in A,in B,in T,out Result)"}.Result
    ItemToModify.0.Transform.Rotation=@n5.Twist
   )
   n6 = &RigVMFunction_MathQuaternionSwingTwist(
    TwistAxis=(X=1,Y=0,Z=0)
    Input=&RigUnit_GetRelativeTransformForItem(Child=(Type=Bone,Name="hand_l"), Parent=(Type=Null,Name="hand_l_null"), bChildInitial=False, bParentInitial=False).RelativeTransform.Rotation
   )
   &RigUnit_ModifyTransforms(
    ItemToModify="((Item=(Type=Bone,Name=\"lowerarm_twist_01_l\"),Transform=(Rotation=(X=0,Y=0,Z=0,W=1),Translation=(X=0,Y=0,Z=0),Scale3D=(X=1,Y=1,Z=1))),(Item=(Type=Bone,Name=\"lowerarm_twist_02_l\"),Transform=(Rotation=(X=0,Y=0,Z=0,W=1),Translation=(X=0,Y=0,Z=0),Scale3D=(X=1, ...(truncated)"
    Weight=1
    WeightMinimum=0
    ItemToModify.1.Transform.Rotation=&RigVMFunction_MathQuaternionSlerp(B=(X=0,Y=0,Z=0,W=1), T=0.5, A=@n6.Twist) {template="Interpolate::Execute(in A,in B,in T,out Result)"}.Result
    ItemToModify.0.Transform.Rotation=@n6.Twist
    ExecutePin=@RigVMFunction_Sequence.D
   )
   n7 = &RigVMFunction_MathQuaternionSwingTwist(
    TwistAxis=(X=1,Y=0,Z=0)
    Input=&RigUnit_GetRelativeTransformForItem(Child=(Type=Bone,Name="hand_r"), Parent=(Type=Null,Name="hand_r_null"), bChildInitial=False, bParentInitial=False).RelativeTransform.Rotation
   )
   > &RigUnit_ModifyTransforms(
    ItemToModify="((Item=(Type=Bone,Name=\"lowerarm_twist_01_r\"),Transform=(Rotation=(X=0,Y=0,Z=0,W=1),Translation=(X=0,Y=0,Z=0),Scale3D=(X=1,Y=1,Z=1))),(Item=(Type=Bone,Name=\"lowerarm_twist_02_r\"),Transform=(Rotation=(X=0,Y=0,Z=0,W=1),Translation=(X=0,Y=0,Z=0),Scale3D=(X=1, ...(truncated)"
    Weight=1
    WeightMinimum=0
    ItemToModify.1.Transform.Rotation=&RigVMFunction_MathQuaternionSlerp(B=(X=0,Y=0,Z=0,W=1), T=0.5, A=@n7.Twist) {template="Interpolate::Execute(in A,in B,in T,out Result)"}.Result
    ItemToModify.0.Transform.Rotation=@n7.Twist
   )
  ]}
 CollapseNode_ContainedGraph {nodes [
   n1 = RigVMFunctionEntryNode()
   &RigUnit_ParentConstraint(
    AdvancedSettings=(RotationOrderForFilter=XZY)
    Child=(Type=Bone,Name="base_buckle_up_r")
    Parents="((Item=(Type=Bone,Name=\"clavicle_r\"),Weight=0.0126),(Item=(Type=Bone,Name=\"upperarm_twist_01_r\"),Weight=0.0221),(Item=(Type=Bone,Name=\"deform_upperarm_fr_r\"),Weight=0.0253),(Item=(Type=Bone,Name=\"spine_05\"),Weight=0.1385),(Item=(Type=Bone,Name=\"deform_pec_ ...(truncated)"
    Weight=1
    ExecutePin=@n1.ExecutePin
   )
   > &RigUnit_ParentConstraint(
    AdvancedSettings=(RotationOrderForFilter=XZY)
    Child=(Type=Bone,Name="base_buckle_dw_r")
    Parents="((Item=(Type=Bone,Name=\"clavicle_r\"),Weight=0.0147),(Item=(Type=Bone,Name=\"deform_upperarm_fr_r\"),Weight=0.045),(Item=(Type=Bone,Name=\"upperarm_twist_01_r\"),Weight=0.0836),(Item=(Type=Bone,Name=\"spine_03\"),Weight=0.1008),(Item=(Type=Bone,Name=\"deform_pec_r ...(truncated)"
    Weight=1
   )
   > &RigUnit_ParentConstraint(
    AdvancedSettings=(RotationOrderForFilter=XZY)
    Child=(Type=Bone,Name="base_buckle_up_l")
    Parents="((Item=(Type=Bone,Name=\"clavicle_l\"),Weight=0.0126),(Item=(Type=Bone,Name=\"upperarm_twist_01_l\"),Weight=0.0221),(Item=(Type=Bone,Name=\"deform_upperarm_fr_l\"),Weight=0.0253),(Item=(Type=Bone,Name=\"spine_05\"),Weight=0.1385),(Item=(Type=Bone,Name=\"deform_pec_ ...(truncated)"
    Weight=1
   )
   > n2 = &RigUnit_ParentConstraint(
    AdvancedSettings=(RotationOrderForFilter=XZY)
    Child=(Type=Bone,Name="base_buckle_dw_l")
    Parents="((Item=(Type=Bone,Name=\"clavicle_l\"),Weight=0.0147),(Item=(Type=Bone,Name=\"deform_upperarm_fr_l\"),Weight=0.045),(Item=(Type=Bone,Name=\"upperarm_twist_01_l\"),Weight=0.0836),(Item=(Type=Bone,Name=\"spine_03\"),Weight=0.1008),(Item=(Type=Bone,Name=\"deform_pec_l ...(truncated)"
    Weight=1
   )
   RigVMFunctionReturnNode(ExecutePin=@n2.ExecutePin)
   &RigUnit_ParentConstraint(
    AdvancedSettings=(RotationOrderForFilter=XZY)
    Child=(Type=Bone,Name="dyn_buckle_up_l_02")
    ExecutePin=()
    Parents="((Item=(Type=Bone,Name=\"deform_upperarm_fr_l\"),Weight=0.0011),(Item=(Type=Bone,Name=\"dyn_jacket_l_02\"),Weight=0.0013),(Item=(Type=Bone,Name=\"clavicle_l\"),Weight=0.0031),(Item=(Type=Bone,Name=\"dyn_jacket_l_01\"),Weight=0.0038),(Item=(Type=Bone,Name=\"upperarm ...(truncated)"
    Weight=1
   )
   > &RigUnit_ParentConstraint(
    AdvancedSettings=(RotationOrderForFilter=XZY)
    Child=(Type=Bone,Name="dyn_buckle_dw_l_02")
    Parents="((Item=(Type=Bone,Name=\"clavicle_l\"),Weight=0.0029),(Item=(Type=Bone,Name=\"upperarm_twist_01_l\"),Weight=0.0146),(Item=(Type=Bone,Name=\"spine_05\"),Weight=0.0305),(Item=(Type=Bone,Name=\"deform_pec_l\"),Weight=0.0647),(Item=(Type=Bone,Name=\"dyn_jacket_l_03\"),W ...(truncated)"
    Weight=1
   )
   &RigUnit_ParentConstraint(
    AdvancedSettings=(RotationOrderForFilter=XZY)
    Child=(Type=Bone,Name="base_buckle_up_l")
    ExecutePin=()
    Parents="((Item=(Type=Bone,Name=\"dyn_jacket_l_01\"),Weight=0.0006),(Item=(Type=Bone,Name=\"upperarm_twist_01_l\"),Weight=0.0135),(Item=(Type=Bone,Name=\"clavicle_l\"),Weight=0.0184),(Item=(Type=Bone,Name=\"deform_upperarm_fr_l\"),Weight=0.0254),(Item=(Type=Bone,Name=\"spin ...(truncated)"
    Weight=1
   )
   > &RigUnit_ParentConstraint(
    AdvancedSettings=(RotationOrderForFilter=XZY)
    Child=(Type=Bone,Name="base_buckle_dw_r")
    Parents="((Item=(Type=Bone,Name=\"dyn_jacket_r_02\"),Weight=0.0039),(Item=(Type=Bone,Name=\"dyn_jacket_r_01\"),Weight=0.0116),(Item=(Type=Bone,Name=\"clavicle_r\"),Weight=0.0183),(Item=(Type=Bone,Name=\"deform_upperarm_fr_r\"),Weight=0.0211),(Item=(Type=Bone,Name=\"upperarm ...(truncated)"
    Weight=1
   )
   > &RigUnit_ParentConstraint(
    AdvancedSettings=(RotationOrderForFilter=XZY)
    Child=(Type=Bone,Name="base_buckle_up_r")
    Parents="((Item=(Type=Bone,Name=\"dyn_jacket_r_02\"),Weight=0.0002),(Item=(Type=Bone,Name=\"dyn_jacket_r_01\"),Weight=0.0028),(Item=(Type=Bone,Name=\"upperarm_twist_01_r\"),Weight=0.0142),(Item=(Type=Bone,Name=\"clavicle_r\"),Weight=0.019),(Item=(Type=Bone,Name=\"deform_upp ...(truncated)"
    Weight=1
   )
   > &RigUnit_ParentConstraint(
    AdvancedSettings=(RotationOrderForFilter=XZY)
    Child=(Type=Bone,Name="base_buckle_dw_r")
    Parents="((Item=(Type=Bone,Name=\"dyn_jacket_r_02\"),Weight=0.0038),(Item=(Type=Bone,Name=\"clavicle_r\"),Weight=0.0191),(Item=(Type=Bone,Name=\"deform_upperarm_fr_r\"),Weight=0.023),(Item=(Type=Bone,Name=\"dyn_jacket_r_01\"),Weight=0.0491),(Item=(Type=Bone,Name=\"upperarm_ ...(truncated)"
    Weight=1
   )
   &RigUnit_ParentConstraint(
    AdvancedSettings=(RotationOrderForFilter=XZY)
    Child=(Type=Bone,Name="dyn_buckle_dw_r_02")
    ExecutePin=()
    Parents="((Item=(Type=Bone,Name=\"clavicle_r\"),Weight=0.0025),(Item=(Type=Bone,Name=\"upperarm_twist_01_r\"),Weight=0.0133),(Item=(Type=Bone,Name=\"spine_05\"),Weight=0.0292),(Item=(Type=Bone,Name=\"deform_pec_r\"),Weight=0.0702),(Item=(Type=Bone,Name=\"dyn_jacket_r_03\"),W ...(truncated)"
    Weight=1
   )
   > &RigUnit_ParentConstraint(
    AdvancedSettings=(RotationOrderForFilter=XZY)
    Child=(Type=Bone,Name="dyn_buckle_up_r_02")
    Parents="((Item=(Type=Bone,Name=\"dyn_jacket_r_03\"),Weight=0.0005),(Item=(Type=Bone,Name=\"clavicle_r\"),Weight=0.0012),(Item=(Type=Bone,Name=\"dyn_jacket_r_02\"),Weight=0.0015),(Item=(Type=Bone,Name=\"upperarm_twist_01_r\"),Weight=0.0047),(Item=(Type=Bone,Name=\"dyn_jacke ...(truncated)"
    Weight=1
   )
  ]}
 DeltaRotation_ContainedGraph {nodes [
   n1 = RigVMFunctionEntryNode()
   RigVMFunction_MathBoolToInteger = &RigVMFunction_MathBoolToInteger(Value=@n1.onlyTwist)
   RigVMFunction_MathQuaternionSwingTwist = &RigVMFunction_MathQuaternionSwingTwist(
    twistAxis=@n1.twistAxis
    Input=&RigUnit_GetRelativeTransformForItem(
     bChildInitial=False
     bParentInitial=False
     child=(Type=Bone,Name="None")
     parent=(Type=Bone,Name="None")
     child.Name=@n1.child
     parent.Name=@n1.parent
     parent.Type=@n1.parentType
    ).RelativeTransform.Rotation
   )
   RigVMFunction_MathQuaternionToEuler = &RigVMFunction_MathQuaternionToEuler(
    Value=RigVMDispatchNode(
     Values=((),())
     Index=@RigVMFunction_MathBoolToInteger.Result
     Values.0=@RigVMFunction_MathQuaternionSwingTwist.Swing
     Values.1=@RigVMFunction_MathQuaternionSwingTwist.Twist
    ) {template="DISPATCH_RigVMDispatch_SelectInt32(in Index,in Values,out Result)"}.Result
    RotationOrder=@n1.Rotation Order
   )
   > DISPATCH_RigVMDispatch_Print = RigVMDispatchNode(ScreenColor=(R=1,G=1,B=1,A=1), ScreenDuration=0.1, Enabled=@n1.debugRotation, Value=@RigVMFunction_MathQuaternionToEuler.Result) {template="DISPATCH_RigVMDispatch_Print(in Prefix,in Value,in Enabled,in ScreenDuration,in ScreenColor)"}
   Clamp = &RigVMFunction_MathIntClamp(Value=@n1.primAxis)
   DISPATCH_RigVMDispatch_SelectInt32 = RigVMDispatchNode(
    Values=(0,0,0)
    Index=@Clamp.Result
    Values.0=@RigVMFunction_MathQuaternionToEuler.Result.X
    Values.1=@RigVMFunction_MathQuaternionToEuler.Result.Y
    Values.2=@RigVMFunction_MathQuaternionToEuler.Result.Z
   ) {template="DISPATCH_RigVMDispatch_SelectInt32(in Index,in Values,out Result)"}
   Remap = &RigVMFunction_MathDoubleRemap(
    TargetMaximum=1
    TargetMinimum=0
    bClamp=False
    Value=@DISPATCH_RigVMDispatch_SelectInt32.Result
    SourceMaximum=@n1.max
    SourceMinimum=@n1.min
   )
   Multiply = &RigVMFunction_MathDoubleMul(A=&RigVMFunction_MathDoubleClamp(Value=@Remap.Result).Result, B=@n1.premult)
   > RigVMFunctionReturnNode(outputWeight=@Multiply.Result)
  ]}
 Multiply_2_ContainedGraph {nodes [
   n1 = RigVMFunctionEntryNode()
   RigVMFunctionReturnNode(Result=&RigVMFunction_MathDoubleMul(B=@n1.C, A=&RigVMFunction_MathDoubleMul(A=@n1.A, B=@n1.B).Result).Result)
  ]}
 Multiply_2_ContainedGraph {nodes [
   n1 = RigVMFunctionEntryNode()
   RigVMFunctionReturnNode(Result=&RigVMFunction_MathDoubleMul(B=@n1.C, A=&RigVMFunction_MathDoubleMul(A=@n1.A, B=@n1.B).Result).Result)
  ]}
 New Function_ContainedGraph {nodes [
   n1 = RigVMFunctionEntryNode()
   RigVMFunction_ControlFlowBranch = &RigVMFunction_ControlFlowBranch(Condition=@n1.Debug, ExecuteContext=@n1.ExecuteContext)
   RigVMFunctionReturnNode(ExecuteContext=@RigVMFunction_ControlFlowBranch.Completed)
   DISPATCH_RigVMDispatch_ArrayIterator = RigVMDispatchNode(
    Array=/Script/ControlRig.RigUnit_CollectionGetAll(TypeToSearch=Control) {is=struct RigVMUnitNode}.Items
    ExecuteContext=@RigVMFunction_ControlFlowBranch.True
   ) {template="DISPATCH_RigVMDispatch_ArrayIterator(in Array,out Element,out Index,out Count,out Ratio)"}
   Replace = /Script/RigVM.RigVMFunction_NameReplace(Old=_ctrl, Name=@DISPATCH_RigVMDispatch_ArrayIterator.Element.Name) {is=struct RigVMUnitNode, template="Replace::Execute(in Name,in Old,in New,out Result)"}
   > &RigUnit_SetTransform(
    Item=(Type=Bone,Name="spine_01")
    Weight=1
    bInitial=False
    Item.Name=@Replace.Result
    Value=&RigUnit_GetTransform(Item=(Type=Control,Name="spine_01_ctrl"), bInitial=False, Item.Name=@DISPATCH_RigVMDispatch_ArrayIterator.Element.Name).Transform
   )
   > TwistSetup = RigVMCollapseNode()
  ]}
 New Function_ContainedGraph {nodes [
   n1 = RigVMFunctionEntryNode()
   RigVMFunction_Sequence = &RigVMFunction_Sequence(ExecuteContext=@n1.ExecuteContext)
   Constraints = RigVMCollapseNode(ExecutePin=@RigVMFunction_Sequence.A)
   &RigUnit_ModifyTransforms(
    ItemToModify="((Item=(Type=Bone,Name=\"base_buckle_dw_r\"),Transform=(Rotation=(X=0.0114,Y=0.13,Z=-0.0864,W=0.9877),Translation=(X=0.3,Y=-0.054,Z=-0.082),Scale3D=(X=1,Y=1,Z=1))),(Item=(Type=Bone,Name=\"base_buckle_up_r\"),Transform=(Rotation=(X=-0.0708,Y=0.1392,Z=-0.0465,W= ...(truncated)"
    Weight=1
    WeightMinimum=0
    ExecutePin=@Constraints.ExecutePin
   )
   > DeltaRotation v4 = RigVMFunctionReferenceNode(
    Rotation Order=ZYX
    absoluteValue=false
    child=(Type=Bone,Name="spine_05")
    debugRotation=false
    onlyTwist=true
    outputMinMax=(X=0,Y=1)
    parent=(Type=Bone,Name="root")
    premult=1
    rotationMinMax=(X=4,Y=45)
    twistAxis=(X=1,Y=0,Z=0)
   )
   > &RigUnit_ModifyTransforms(
    ItemToModify="((Item=(Type=Bone,Name=\"base_buckle_up_l\"),Transform=(Rotation=(X=-0.0807,Y=0.1337,Z=-0.0317,W=0.9872),Translation=(X=0,Y=-0.3,Z=0),Scale3D=(X=1,Y=1,Z=1))),(Item=(Type=Bone,Name=\"base_buckle_dw_l\"),Transform=(Rotation=(X=-0.0159,Y=0.1313,Z=-0.0404,W=0.9904 ...(truncated)"
    WeightMinimum=0
    Weight=@"DeltaRotation v4".outputWeight
   )
   > &RigUnit_ModifyTransforms(
    ItemToModify="((Item=(Type=Bone,Name=\"None\"),Transform=(Rotation=(X=0,Y=0,Z=0,W=1),Translation=(X=0,Y=0,Z=0),Scale3D=(X=1,Y=1,Z=1))),(Item=(Type=Bone,Name=\"dyn_buckle_up_r_02\"),Transform=(Rotation=(X=0.0019,Y=0.0436,Z=-0.0436,W=0.9981),Translation=(X=0,Y=0,Z=0),Scale3D= ...(truncated)"
    WeightMinimum=0
    Weight=@"DeltaRotation v4".outputWeight
   )
   > DeltaRotation v4_1 = RigVMFunctionReferenceNode(
    Rotation Order=ZYX
    absoluteValue=false
    child=(Type=Bone,Name="spine_05")
    debugRotation=false
    onlyTwist=true
    outputMinMax=(X=0,Y=1)
    parent=(Type=Bone,Name="root")
    premult=1
    rotationMinMax=(X=4,Y=-40)
    twistAxis=(X=1,Y=0,Z=0)
   )
   > &RigUnit_ModifyTransforms(
    ItemToModify="((Item=(Type=Bone,Name=\"None\"),Transform=(Rotation=(X=0,Y=0,Z=0,W=1),Translation=(X=0,Y=0,Z=0),Scale3D=(X=1,Y=1,Z=1))),(Item=(Type=Bone,Name=\"base_buckle_up_r\"),Transform=(Rotation=(X=-0.0661,Y=-0.1275,Z=0.0511,W=-0.9883),Translation=(X=0.3113,Y=0.1559,Z=- ...(truncated)"
    WeightMinimum=0
    Weight=@"DeltaRotation v4_1".outputWeight
   )
   Get_Animation_Attribute = RigVMDispatchNode(BoneName=root, Default=0, Name=deformRig_l_upperarm_fr_alpha) {template="DISPATCH_RigDispatch_GetAnimAttribute(in Name,in BoneName,in Default,out Value,out Found)"}
   > &RigUnit_ModifyTransforms(
    ItemToModify="((Item=(Type=Bone,Name=\"base_buckle_up_l\"),Transform=(Rotation=(X=0.0436,Y=0,Z=0,W=0.999),Translation=(X=0,Y=0.25,Z=0),Scale3D=(X=1,Y=1,Z=1))),(Item=(Type=Bone,Name=\"base_buckle_dw_l\"),Transform=(Rotation=(X=-0.0436,Y=0,Z=0,W=0.999),Translation=(X=0,Y=0.5, ...(truncated)"
    WeightMinimum=0
    Weight=@Get_Animation_Attribute.Value
   )
   Get_Animation_Attribute_1 = RigVMDispatchNode(BoneName=root, Default=0, Name=deformRig_r_upperarm_fr_alpha) {template="DISPATCH_RigDispatch_GetAnimAttribute(in Name,in BoneName,in Default,out Value,out Found)"}
   > &RigUnit_ModifyTransforms(
    ItemToModify="((Item=(Type=Bone,Name=\"None\"),Transform=(Rotation=(X=0,Y=0,Z=0,W=1),Translation=(X=0,Y=0,Z=0),Scale3D=(X=1,Y=1,Z=1))),(Item=(Type=Bone,Name=\"base_buckle_up_r\"),Transform=(Rotation=(X=0,Y=0,Z=0,W=1),Translation=(X=0,Y=0.3,Z=0),Scale3D=(X=1,Y=1,Z=1))),(Item ...(truncated)"
    WeightMinimum=0
    Weight=@Get_Animation_Attribute_1.Value
   )
   n2 = RigVMVariableNode() {variable=Pistol ADS up}
   > &RigUnit_ModifyTransforms(
    ItemToModify="((Item=(Type=Bone,Name=\"base_buckle_up_l\"),Transform=(Rotation=(X=0.0436,Y=0.0436,Z=-0.0019,W=0.9981),Translation=(X=0,Y=0.25,Z=0),Scale3D=(X=1,Y=1,Z=1))),(Item=(Type=Bone,Name=\"base_buckle_dw_l\"),Transform=(Rotation=(X=0.0436,Y=0.0349,Z=-0.0015,W=0.9984), ...(truncated)"
    WeightMinimum=0
    Weight=@n2.Value
   )
   > &RigUnit_ModifyTransforms(
    ItemToModify="((Item=(Type=Bone,Name=\"None\"),Transform=(Rotation=(X=0,Y=0,Z=0,W=1),Translation=(X=0,Y=0,Z=0),Scale3D=(X=1,Y=1,Z=1))),(Item=(Type=Bone,Name=\"base_buckle_up_r\"),Transform=(Rotation=(X=0,Y=0,Z=0,W=1),Translation=(X=0,Y=0.3,Z=0),Scale3D=(X=1,Y=1,Z=1))),(Item ...(truncated)"
    WeightMinimum=0
    Weight=@n2.Value
   )
   RigVMFunctionReturnNode(ExecuteContext=@RigVMFunction_Sequence.B)
   &RigUnit_SphericalPoseReader(
    ActiveRegionScaleFactors=(PositiveWidth=1,NegativeWidth=1,PositiveHeight=1,NegativeHeight=1)
    ActiveRegionSize=0.035
    Debug=(bDrawDebug=True,bDraw2D=False,bDrawLocalAxes=False,DebugScale=25,DebugSegments=20,DebugThickness=0.25)
    DriverAxis=(X=1,Y=0,Z=0)
    DriverItem=(Type=Bone,Name="upperarm_l")
    ExecutePin=()
    FalloffRegionScaleFactors=(PositiveWidth=1,NegativeWidth=1,PositiveHeight=1,NegativeHeight=1)
    FalloffSize=0.12
    FlipHeightScaling=False
    FlipWidthScaling=False
    OptionalParentItem=(Type=Bone,Name="None")
    RotationOffset=(X=65,Y=0,Z=-40)
   )
  ]}
 New Function_ContainedGraph {nodes [
   RigVMFunctionEntryNode()
   > RigVMFunction_Sequence = RigVMAggregateNode()
   &RigUnit_ModifyTransforms(
    ItemToModify="((Item=(Type=Bone,Name=\"base_collar_c\"),Transform=(Rotation=(X=0,Y=0,Z=0.1437,W=-0.9896),Translation=(X=0,Y=-1,Z=-1.0361),Scale3D=(X=1,Y=1,Z=1))),(Item=(Type=Bone,Name=\"base_collar_B_l\"),Transform=(Rotation=(X=-0.048,Y=0.0047,Z=0.1126,W=-0.9925),Translatio ...(truncated)"
    Weight=1
    WeightMinimum=0
    ExecutePin=@RigVMFunction_Sequence.A
   )
   > DeltaRotation v4_1 = RigVMFunctionReferenceNode(
    Rotation Order=ZYX
    child=(Type=Bone,Name="head")
    onlyTwist=false
    outputMinMax=(X=0,Y=1)
    parent=(Type=Null,Name="head_null")
    premult=1
    rotationMinMax=(X=0,Y=-60)
    triggerAxis=Z
    twistAxis=(X=1,Y=0,Z=0)
   )
   > &RigUnit_ModifyTransforms(
    ItemToModify="((Item=(Type=Bone,Name=\"base_collar_c\"),Transform=(Rotation=(X=0,Y=0,Z=-0.1128,W=0.9936),Translation=(X=-3,Y=-0.5,Z=0),Scale3D=(X=1,Y=1,Z=1))),(Item=(Type=Bone,Name=\"base_collar_B_l\"),Transform=(Rotation=(X=-0.0151,Y=-0.173,Z=-0.0858,W=0.9811),Translation= ...(truncated)"
    WeightMinimum=0
    Weight=@"DeltaRotation v4_1".outputWeight
   )
   > &RigUnit_ModifyTransforms(
    ItemToModify="((Item=(Type=Bone,Name=\"base_collar_c\"),Transform=(Rotation=(X=0,Y=0,Z=0,W=1),Translation=(X=0,Y=0,Z=0),Scale3D=(X=1,Y=1,Z=1))),(Item=(Type=Bone,Name=\"base_collar_A_r\"),Transform=(Rotation=(X=-0.1347,Y=-0.0197,Z=0.1594,W=0.9778),Translation=(X=0,Y=0,Z=0),S ...(truncated)"
    WeightMinimum=0
    Weight=RigVMVariableNode() {variable=R_Arm_Front_SPR}.Value
   )
   > &RigUnit_ModifyTransforms(
    ItemToModify="((Item=(Type=Bone,Name=\"base_collar_c\"),Transform=(Rotation=(X=0,Y=0,Z=0,W=1),Translation=(X=0,Y=0,Z=0),Scale3D=(X=1,Y=1,Z=1))),(Item=(Type=Bone,Name=\"base_collar_A_r\"),Transform=(Rotation=(X=-0.2408,Y=-0.0499,Z=0.2749,W=0.9295),Translation=(X=0,Y=0,Z=0),S ...(truncated)"
    WeightMinimum=0
    Weight=RigVMVariableNode() {variable=R_Arm_Up_SPR}.Value
   )
   > &RigUnit_ModifyTransforms(
    ItemToModify="((Item=(Type=Bone,Name=\"base_collar_c\"),Transform=(Rotation=(X=0,Y=0,Z=0,W=1),Translation=(X=0,Y=0,Z=0),Scale3D=(X=1,Y=1,Z=1))),(Item=(Type=Bone,Name=\"base_collar_A_l\"),Transform=(Rotation=(X=0.0525,Y=0.0059,Z=-0.112,W=-0.9923),Translation=(X=0,Y=0,Z=0),Sc ...(truncated)"
    WeightMinimum=0
    Weight=RigVMVariableNode() {variable=Pistol ADS up}.Value
   )
   > &RigUnit_ModifyTransforms(
    ItemToModify="((Item=(Type=Bone,Name=\"base_collar_c\"),Transform=(Rotation=(X=0,Y=0,Z=0,W=1),Translation=(X=0,Y=0,Z=0),Scale3D=(X=1,Y=1,Z=1))),(Item=(Type=Bone,Name=\"base_collar_A_l\"),Transform=(Rotation=(X=-0.051,Y=-0.0052,Z=-0.1018,W=-0.9935),Translation=(X=0,Y=0,Z=0), ...(truncated)"
    WeightMinimum=0
    Weight=RigVMVariableNode() {variable=L_Arm_Up_SPR}.Value
   )
   > DeltaRotation v4_2 = RigVMFunctionReferenceNode(
    Rotation Order=YXZ
    child=(Type=Bone,Name="head")
    debugRotation=false
    onlyTwist=false
    outputMinMax=(X=0,Y=1)
    parent=(Type=Null,Name="head_null")
    premult=1
    rotationMinMax=(X=-30,Y=30)
    triggerAxis=Y
    twistAxis=(X=1,Y=0,Z=0)
   )
   Multiply = &RigVMFunction_MathDoubleMul(
    A=&RigVMFunction_MathDoubleRemap(SourceMaximum=0, SourceMinimum=0.5, TargetMaximum=1, TargetMinimum=0, Value=@"DeltaRotation v4_2".outputWeight).Result
    B=RigVMVariableNode() {variable=L_Arm_Up_SPR}.Value
   )
   > &RigUnit_ModifyTransforms(
    ItemToModify="((Item=(Type=Bone,Name=\"base_collar_c\"),Transform=(Rotation=(X=0.2206,Y=-0.0322,Z=0.1132,W=0.9682),Translation=(X=0,Y=0,Z=0),Scale3D=(X=1,Y=1,Z=1))),(Item=(Type=Bone,Name=\"base_collar_B_l\"),Transform=(Rotation=(X=0,Y=0,Z=0,W=1),Translation=(X=0,Y=0,Z=0),Sc ...(truncated)"
    WeightMinimum=0
    Weight=@Multiply.Result
   )
   RigVMFunctionReturnNode(ExecuteContext=@RigVMFunction_Sequence.D)
   Get_Animation_Attribute = RigVMDispatchNode(BoneName=root, Default=0, Name=deformRig_l_upperarm_fr_alpha) {template="DISPATCH_RigDispatch_GetAnimAttribute(in Name,in BoneName,in Default,out Value,out Found)"}
   &RigUnit_ModifyTransforms(
    ItemToModify="((Item=(Type=Bone,Name=\"base_collar_c\"),Transform=(Rotation=(X=0,Y=0,Z=0,W=1),Translation=(X=0,Y=0,Z=0),Scale3D=(X=1,Y=1,Z=1))),(Item=(Type=Bone,Name=\"base_collar_A_l\"),Transform=(Rotation=(X=0,Y=0,Z=0.0872,W=0.9962),Translation=(X=0,Y=0,Z=0),Scale3D=(X=1, ...(truncated)"
    WeightMinimum=0
    Weight=&RigVMFunction_MathDoubleMul(A=@Get_Animation_Attribute.Value, B=RigVMVariableNode() {variable=isNotSkydiving}.Value).Result
    ExecutePin=@RigVMFunction_Sequence.B
   )
   Get_Animation_Attribute_1 = RigVMDispatchNode(BoneName=root, Default=0, Name=deformRig_r_upperarm_fr_alpha) {template="DISPATCH_RigDispatch_GetAnimAttribute(in Name,in BoneName,in Default,out Value,out Found)"}
   > &RigUnit_ModifyTransforms(
    ItemToModify="((Item=(Type=Bone,Name=\"base_collar_c\"),Transform=(Rotation=(X=0,Y=0,Z=0,W=1),Translation=(X=0,Y=0,Z=0),Scale3D=(X=1,Y=1,Z=1))),(Item=(Type=Bone,Name=\"base_collar_A_r\"),Transform=(Rotation=(X=-0.0435,Y=-0.0038,Z=0.0871,W=0.9952),Translation=(X=0,Y=0,Z=0),S ...(truncated)"
    WeightMinimum=0
    Weight=&RigVMFunction_MathDoubleMul(A=@Get_Animation_Attribute_1.Value, B=RigVMVariableNode() {variable=isNotSkydiving}.Value).Result
   )
   > &RigUnit_ModifyTransforms(
    ItemToModify="((Item=(Type=Bone,Name=\"base_collar_A_r\"),Transform=(Rotation=(X=0,Y=0,Z=0.0872,W=0.9962),Translation=(X=0,Y=0,Z=0),Scale3D=(X=1,Y=1,Z=1))),(Item=(Type=Bone,Name=\"base_collar_B_r\"),Transform=(Rotation=(X=0,Y=0,Z=0.1305,W=0.9914),Translation=(X=0,Y=0,Z=0),S ...(truncated)"
    WeightMinimum=0
    Weight=RigVMVariableNode() {variable=Large_ADS_down}.Value
   )
   > &RigUnit_ModifyTransforms(
    ItemToModify="((Item=(Type=Bone,Name=\"base_collar_A_l\"),Transform=(Rotation=(X=0,Y=0,Z=0,W=1),Translation=(X=0,Y=0,Z=0),Scale3D=(X=1,Y=1,Z=1))),(Item=(Type=Bone,Name=\"base_collar_B_l\"),Transform=(Rotation=(X=0,Y=0,Z=-0.1109,W=-0.9938),Translation=(X=0,Y=0,Z=0),Scale3D=( ...(truncated)"
    WeightMinimum=0
    Weight=RigVMVariableNode() {variable=Large_ADS_up}.Value
   )
   > &RigUnit_ModifyTransforms(
    ItemToModify="((Item=(Type=Bone,Name=\"base_collar_A_l\"),Transform=(Rotation=(X=0,Y=0,Z=0.0872,W=0.9962),Translation=(X=0,Y=0,Z=0),Scale3D=(X=1,Y=1,Z=1))),(Item=(Type=Bone,Name=\"base_collar_B_l\"),Transform=(Rotation=(X=0,Y=0,Z=0.2588,W=0.9659),Translation=(X=0,Y=0,Z=0),S ...(truncated)"
    WeightMinimum=0
    Weight=RigVMVariableNode() {variable=L_Clav_SPR}.Value
   )
   > &RigUnit_ModifyTransforms(
    ItemToModify="((Item=(Type=Bone,Name=\"base_collar_A_r\"),Transform=(Rotation=(X=0,Y=0,Z=0.0872,W=0.9962),Translation=(X=0,Y=0,Z=0),Scale3D=(X=1,Y=1,Z=1))),(Item=(Type=Bone,Name=\"base_collar_B_r\"),Transform=(Rotation=(X=0,Y=0,Z=0.342,W=0.9397),Translation=(X=0,Y=0,Z=0),Sc ...(truncated)"
    WeightMinimum=0
    Weight=RigVMVariableNode() {variable=R_Clav_SPR}.Value
   )
  ]}
 New Function_ContainedGraph {nodes [
   RigVMFunctionEntryNode()
   > RigVMFunction_Sequence = RigVMAggregateNode()
   DeltaRotation v4_1 = RigVMFunctionReferenceNode(
    Rotation Order=ZYX
    child=(Type=Bone,Name="head")
    onlyTwist=false
    outputMinMax=(X=0,Y=1)
    parent=(Type=Null,Name="head_null")
    premult=1
    rotationMinMax=(X=-20,Y=-60)
    triggerAxis=Z
    twistAxis=(X=1,Y=0,Z=0)
    ExecuteContext=@RigVMFunction_Sequence.A
   )
   > &RigUnit_ModifyTransforms(
    ItemToModify="((Item=(Type=Bone,Name=\"base_lapel_c\"),Transform=(Rotation=(X=0.0003,Y=-0.0001,Z=-0.342,W=0.9397),Translation=(X=-2.4219,Y=0,Z=-0.0006),Scale3D=(X=1,Y=1,Z=1))),(Item=(Type=Bone,Name=\"base_lapel_C_l\"),Transform=(Rotation=(X=0,Y=0,Z=-0.1736,W=0.9848),Transla ...(truncated)"
    WeightMinimum=0
    Weight=&RigVMFunction_MathDoubleMul(A=@"DeltaRotation v4_1".outputWeight, B=RigVMVariableNode() {variable=isNotSkydiving}.Value).Result
   )
   > &RigUnit_ModifyTransforms(
    ItemToModify="((Item=(Type=Bone,Name=\"base_lapel_c\"),Transform=(Rotation=(X=0,Y=0,Z=0,W=1),Translation=(X=0,Y=0,Z=0),Scale3D=(X=1,Y=1,Z=1))),(Item=(Type=Bone,Name=\"base_lapel_A_r\"),Transform=(Rotation=(X=-0.0874,Y=-0.0228,Z=0.2517,W=0.9636),Translation=(X=0,Y=1,Z=0),Sca ...(truncated)"
    WeightMinimum=0
    Weight=RigVMVariableNode() {variable=R_Arm_Front_SPR}.Value
   )
   > &RigUnit_ModifyTransforms(
    ItemToModify="((Item=(Type=Bone,Name=\"dyn_lapel_A_r_03\"),Transform=(Rotation=(X=0.0989,Y=0.0256,Z=-0.2498,W=-0.9629),Translation=(X=0,Y=0,Z=0),Scale3D=(X=1,Y=1,Z=1))),(Item=(Type=Bone,Name=\"base_lapel_A_r\"),Transform=(Rotation=(X=-0.0086,Y=0.0074,Z=-0.4403,W=-0.8978),Tr ...(truncated)"
    WeightMinimum=0
    Weight=RigVMVariableNode() {variable=R_Arm_Up_SPR}.Value
   )
   > &RigUnit_ModifyTransforms(
    ItemToModify="((Item=(Type=Bone,Name=\"base_lapel_c\"),Transform=(Rotation=(X=0,Y=0,Z=0,W=1),Translation=(X=0,Y=0,Z=0),Scale3D=(X=1,Y=1,Z=1))),(Item=(Type=Bone,Name=\"base_lapel_A_l\"),Transform=(Rotation=(X=0.1178,Y=0.0265,Z=-0.2188,W=-0.9683),Translation=(X=0,Y=0,Z=0),Sca ...(truncated)"
    WeightMinimum=0
    Weight=RigVMVariableNode() {variable=Pistol ADS up}.Value
   )
   > &RigUnit_ModifyTransforms(
    ItemToModify="((Item=(Type=Bone,Name=\"base_lapel_c\"),Transform=(Rotation=(X=0,Y=0,Z=0,W=1),Translation=(X=0,Y=0,Z=0),Scale3D=(X=1,Y=1,Z=1))),(Item=(Type=Bone,Name=\"base_lapel_A_l\"),Transform=(Rotation=(X=-0.01,Y=-0.0038,Z=-0.3632,W=-0.9317),Translation=(X=0,Y=0,Z=0),Sca ...(truncated)"
    WeightMinimum=0
    Weight=RigVMVariableNode() {variable=L_Arm_Up_SPR}.Value
   )
   > DeltaRotation v4_2 = RigVMFunctionReferenceNode(
    Rotation Order=YXZ
    child=(Type=Bone,Name="head")
    debugRotation=false
    onlyTwist=false
    outputMinMax=(X=0,Y=1)
    parent=(Type=Null,Name="head_null")
    premult=1
    rotationMinMax=(X=-30,Y=30)
    triggerAxis=Y
    twistAxis=(X=1,Y=0,Z=0)
   )
   Multiply = &RigVMFunction_MathDoubleMul(
    A=&RigVMFunction_MathDoubleRemap(SourceMaximum=0, SourceMinimum=0.5, TargetMaximum=1, TargetMinimum=0, Value=@"DeltaRotation v4_2".outputWeight).Result
    B=RigVMVariableNode() {variable=L_Arm_Up_SPR}.Value
   )
   > &RigUnit_ModifyTransforms(
    ItemToModify="((Item=(Type=Bone,Name=\"base_lapel_c\"),Transform=(Rotation=(X=-0.1349,Y=-0.0423,Z=0.1107,W=0.9838),Translation=(X=-0.178,Y=0.8022,Z=-0.203),Scale3D=(X=1,Y=1,Z=1))),(Item=(Type=Bone,Name=\"base_lapel_C_l\"),Transform=(Rotation=(X=0.1043,Y=0,Z=0,W=0.9945),Tran ...(truncated)"
    WeightMinimum=0
    Weight=@Multiply.Result
   )
   > &RigUnit_ModifyTransforms(
    ItemToModify="((Item=(Type=Bone,Name=\"base_lapel_c\"),Transform=(Rotation=(X=0,Y=0,Z=0,W=1),Translation=(X=0,Y=0,Z=0),Scale3D=(X=1,Y=1,Z=1))),(Item=(Type=Bone,Name=\"base_lapel_A_r\"),Transform=(Rotation=(X=0,Y=0,Z=0.2164,W=0.9763),Translation=(X=-0.4631,Y=1.1528,Z=0),Scal ...(truncated)"
    WeightMinimum=0
    Weight=RigVMVariableNode() {variable=Large_ADS_down}.Value
   )
   > &RigUnit_ModifyTransforms(
    ItemToModify="((Item=(Type=Bone,Name=\"base_lapel_c\"),Transform=(Rotation=(X=0.3376,Y=-0.058,Z=-0.0613,W=0.9375),Translation=(X=0,Y=0,Z=0),Scale3D=(X=1,Y=1,Z=1))),(Item=(Type=Bone,Name=\"base_lapel_A_l\"),Transform=(Rotation=(X=0.116,Y=-0.0098,Z=0,W=-0.9932),Translation=(X ...(truncated)"
    WeightMinimum=0
    Weight=RigVMVariableNode() {variable=Large_ADS_up}.Value
   )
   RigVMFunctionReturnNode(ExecuteContext=@RigVMFunction_Sequence.C)
   Get_Animation_Attribute = RigVMDispatchNode(BoneName=root, Default=0, Name=deformRig_l_upperarm_fr_alpha) {template="DISPATCH_RigDispatch_GetAnimAttribute(in Name,in BoneName,in Default,out Value,out Found)"}
   &RigUnit_ModifyTransforms(
    ItemToModify="((Item=(Type=Bone,Name=\"base_lapel_c\"),Transform=(Rotation=(X=0,Y=0,Z=0,W=1),Translation=(X=0,Y=0,Z=0),Scale3D=(X=1,Y=1,Z=1))),(Item=(Type=Bone,Name=\"base_lapel_A_l\"),Transform=(Rotation=(X=0,Y=0,Z=0.2588,W=0.9659),Translation=(X=0,Y=0,Z=0),Scale3D=(X=1,Y= ...(truncated)"
    WeightMinimum=0
    Weight=&RigVMFunction_MathDoubleMul(A=@Get_Animation_Attribute.Value, B=RigVMVariableNode() {variable=isNotSkydiving}.Value).Result
    ExecutePin=@RigVMFunction_Sequence.B
   )
   Get_Animation_Attribute_1 = RigVMDispatchNode(BoneName=root, Default=0, Name=deformRig_r_upperarm_fr_alpha) {template="DISPATCH_RigDispatch_GetAnimAttribute(in Name,in BoneName,in Default,out Value,out Found)"}
   > &RigUnit_ModifyTransforms(
    ItemToModify="((Item=(Type=Bone,Name=\"base_lapel_c\"),Transform=(Rotation=(X=0,Y=0,Z=0,W=1),Translation=(X=0,Y=0,Z=0),Scale3D=(X=1,Y=1,Z=1))),(Item=(Type=Bone,Name=\"base_lapel_A_r\"),Transform=(Rotation=(X=-0.0451,Y=-0.0182,Z=0.2098,W=0.9765),Translation=(X=0,Y=1,Z=0),Sca ...(truncated)"
    WeightMinimum=0
    Weight=&RigVMFunction_MathDoubleMul(A=@Get_Animation_Attribute_1.Value, B=RigVMVariableNode() {variable=isNotSkydiving}.Value).Result
   )
   > &RigUnit_ModifyTransforms(
    ItemToModify="((Item=(Type=Bone,Name=\"base_lapel_A_l\"),Transform=(Rotation=(X=-0.1677,Y=-0.0449,Z=0.2549,W=0.9513),Translation=(X=0,Y=0,Z=0),Scale3D=(X=1,Y=1,Z=1))),(Item=(Type=Bone,Name=\"base_lapel_B_l\"),Transform=(Rotation=(X=0,Y=0,Z=0,W=1),Translation=(X=0,Y=2,Z=0),S ...(truncated)"
    WeightMinimum=0
    Weight=RigVMVariableNode() {variable=L_Clav_SPR}.Value
   )
   > &RigUnit_ModifyTransforms(
    ItemToModify="((Item=(Type=Bone,Name=\"base_lapel_A_r\"),Transform=(Rotation=(X=-0.171,Y=-0.0302,Z=0.171,W=0.9698),Translation=(X=0,Y=0,Z=0),Scale3D=(X=1,Y=1,Z=1))),(Item=(Type=Bone,Name=\"base_lapel_B_r\"),Transform=(Rotation=(X=0,Y=0,Z=0.2588,W=0.9659),Translation=(X=0,Y= ...(truncated)"
    WeightMinimum=0
    Weight=RigVMVariableNode() {variable=R_Clav_SPR}.Value
   )
   > &RigUnit_ModifyTransforms(
    ItemToModify="((Item=(Type=Bone,Name=\"base_lapel_c\"),Transform=(Rotation=(X=0,Y=0,Z=-0.0872,W=0.9962),Translation=(X=0,Y=0,Z=0),Scale3D=(X=1,Y=1,Z=1))),(Item=(Type=Bone,Name=\"base_lapel_C_l\"),Transform=(Rotation=(X=0,Y=0,Z=-0.0872,W=0.9962),Translation=(X=0,Y=0,Z=0),Sca ...(truncated)"
    WeightMinimum=0
    Weight=RigVMVariableNode() {variable=isSkydiving}.Value
   )
  ]}
 New Function_ContainedGraph {nodes [
   RigVMFunctionEntryNode()
   > RigVMFunction_Sequence = RigVMAggregateNode()
   DeltaRotation v4 = RigVMFunctionReferenceNode(
    Rotation Order=ZYX
    child=(Type=Bone,Name="thigh_l")
    debugRotation=false
    outputMinMax=(X=0,Y=1)
    parent=(Type=Null,Name="thigh_l_null")
    premult=1
    rotationMinMax=(X=0,Y=-89)
    triggerAxis=Z
    twistAxis=(X=1,Y=0,Z=0)
    ExecuteContext=@RigVMFunction_Sequence.A
   )
   > &RigUnit_ModifyTransforms(
    ItemToModify="((Item=(Type=Bone,Name=\"dyn_shirt_l_01\"),Transform=(Rotation=(X=-0.0158,Y=0.3015,Z=-0.4702,W=0.8293),Translation=(X=-3.9219,Y=2.4968,Z=2.3334),Scale3D=(X=1,Y=1,Z=1))),(Item=(Type=Bone,Name=\"dyn_shirt_l_02\"),Transform=(Rotation=(X=0,Y=0,Z=0,W=1),Translation ...(truncated)"
    WeightMinimum=0
    Weight=@"DeltaRotation v4".outputWeight
   )
   > DeltaRotation v4_2 = RigVMFunctionReferenceNode(
    Rotation Order=ZYX
    child=(Type=Bone,Name="thigh_l")
    debugRotation=false
    outputMinMax=(X=0,Y=1)
    parent=(Type=Null,Name="thigh_l_null")
    premult=1
    rotationMinMax=(X=-90,Y=-110)
    triggerAxis=Z
    twistAxis=(X=1,Y=0,Z=0)
   )
   > &RigUnit_ModifyTransforms(
    ItemToModify="((Item=(Type=Bone,Name=\"dyn_shirt_l_01\"),Transform=(Rotation=(X=-0.0574,Y=0.0067,Z=-0.116,W=0.9916),Translation=(X=-7.255,Y=-1.3305,Z=0),Scale3D=(X=1,Y=1,Z=1))),(Item=(Type=Bone,Name=\"dyn_shirt_l_02\"),Transform=(Rotation=(X=0,Y=0,Z=0,W=1),Translation=(X=0, ...(truncated)"
    WeightMinimum=0
    Weight=@"DeltaRotation v4_2".outputWeight
   )
   > &RigUnit_ModifyTransforms(
    ItemToModify=((Item=(Type=Bone,Name="dyn_shirtStrap_l_01"),Transform=(Rotation=(X=0.5,Y=0,Z=0,W=0.866),Translation=(X=0,Y=0,Z=-1.5),Scale3D=(X=1,Y=1,Z=1))))
    WeightMinimum=0
    Weight=@"DeltaRotation v4_2".outputWeight
   )
   Multiply = &RigVMFunction_MathDoubleMul(A=@"DeltaRotation v4_2".outputWeight, B=RigVMVariableNode() {variable=isDBNO}.Value)
   > &RigUnit_ModifyTransforms(
    ItemToModify=((Item=(Type=Bone,Name="dyn_shirtStrap_l_01"),Transform=(Rotation=(X=-0.2393,Y=0.1277,Z=-0.1449,W=0.9515),Translation=(X=0,Y=0,Z=0),Scale3D=(X=1,Y=1,Z=1))))
    WeightMinimum=0
    Weight=@Multiply.Result
   )
   > &RigUnit_ModifyTransforms(
    ItemToModify=((Item=(Type=Bone,Name="dyn_shirtStrap_l_01"),Transform=(Rotation=(X=0.0151,Y=0.173,Z=-0.0858,W=0.9811),Translation=(X=0,Y=0,Z=0),Scale3D=(X=1,Y=1,Z=1))))
    WeightMinimum=0
    Weight=RigVMVariableNode() {variable=isDBNO}.Value
   )
   > &RigUnit_ModifyTransforms(
    ItemToModify="((Item=(Type=Bone,Name=\"dyn_shirtStrap_l_01\"),Transform=(Rotation=(X=0,Y=0,Z=0,W=1),Translation=(X=0,Y=0,Z=0),Scale3D=(X=1,Y=1,Z=1))),(Item=(Type=Bone,Name=\"dyn_shirtStrap_r_01\"),Transform=(Rotation=(X=0,Y=-0.0436,Z=0,W=0.999),Translation=(X=0,Y=0,Z=0),Sca ...(truncated)"
    Weight=1
    WeightMinimum=0
   )
   DeltaRotation v4_1 = RigVMFunctionReferenceNode(
    Rotation Order=ZYX
    child=(Type=Bone,Name="thigh_r")
    debugRotation=false
    outputMinMax=(X=0,Y=1)
    parent=(Type=Null,Name="thigh_r_null")
    premult=1
    rotationMinMax=(X=0,Y=-89)
    triggerAxis=Z
    twistAxis=(X=1,Y=0,Z=0)
    ExecuteContext=@RigVMFunction_Sequence.B
   )
   > &RigUnit_ModifyTransforms(
    ItemToModify="((Item=(Type=Bone,Name=\"dyn_shirt_r_01\"),Transform=(Rotation=(X=0.0045,Y=-0.2964,Z=-0.4535,W=0.8405),Translation=(X=-4.1829,Y=0.1594,Z=-3.3346),Scale3D=(X=1,Y=1,Z=1))),(Item=(Type=Bone,Name=\"dyn_shirt_r_02\"),Transform=(Rotation=(X=-0.0755,Y=0.0048,Z=0.1042 ...(truncated)"
    WeightMinimum=0
    Weight=@"DeltaRotation v4_1".outputWeight
   )
   > DeltaRotation v4_3 = RigVMFunctionReferenceNode(
    Rotation Order=ZYX
    child=(Type=Bone,Name="thigh_r")
    debugRotation=false
    outputMinMax=(X=0,Y=1)
    parent=(Type=Null,Name="thigh_r_null")
    premult=1
    rotationMinMax=(X=-90,Y=-110)
    triggerAxis=Z
    twistAxis=(X=1,Y=0,Z=0)
   )
   > &RigUnit_ModifyTransforms(
    ItemToModify="((Item=(Type=Bone,Name=\"dyn_shirt_r_01\"),Transform=(Rotation=(X=0.1096,Y=-0.0229,Z=-0.1674,W=0.9795),Translation=(X=-5.5645,Y=0.4062,Z=0.084),Scale3D=(X=1,Y=1,Z=1))),(Item=(Type=Bone,Name=\"dyn_shirt_r_02\"),Transform=(Rotation=(X=0,Y=0,Z=0.0162,W=-0.9999),T ...(truncated)"
    WeightMinimum=0
    Weight=@"DeltaRotation v4_3".outputWeight
   )
   > &RigUnit_ModifyTransforms(
    ItemToModify=((Item=(Type=Bone,Name="dyn_shirtStrap_r_01"),Transform=(Rotation=(X=-0.5,Y=0,Z=0,W=0.866),Translation=(X=0,Y=0,Z=1.5),Scale3D=(X=1,Y=1,Z=1))))
    WeightMinimum=0
    Weight=@"DeltaRotation v4_3".outputWeight
   )
   RigVMFunctionReturnNode(ExecuteContext=@RigVMFunction_Sequence.C)
  ]}
 New Function_ContainedGraph {nodes [
   n1 = RigVMFunctionEntryNode()
   RigVMFunction_MathBoolToInteger = &RigVMFunction_MathBoolToInteger(Value=@n1.onlyTwist)
   RigVMFunction_MathQuaternionSwingTwist = &RigVMFunction_MathQuaternionSwingTwist(
    TwistAxis=@n1.twistAxis
    Input=&RigUnit_GetRelativeTransformForItem(bChildInitial=False, bParentInitial=False, Child=@n1.child, Parent=@n1.parent).RelativeTransform.Rotation
   )
   n2 = RigVMDispatchNode(
    Values=((),())
    Index=@RigVMFunction_MathBoolToInteger.Result
    Values.0=@RigVMFunction_MathQuaternionSwingTwist.Swing
    Values.1=@RigVMFunction_MathQuaternionSwingTwist.Twist
   ) {template="DISPATCH_RigVMDispatch_SelectInt32(in Index,in Values,out Result)"}
   RigVMFunction_MathQuaternionToEuler = &RigVMFunction_MathQuaternionToEuler(Value=@n2.Result, RotationOrder=@n1.Rotation Order)
   > DISPATCH_RigVMDispatch_Print = RigVMDispatchNode(ScreenColor=(R=1,G=1,B=1,A=1), ScreenDuration=0.05, Enabled=@n1.debugRotation, Value=@RigVMFunction_MathQuaternionToEuler.Result) {template="DISPATCH_RigVMDispatch_Print(in Prefix,in Value,in Enabled,in ScreenDuration,in ScreenColor)"}
   Cast = RigVMDispatchNode(Value=@n1.triggerAxis) {template="DISPATCH_RigVMDispatch_CastEnumToInt(in Value,out Result)"}
   DISPATCH_RigVMDispatch_SelectInt32 = RigVMDispatchNode(
    Values=(0,0,0)
    Index=@Cast.Result
    Values.0=@RigVMFunction_MathQuaternionToEuler.Result.X
    Values.1=@RigVMFunction_MathQuaternionToEuler.Result.Y
    Values.2=@RigVMFunction_MathQuaternionToEuler.Result.Z
   ) {template="DISPATCH_RigVMDispatch_SelectInt32(in Index,in Values,out Result)"}
   Remap = &RigVMFunction_MathDoubleRemap(
    bClamp=False
    Value=@DISPATCH_RigVMDispatch_SelectInt32.Result
    TargetMinimum=@n1.outputMinMax.X
    TargetMaximum=@n1.outputMinMax.Y
    SourceMinimum=@n1.rotationMinMax.X
    SourceMaximum=@n1.rotationMinMax.Y
   )
   Absolute = &RigVMFunction_MathDoubleAbs(Value=@Remap.Result)
   Multiply = &RigVMFunction_MathDoubleMul(
    A=&RigVMFunction_MathDoubleClamp(Value=RigVMDispatchNode(True=@Absolute.Result, Condition=@n1.absoluteValue, False=@Remap.Result) {template="DISPATCH_RigVMDispatch_If(in Condition,in True,in False,out Result)"}.Result).Result
    B=@n1.premult
   )
   > RigVMFunctionReturnNode(outputQuat=@n2.Result, outputWeight=@Multiply.Result)
  ]}
 New Function_ContainedGraph {nodes [
   RigVMFunctionEntryNode()
   > RigVMFunction_Sequence = RigVMAggregateNode()
   Remap = &RigVMFunction_MathDoubleRemap(SourceMaximum=0, SourceMinimum=1, TargetMaximum=1, TargetMinimum=0, bClamp=False, Value=RigVMVariableNode() {variable=isSkydiving}.Value)
   RigVMVariableNode(Value=@Remap.Result, ExecuteContext=@RigVMFunction_Sequence.A) {variable=isNotSkydiving}
   n1 = &RigUnit_SphericalPoseReader(
    ActiveRegionScaleFactors=(PositiveWidth=1,NegativeWidth=0.5,PositiveHeight=0.75,NegativeHeight=0.75)
    ActiveRegionSize=0.1
    Debug=(bDrawDebug=false,bDraw2D=False,bDrawLocalAxes=False,DebugScale=25,DebugSegments=20,DebugThickness=0.25)
    DriverAxis=(X=0,Y=1,Z=0)
    DriverItem=(Type=Bone,Name="head")
    FalloffRegionScaleFactors=(PositiveWidth=1,NegativeWidth=0.5,PositiveHeight=1,NegativeHeight=1)
    FalloffSize=0.2
    FlipHeightScaling=False
    FlipWidthScaling=False
    OptionalParentItem=(Type=Bone,Name="None")
    RotationOffset=(X=90,Y=0,Z=-30)
    ExecutePin=@RigVMFunction_Sequence.B
   )
   > n2 = &RigUnit_SphericalPoseReader(
    ActiveRegionScaleFactors=(PositiveWidth=1,NegativeWidth=1,PositiveHeight=1,NegativeHeight=1)
    ActiveRegionSize=0.025
    Debug=(bDrawDebug=false,bDraw2D=False,bDrawLocalAxes=False,DebugScale=25,DebugSegments=20,DebugThickness=0.25)
    DriverAxis=(X=-1,Y=0,Z=0)
    DriverItem=(Type=Bone,Name="upperarm_r")
    FalloffRegionScaleFactors=(PositiveWidth=1,NegativeWidth=1,PositiveHeight=1,NegativeHeight=1)
    FalloffSize=0.1
    FlipHeightScaling=False
    FlipWidthScaling=False
    OptionalParentItem=(Type=Bone,Name="None")
    RotationOffset=(X=-115,Y=0,Z=25)
   )
   Multiply = &RigVMFunction_MathDoubleMul(A=@n2.OutputParam, B=@n1.OutputParam)
   n3 = &RigVMFunction_AlphaInterp(InterpSpeedDecreasing=10, InterpSpeedIncreasing=10, Scale=1, bClampResult=true, Value=@Multiply.Result)
   > RigVMVariableNode(Value=@n3.Result) {variable=Pistol ADS up}
   > n4 = &RigUnit_SphericalPoseReader(
    ActiveRegionScaleFactors=(PositiveWidth=1,NegativeWidth=1,PositiveHeight=2,NegativeHeight=1)
    ActiveRegionSize=0.05
    Debug=(bDrawDebug=false,bDraw2D=False,bDrawLocalAxes=False,DebugScale=25,DebugSegments=20,DebugThickness=0.25)
    DriverAxis=(X=-1,Y=0,Z=0)
    DriverItem=(Type=Bone,Name="upperarm_r")
    FalloffRegionScaleFactors=(PositiveWidth=1,NegativeWidth=1,PositiveHeight=1,NegativeHeight=1)
    FalloffSize=0.25
    FlipHeightScaling=False
    FlipWidthScaling=False
    OptionalParentItem=(Type=Bone,Name="None")
    RotationOffset=(X=-80,Y=10,Z=10)
   )
   > RigVMVariableNode(Value=@n4.OutputParam) {variable=R_Arm_Front_SPR}
   > n5 = &RigUnit_SphericalPoseReader(
    ActiveRegionScaleFactors=(PositiveWidth=1,NegativeWidth=1,PositiveHeight=2,NegativeHeight=1)
    ActiveRegionSize=0.05
    Debug=(bDrawDebug=false,bDraw2D=False,bDrawLocalAxes=False,DebugScale=25,DebugSegments=20,DebugThickness=0.25)
    DriverAxis=(X=-1,Y=0,Z=0)
    DriverItem=(Type=Bone,Name="upperarm_r")
    FalloffRegionScaleFactors=(PositiveWidth=1,NegativeWidth=1,PositiveHeight=1,NegativeHeight=1)
    FalloffSize=0.25
    FlipHeightScaling=False
    FlipWidthScaling=False
    OptionalParentItem=(Type=Bone,Name="None")
    RotationOffset=(X=-120,Y=20,Z=20)
   )
   > RigVMVariableNode(Value=@n5.OutputParam) {variable=R_Arm_Up_SPR}
   > n6 = &RigUnit_SphericalPoseReader(
    ActiveRegionScaleFactors=(PositiveWidth=1,NegativeWidth=1,PositiveHeight=1,NegativeHeight=1)
    ActiveRegionSize=0.1
    Debug=(bDrawDebug=false,bDraw2D=False,bDrawLocalAxes=False,DebugScale=25,DebugSegments=20,DebugThickness=0.25)
    DriverAxis=(X=1,Y=0,Z=0)
    DriverItem=(Type=Bone,Name="upperarm_l")
    FalloffRegionScaleFactors=(PositiveWidth=1,NegativeWidth=1,PositiveHeight=1,NegativeHeight=1)
    FalloffSize=0.2
    FlipHeightScaling=False
    FlipWidthScaling=False
    OptionalParentItem=(Type=Bone,Name="None")
    RotationOffset=(X=15,Y=-45,Z=0)
   )
   > RigVMVariableNode(Value=&RigVMFunction_MathDoubleMul(A=@n6.OutputParam, B=RigVMVariableNode() {variable=isNotSkydiving}.Value).Result) {variable=L_Arm_Up_SPR}
   > n7 = &RigUnit_SphericalPoseReader(
    ActiveRegionScaleFactors=(PositiveWidth=0.7,NegativeWidth=1.25,PositiveHeight=0.5,NegativeHeight=0.5)
    ActiveRegionSize=0.1
    Debug=(bDrawDebug=false,bDraw2D=False,bDrawLocalAxes=False,DebugScale=25,DebugSegments=20,DebugThickness=0.25)
    DriverAxis=(X=-1,Y=0,Z=0)
    DriverItem=(Type=Bone,Name="hand_r")
    FalloffRegionScaleFactors=(PositiveWidth=0.3,NegativeWidth=2.3,PositiveHeight=0.5,NegativeHeight=0.5)
    FalloffSize=0.2
    FlipHeightScaling=False
    FlipWidthScaling=False
    OptionalParentItem=(Type=Bone,Name="None")
    RotationOffset=(X=140,Y=0,Z=30)
   )
   AlphaInterp = &RigVMFunction_AlphaInterp(
    InterpSpeedDecreasing=10
    InterpSpeedIncreasing=10
    Scale=1
    bClampResult=true
    Value=&RigVMFunction_MathDoubleMul(A=@n7.OutputParam, B=RigVMVariableNode() {variable=isLargeWeapon}.Value).Result
   )
   > RigVMVariableNode(Value=@AlphaInterp.Result) {variable=Large_ADS_down}
   > n8 = &RigUnit_SphericalPoseReader(
    ActiveRegionScaleFactors=(PositiveWidth=1,NegativeWidth=1,PositiveHeight=2,NegativeHeight=1)
    ActiveRegionSize=0.1
    Debug=(bDrawDebug=false,bDraw2D=False,bDrawLocalAxes=False,DebugScale=25,DebugSegments=20,DebugThickness=0.25)
    DriverAxis=(X=1,Y=0,Z=0)
    DriverItem=(Type=Bone,Name="upperarm_l")
    FalloffRegionScaleFactors=(PositiveWidth=1,NegativeWidth=1,PositiveHeight=3,NegativeHeight=1)
    FalloffSize=0.23
    FlipHeightScaling=False
    FlipWidthScaling=False
    OptionalParentItem=(Type=Bone,Name="None")
    RotationOffset=(X=-10,Y=-55,Z=0)
   )
   n9 = &RigVMFunction_AlphaInterp(
    InterpSpeedDecreasing=10
    InterpSpeedIncreasing=10
    Scale=1
    bClampResult=true
    Value=&RigVMFunction_MathDoubleMul(A=@n8.OutputParam, B=RigVMVariableNode() {variable=isLargeWeapon}.Value).Result
   )
   > RigVMVariableNode(Value=@n9.Result) {variable=Large_ADS_up}
   > n10 = &RigUnit_SphericalPoseReader(
    ActiveRegionScaleFactors=(PositiveWidth=1,NegativeWidth=1,PositiveHeight=1,NegativeHeight=1)
    ActiveRegionSize=0.1
    Debug=(bDrawDebug=false,bDraw2D=False,bDrawLocalAxes=true,DebugScale=25,DebugSegments=20,DebugThickness=0.25)
    DriverAxis=(X=1,Y=0,Z=0)
    DriverItem=(Type=Bone,Name="clavicle_l")
    FalloffRegionScaleFactors=(PositiveWidth=0.8,NegativeWidth=1,PositiveHeight=1,NegativeHeight=1)
    FalloffSize=0.2
    FlipHeightScaling=False
    FlipWidthScaling=False
    OptionalParentItem=(Type=Bone,Name="None")
    RotationOffset=(X=12,Y=-32,Z=0)
   )
   Maximum = &RigVMFunction_MathDoubleMax(A=RigVMVariableNode() {variable=isEmoting}.Value, B=RigVMVariableNode() {variable=isSkydiving}.Value)
   > RigVMVariableNode(Value=&RigVMFunction_MathDoubleMul(B=@Maximum.Result, A=@n10.OutputParam).Result) {variable=L_Clav_SPR}
   > n11 = &RigUnit_SphericalPoseReader(
    ActiveRegionScaleFactors=(PositiveWidth=1,NegativeWidth=1,PositiveHeight=1,NegativeHeight=1)
    ActiveRegionSize=0.1
    Debug=(bDrawDebug=false,bDraw2D=False,bDrawLocalAxes=true,DebugScale=25,DebugSegments=20,DebugThickness=0.25)
    DriverAxis=(X=-1,Y=0,Z=0)
    DriverItem=(Type=Bone,Name="clavicle_r")
    FalloffRegionScaleFactors=(PositiveWidth=0.8,NegativeWidth=1,PositiveHeight=1,NegativeHeight=1)
    FalloffSize=0.2
    FlipHeightScaling=False
    FlipWidthScaling=False
    OptionalParentItem=(Type=Bone,Name="None")
    RotationOffset=(X=0,Y=148,Z=0)
   )
   n12 = RigVMVariableNode() {variable=isEmoting}
   n13 = RigVMVariableNode() {variable=isSkydiving}
   > RigVMVariableNode(Value=&RigVMFunction_MathDoubleMul(B=&RigVMFunction_MathDoubleMax(A=@n12.Value, B=@n13.Value).Result, A=@n11.OutputParam).Result) {variable=R_Clav_SPR}
   RigVMFunctionReturnNode(ExecuteContext=@RigVMFunction_Sequence.C)
  ]}
 New Function_ContainedGraph {nodes [
   n1 = RigVMFunctionEntryNode()
   RigVMFunction_Sequence = &RigVMFunction_Sequence(ExecuteContext=@n1.ExecuteContext)
   RigUnit_GetRelativeTransformForItem = &RigUnit_GetRelativeTransformForItem(Child=(Type=Bone,Name="dyn_tail_r_02"), Parent=(Type=Bone,Name="dyn_tail_r_02"), bChildInitial=False, bParentInitial=true)
   RigVMFunction_MathQuaternionToAxisAndAngle = /Script/RigVM.RigVMFunction_MathQuaternionToAxisAndAngle(Value=@RigUnit_GetRelativeTransformForItem.RelativeTransform.Rotation) {is=struct RigVMUnitNode, template="ToAxisAndAngle::Execute(in Value,out Axis,out Angle)"}
   Maximum = &RigVMFunction_MathDoubleMax(
    A=@RigVMFunction_MathQuaternionToAxisAndAngle.Axis.Z
    B=&RigVMFunction_MathQuaternionToAxisAndAngle(Value=&RigUnit_GetRelativeTransformForItem(Child=(Type=Bone,Name="dyn_tail_l_02"), Parent=(Type=Bone,Name="dyn_tail_l_02"), bChildInitial=False, bParentInitial=true).RelativeTransform.Rotation) {template="ToAxisAndAngle::Execute(in Value,out Axis,out Angle)"}.Axis.Z
   )
   RigVMFunction_MathQuaternionMake = /Script/RigVM.RigVMFunction_MathQuaternionMake(W=0, X=0, Y=0, Z=@Maximum.Result) {is=struct RigVMUnitNode}
   &RigUnit_ModifyTransforms(
    ItemToModify=((Item=(Type=Bone,Name="dyn_tail_c_02"),Transform=(Rotation=(X=0,Y=0,Z=0,W=1),Translation=(X=0,Y=0,Z=0),Scale3D=(X=1,Y=1,Z=1))))
    Weight=1
    WeightMinimum=0
    ItemToModify.0.Transform.Rotation=@RigVMFunction_MathQuaternionMake.Result
    ExecutePin=@RigVMFunction_Sequence.A
   )
   RigVMFunctionReturnNode(ExecuteContext=@RigVMFunction_Sequence.B)
  ]}
 RigVMFunctionLibrary {nodes [
   Buckles = RigVMCollapseNode(ExecuteContext=())
   Collar = RigVMCollapseNode(ExecuteContext=())
   Controls = RigVMCollapseNode(Debug=False, ExecuteContext=())
   DeltaRotation = RigVMCollapseNode(ExecuteContext=(), Rotation Order=ZYX, debugRotation=false, premult=1, twistAxis=(X=1,Y=0,Z=0))
   DeltaRotation v4 = RigVMCollapseNode(
    ExecuteContext=()
    Rotation Order=ZYX
    child=(Type=Bone,Name="None")
    outputMinMax=(X=0,Y=1)
    parent=(Type=Bone,Name="None")
    premult=1
    rotationMinMax=(X=0,Y=0)
    twistAxis=(X=1,Y=0,Z=0)
   )
   Jacket = RigVMCollapseNode(ExecuteContext=())
   Lapel = RigVMCollapseNode(ExecuteContext=())
   Shirt = RigVMCollapseNode(ExecuteContext=())
   Variables = RigVMCollapseNode(ExecuteContext=())
  ]}
 RigVMFunction_Sequence_1_ContainedGraph {nodes [
   n1 = RigVMFunctionEntryNode()
   RigVMFunction_Sequence = &RigVMFunction_Sequence(ExecuteContext=@n1.ExecuteContext)
   n2 = &RigVMFunction_Sequence(ExecuteContext=@RigVMFunction_Sequence.B)
   n3 = &RigVMFunction_Sequence(ExecuteContext=@n2.B)
   n4 = &RigVMFunction_Sequence(ExecuteContext=@n3.B)
   RigVMFunctionReturnNode(A=@RigVMFunction_Sequence.A, B=@n2.A, C=@n3.A, D=@n4.A, E=@n4.B)
  ]}
 RigVMFunction_Sequence_1_ContainedGraph {nodes [
   n1 = RigVMFunctionEntryNode()
   RigVMFunction_Sequence = &RigVMFunction_Sequence(ExecuteContext=@n1.ExecuteContext)
   n2 = &RigVMFunction_Sequence(ExecuteContext=@RigVMFunction_Sequence.B)
   n3 = &RigVMFunction_Sequence(ExecuteContext=@n2.B)
   RigVMFunctionReturnNode(A=@RigVMFunction_Sequence.A, B=@n2.A, C=@n3.A, D=@n3.B)
  ]}
 RigVMFunction_Sequence_1_ContainedGraph {nodes [
   n1 = RigVMFunctionEntryNode()
   RigVMFunction_Sequence = &RigVMFunction_Sequence(ExecuteContext=@n1.ExecuteContext)
   n2 = &RigVMFunction_Sequence(ExecuteContext=@RigVMFunction_Sequence.B)
   RigVMFunctionReturnNode(A=@RigVMFunction_Sequence.A, B=@n2.A, C=@n2.B)
  ]}
 RigVMFunction_Sequence_1_ContainedGraph {nodes [
   n1 = RigVMFunctionEntryNode()
   RigVMFunction_Sequence = &RigVMFunction_Sequence(ExecuteContext=@n1.ExecuteContext)
   n2 = &RigVMFunction_Sequence(ExecuteContext=@RigVMFunction_Sequence.B)
   RigVMFunctionReturnNode(A=@RigVMFunction_Sequence.A, B=@n2.A, C=@n2.B)
  ]}
 RigVMFunction_Sequence_1_ContainedGraph {nodes [
   n1 = RigVMFunctionEntryNode()
   RigVMFunction_Sequence = &RigVMFunction_Sequence(ExecuteContext=@n1.ExecuteContext)
   n2 = &RigVMFunction_Sequence(ExecuteContext=@RigVMFunction_Sequence.B)
   RigVMFunctionReturnNode(A=@RigVMFunction_Sequence.A, B=@n2.A, C=@n2.B)
  ]}
 RigVMModel {nodes [
   RigUnit_BeginExecution = &RigUnit_BeginExecution()
   Controls = RigVMFunctionReferenceNode(Debug=false, ExecuteContext=@RigUnit_BeginExecution.ExecutePin)
   > FN_DeformRig_FNC = RigVMFunctionReferenceNode(
    DeformRig_ Curve="(deformRig_l_upperarm_up_alpha_72_#1=false,deformRig_r_upperarm_up_alpha_73_#2=false,deformRig_l_upperarm_fr_alpha_74_#3=true,deformRig_r_upperarm_fr_alpha_75_CB51B80 ...(truncated)"
    DeformRig_Bool="(upperarm_1_#4=True,elbow_3_#5=True,wrist_7_#6=True,groin_8_#7=True,glute_17_#8=True,knee_9_40C408324CA261 ...(truncated)"
    DeformRig_Debug=(upperarm_debug_18_#4=False,wrist_debug_20_#6=False,lat_debug_26_#9=False,groinGlute_debug_27_#7=False)
    DeformRig_Value="(upperarm_up_18_#10=(Rotation=(X=0.0,Y=-0.0,Z=0.0,W=1.0),Translation=(X=0.0,Y=0.0,Z=8.0),Scale3D=(X=0.70,Y=1.0,Z=1.0)),upperarm_fwd_fr_19_#11=(Rotation=(X=0.0,Y=-0.0,Z=0.0,W=1.0),Translation=(X=0.0, ...(truncated)"
   )
   > Variables = RigVMFunctionReferenceNode()
   > Buckles = RigVMFunctionReferenceNode()
   > Collar = RigVMFunctionReferenceNode()
   > Lapel = RigVMFunctionReferenceNode()
   > Shirt = RigVMFunctionReferenceNode()
   Multiply_2 = RigVMAggregateNode(
    A=1
    C=&RigVMFunction_MathDoubleRemap(SourceMaximum=0, SourceMinimum=1, TargetMaximum=1, TargetMinimum=0, bClamp=False, Value=RigVMVariableNode() {variable=isEmoting}.Value).Result
    B=RigVMVariableNode() {variable=isLargeWeapon}.Value
   )
   &RigUnit_ModifyTransforms(
    ExecutePin=()
    ItemToModify="((Item=(Type=Bone,Name=\"base_collar_c\"),Transform=(Rotation=(X=0,Y=0,Z=0,W=1),Translation=(X=0,Y=0,Z=0),Scale3D=(X=1,Y=1,Z=1))),(Item=(Type=Bone,Name=\"base_collar_B_l\"),Transform=(Rotation=(X=0,Y=0,Z=0,W=1),Translation=(X=0,Y=0,Z=0),Scale3D=(X=1,Y=1,Z=1))) ...(truncated)"
    WeightMinimum=0
    Weight=@Multiply_2.Result
   )
   Multiply_1 = RigVMAggregateNode(
    A=1
    C=&RigVMFunction_MathDoubleRemap(SourceMaximum=0, SourceMinimum=1, TargetMaximum=1, TargetMinimum=0, bClamp=False, Value=RigVMVariableNode() {variable=isEmoting}.Value).Result
    B=RigVMVariableNode() {variable=isLargeWeapon}.Value
   )
   &RigUnit_ModifyTransforms(
    ExecutePin=()
    ItemToModify="((Item=(Type=Bone,Name=\"base_collar_c\"),Transform=(Rotation=(X=0,Y=0,Z=0,W=1),Translation=(X=0,Y=0,Z=0),Scale3D=(X=1,Y=1,Z=1))),(Item=(Type=Bone,Name=\"base_collar_B_l\"),Transform=(Rotation=(X=0,Y=0,Z=0,W=1),Translation=(X=0,Y=0,Z=0),Scale3D=(X=1,Y=1,Z=1))) ...(truncated)"
    WeightMinimum=0
    Weight=@Multiply_1.Result
   )
   &RigUnit_ModifyTransforms(
    ExecutePin=()
    ItemToModify="((Item=(Type=Bone,Name=\"dyn_collar_A_l_01\"),Transform=(Rotation=(X=0,Y=0,Z=0,W=1),Translation=(X=0,Y=0,Z=0),Scale3D=(X=1,Y=1,Z=1))),(Item=(Type=Bone,Name=\"dyn_collar_B_l_01\"),Transform=(Rotation=(X=0,Y=0,Z=0.3007,W=0.9537),Translation=(X=0,Y=0,Z=0),Scale3D ...(truncated)"
    WeightMinimum=0
    Weight=RigVMVariableNode() {variable=L_Clav_SPR}.Value
   )
   &RigUnit_ModifyTransforms(
    ExecutePin=()
    ItemToModify="((Item=(Type=Bone,Name=\"dyn_collar_A_l_01\"),Transform=(Rotation=(X=0,Y=0,Z=0,W=1),Translation=(X=0,Y=0,Z=0),Scale3D=(X=1,Y=1,Z=1))),(Item=(Type=Bone,Name=\"dyn_collar_B_l_01\"),Transform=(Rotation=(X=0,Y=0,Z=0.3007,W=0.9537),Translation=(X=0,Y=0,Z=0),Scale3D ...(truncated)"
    WeightMinimum=0
    Weight=RigVMVariableNode() {variable=L_Clav_SPR}.Value
   )
  ]}
]
hierarchy [
 root {loc=<0 0 0>, rot=<0 0 0 1>}
 pelvis {parent=root, loc=<0 -0.5301 98.0571>, rot=<0.0127 0.707 -0.0127 -0.707>}
 spine_01 {parent=pelvis, loc=<6.1208 -0.9328 0>, rot=<0 0 0.034 -0.9994>}
 spine_02 {parent=spine_01, loc=<6.5997 -0.2095 0>, rot=<0 0 0.0238 -0.9997>}
 spine_03 {parent=spine_02, loc=<6.4914 -0.1237 0>, rot=<0 0 0.0149 -0.9999>}
 spine_04 {parent=spine_03, loc=<6.9598 0.0794 0>, rot=<0 0 -0.0132 -0.9999>}
 spine_05 {parent=spine_04, loc=<6.9987 0 0>, rot=<0 0 -0.0015 -1>}
 neck_01 {parent=spine_05, loc=<11.3138 0.1529 0>, rot=<0 0 -0.2003 -0.9797>}
 neck_02 {parent=neck_01, loc=<4.5627 0 0>, rot=<0 0 0.0504 -0.9987>}
 head {parent=neck_02, loc=<5.3914 0 0>, rot=<0 0 0.115 -0.9934>}
 base_leftHair_J {parent=head, loc=<0.0665 -4.4521 -4.8703>, rot=<-0.04 -0.5951 0.7932 -0.1232>}
 dyn_leftHair_J {parent=base_leftHair_J, loc=<0 0 0>, rot=<0 0 0 1>}
 leftHair_J_end {parent=dyn_leftHair_J, loc=<2.807 0 0>, rot=<0 0 0 1>}
 base_backHair_K {parent=head, loc=<-0.2282 -3.8002 4.975>, rot=<0.0334 0.1278 0.9775 -0.1645>}
 dyn_backHair_K {parent=base_backHair_K, loc=<0 0 0>, rot=<0 0 0 1>}
 backHair_K_end {parent=dyn_backHair_K, loc=<2.3042 0 0>, rot=<0 0 0 1>}
 base_backHair_J {parent=head, loc=<-0.0568 -5.5087 2.1745>, rot=<0.158 0.1255 0.9646 -0.1696>}
 dyn_backHair_J {parent=base_backHair_J, loc=<0 0 0>, rot=<0 0 0 1>}
 backHair_J_end {parent=dyn_backHair_J, loc=<2.5528 0 0>, rot=<0 0 0 1>}
 base_backHair_F {parent=head, loc=<3.0164 -7.7468 0.6495>, rot=<0.1621 0.0707 0.9596 -0.2189>}
 dyn_backHair_F {parent=base_backHair_F, loc=<0 0 0>, rot=<0 0 0 1>}
 backHair_F_end {parent=dyn_backHair_F, loc=<2.7302 0 0>, rot=<0 0 0 1>}
 base_leftHair_F {parent=head, loc=<2.2954 -2.6147 -6.0567>, rot=<0.1316 -0.9698 0.0259 -0.2035>}
 dyn_leftHair_F {parent=base_leftHair_F, loc=<0 0 0>, rot=<-0.6687 0 0 -0.7435>}
 base_rightHair_A {parent=head, loc=<8.0356 2.3381 6.1295>, rot=<-0.0164 0.6731 0.7256 -0.1421>}
 dyn_rightHair_A_01 {parent=base_rightHair_A, loc=<0 0 0>, rot=<0 0 0 1>}
 dyn_rightHair_A_02 {parent=dyn_rightHair_A_01, loc=<6.0154 -0.3861 0.6407>, rot=<-0.0204 0.2116 0.1803 -0.9604>}
 rightHair_A_01_end {parent=dyn_rightHair_A_02, loc=<6.5038 0 0>, rot=<0 0 0 1>}
 base_leftHair_A {parent=head, loc=<4.8554 4.4969 -6.5347>, rot=<0.0218 -0.7054 0.7046 -0.0741>}
 dyn_leftHair_A_01 {parent=base_leftHair_A, loc=<0 0 0>, rot=<0 0 0 1>}
 dyn_leftHair_A_02 {parent=dyn_leftHair_A_01, loc=<4.5956 0.1008 0.1365>, rot=<-0.0047 -0.1173 0.0732 -0.9904>}
 leftHair_A_end {parent=dyn_leftHair_A_02, loc=<10.6343 0 0>, rot=<0 0 0 1>}
 base_leftHair_I {parent=head, loc=<-0.0734 -2.6476 -5.6768>, rot=<-0.1737 -0.6793 0.7105 0.0598>}
 dyn_leftHair_I {parent=base_leftHair_I, loc=<0 0 0>, rot=<0 0 0 1>}
 leftHair_I_end {parent=dyn_leftHair_I, loc=<3.756 0 0>, rot=<0 0 0 1>}
 clavicle_l {parent=spine_05, loc=<5.4831 -0.1053 -3.2946>, rot=<0.0465 -0.7289 -0.009 -0.6829>}
 upperarm_l {parent=clavicle_l, loc=<9.8658 -1.2136 -0.4105>, rot=<0.0164 -0.3245 -0.0269 -0.9454>}
 lowerarm_l {parent=upperarm_l, loc=<26.465 -0.8345 0.259>, rot=<0.0036 0.0161 -0.2298 -0.9731>}
 hand_l {parent=lowerarm_l, loc=<22.7355 -0.3203 -0.6105>, rot=<-0.6784 -0.0223 0.0735 -0.7307>}
 thumb_01_l {parent=hand_l, loc=<1.3118 -0.3191 -2.459>, rot=<0.6546 -0.4042 0.0664 -0.6354>}
 thumb_02_l {parent=thumb_01_l, loc=<5.9336 -0.0448 0.0401>, rot=<-0.0037 0.0213 0.298 -0.9543>}
 thumb_03_l {parent=thumb_02_l, loc=<2.2223 0.3682 -0.0907>, rot=<0.0118 -0.013 0.0066 -0.9998>}
 index_metacarpal_l {parent=hand_l, loc=<1.6535 0.0183 -2.96>, rot=<0.0158 -0.0124 -0.0173 -0.9996>}
 index_01_l {parent=index_metacarpal_l, loc=<8.3008 0.0357 -0.0524>, rot=<-0.0044 0.0485 0.0924 -0.9945>}
 index_02_l {parent=index_01_l, loc=<3.3587 -0.3004 0.0452>, rot=<-0.002 0.0027 0.3225 -0.9466>}
 index_03_l {parent=index_02_l, loc=<2.5578 -0.123 -0.0252>, rot=<-0.002 0.0067 0.2558 -0.9667>}
 middle_metacarpal_l {parent=hand_l, loc=<2.1856 -0.1725 -0.5984>, rot=<0.006 0.0168 -0.0329 -0.9993>}
 middle_01_l {parent=middle_metacarpal_l, loc=<7.3275 -0.2447 0.0129>, rot=<-0.0208 0.0042 0.0858 -0.9961>}
 middle_02_l {parent=middle_01_l, loc=<4.0123 0.1401 0.1175>, rot=<-0.0218 0.0258 0.3512 -0.9357>}
 middle_03_l {parent=middle_02_l, loc=<2.6077 0.0534 0.1373>, rot=<0.0082 0.0188 0.1708 -0.9851>}
 ring_metacarpal_l {parent=hand_l, loc=<2.2074 -0.3625 1.3638>, rot=<-0.0284 0.0722 -0.0458 -0.9959>}
 ring_01_l {parent=ring_metacarpal_l, loc=<7.0295 -0.6197 -0.1446>, rot=<-0.0489 -0.0252 0.1892 -0.9804>}
 ring_02_l {parent=ring_01_l, loc=<3.6174 0.4647 -0.1027>, rot=<0 0 0.2806 -0.9598>}
 ring_03_l {parent=ring_02_l, loc=<2.4152 0.114 0.0199>, rot=<0 0 0.1189 -0.9929>}
 pinky_metacarpal_l {parent=hand_l, loc=<1.7616 -0.9753 3.3814>, rot=<-0.0397 0.0814 -0.069 -0.9935>}
 pinky_01_l {parent=pinky_metacarpal_l, loc=<7.1709 -0.4831 -0.1094>, rot=<-0.0654 -0.0035 0.2242 -0.9723>}
 pinky_02_l {parent=pinky_01_l, loc=<2.8274 0.0893 -0.1043>, rot=<0.0037 0.0011 0.2854 -0.9584>}
 pinky_03_l {parent=pinky_02_l, loc=<2.2046 0.0307 -0.0078>, rot=<-0.0002 0.0098 0.1445 -0.9895>}
 deform_wrist_dn_base_l {parent=hand_l, loc=<0 -1.5326 0>, rot=<0 0 0 -1>}
 deform_wrist_dn_l {parent=deform_wrist_dn_base_l, loc=<0 0 0>, rot=<0 0 0 1>}
 deform_wrist_up_base_l {parent=hand_l, loc=<0 2.3009 0>, rot=<0 0 0 -1>}
 deform_wrist_up_l {parent=deform_wrist_up_base_l, loc=<0 0 0>, rot=<0 0 0 1>}
 lowerarm_twist_02_l {parent=lowerarm_l, loc=<10.6707 0 0>, rot=<0 0 0 1>}
 lowerarm_twist_01_l {parent=lowerarm_l, loc=<18.316 0 0>, rot=<0 0 0 1>}
 deform_elbow_in_base_l {parent=lowerarm_l, loc=<0.1834 3.6062 -0.2572>, rot=<0 0 0 -1>}
 deform_elbow_in_l {parent=deform_elbow_in_base_l, loc=<0 0 0>, rot=<0 0 0 1>}
 deform_elbow_out_base_l {parent=lowerarm_l, loc=<0 -4 0>, rot=<0 0 0 1>}
 deform_elbow_out_l {parent=deform_elbow_out_base_l, loc=<0 0 0>, rot=<0 0 0 1>}
 hand_l_null {parent=lowerarm_l, type=Null, loc=<22.7355 -0.3203 -0.6105>, rot=<-0.6784 -0.0223 0.0735 -0.7307>}
 upperarm_twist_01_l {parent=upperarm_l, loc=<7.2135 -0.1134 0>, rot=<0 0 0 1>}
 deform_upperarm_bk_base_l {parent=upperarm_twist_01_l, loc=<-6.4053 -4.3746 0.314>, rot=<0 0 0 -1>}
 deform_upperarm_bk_l {parent=deform_upperarm_bk_base_l, loc=<0 0 0>, rot=<0 0 0 1>}
 deform_upperarm_fr_base_l {parent=upperarm_twist_01_l, loc=<-5.6312 5.092 0.3354>, rot=<0 0 0 -1>}
 deform_upperarm_fr_l {parent=deform_upperarm_fr_base_l, loc=<0 0 0>, rot=<0 0 0 1>}
 deform_upperarm_up_base_l {parent=upperarm_twist_01_l, loc=<-8.891 0.4825 3.9517>, rot=<0 0 0 -1>}
 deform_upperarm_up_l {parent=deform_upperarm_up_base_l, loc=<0 0 0>, rot=<0 0 0 1>}
 upperarm_twist_02_l {parent=upperarm_l, loc=<19.7998 -0.1134 0>, rot=<0 0 0 1>}
 upperarm_l_null {parent=clavicle_l, type=Null, loc=<9.8658 -1.2136 -0.4105>, rot=<0.0164 -0.3245 -0.0269 -0.9454>}
 clavicle_r {parent=spine_05, loc=<5.4831 -0.1053 3.2946>, rot=<0.7289 0.0465 -0.6829 0.009>}
 upperarm_r {parent=clavicle_r, loc=<-9.8658 1.2136 0.4105>, rot=<0.0164 -0.3245 -0.0269 -0.9454>}
 lowerarm_r {parent=upperarm_r, loc=<-26.465 0.8345 -0.259>, rot=<0.0036 0.0161 -0.2298 -0.9731>}
 hand_r {parent=lowerarm_r, loc=<-22.7355 0.3203 0.6106>, rot=<-0.6784 -0.0223 0.0735 -0.7307>}
 thumb_01_r {parent=hand_r, loc=<-1.3118 0.3191 2.459>, rot=<0.6546 -0.4042 0.0664 -0.6354>}
 thumb_02_r {parent=thumb_01_r, loc=<-5.9336 0.0448 -0.0401>, rot=<-0.0037 0.0213 0.298 -0.9543>}
 thumb_03_r {parent=thumb_02_r, loc=<-2.2223 -0.3682 0.0907>, rot=<0.0118 -0.013 0.0066 -0.9998>}
 index_metacarpal_r {parent=hand_r, loc=<-1.6535 -0.0183 2.96>, rot=<0.0158 -0.0124 -0.0173 -0.9996>}
 index_01_r {parent=index_metacarpal_r, loc=<-8.3008 -0.0357 0.0524>, rot=<-0.0044 0.0485 0.0924 -0.9945>}
 index_02_r {parent=index_01_r, loc=<-3.3587 0.3004 -0.0452>, rot=<-0.002 0.0027 0.3225 -0.9466>}
 index_03_r {parent=index_02_r, loc=<-2.5578 0.123 0.0252>, rot=<-0.002 0.0067 0.2558 -0.9667>}
 middle_metacarpal_r {parent=hand_r, loc=<-2.1856 0.1725 0.5984>, rot=<0.006 0.0168 -0.0329 -0.9993>}
 middle_01_r {parent=middle_metacarpal_r, loc=<-7.3275 0.2447 -0.0129>, rot=<-0.0208 0.0042 0.0858 -0.9961>}
 middle_02_r {parent=middle_01_r, loc=<-4.0123 -0.1401 -0.1175>, rot=<-0.0218 0.0258 0.3512 -0.9357>}
 middle_03_r {parent=middle_02_r, loc=<-2.6077 -0.0534 -0.1373>, rot=<0.0082 0.0188 0.1708 -0.9851>}
 ring_metacarpal_r {parent=hand_r, loc=<-2.2074 0.3625 -1.3638>, rot=<-0.0284 0.0722 -0.0458 -0.9959>}
 ring_01_r {parent=ring_metacarpal_r, loc=<-7.0295 0.6197 0.1446>, rot=<-0.0489 -0.0252 0.1892 -0.9804>}
 ring_02_r {parent=ring_01_r, loc=<-3.6174 -0.4647 0.1027>, rot=<0 0 0.2806 -0.9598>}
 ring_03_r {parent=ring_02_r, loc=<-2.4152 -0.114 -0.0199>, rot=<0 0 0.1189 -0.9929>}
 pinky_metacarpal_r {parent=hand_r, loc=<-1.7616 0.9753 -3.3814>, rot=<-0.0397 0.0814 -0.069 -0.9935>}
 pinky_01_r {parent=pinky_metacarpal_r, loc=<-7.1709 0.4831 0.1094>, rot=<-0.0654 -0.0035 0.2242 -0.9723>}
 pinky_02_r {parent=pinky_01_r, loc=<-2.8274 -0.0893 0.1043>, rot=<0.0037 0.0011 0.2854 -0.9584>}
 pinky_03_r {parent=pinky_02_r, loc=<-2.2046 -0.0307 0.0078>, rot=<-0.0002 0.0098 0.1445 -0.9895>}
 deform_wrist_dn_base_r {parent=hand_r, loc=<-0.0001 1.5328 0>, rot=<0 0 0 -1>}
 deform_wrist_dn_r {parent=deform_wrist_dn_base_r, loc=<0 0 0>, rot=<0 0 0 1>}
 deform_wrist_up_base_r {parent=hand_r, loc=<-0.0002 -2.3007 -0.0001>, rot=<0 0 0 -1>}
 deform_wrist_up_r {parent=deform_wrist_up_base_r, loc=<0 0 0>, rot=<0 0 0 1>}
 lowerarm_twist_02_r {parent=lowerarm_r, loc=<-10.6707 0 0>, rot=<0 0 0 -1>}
 lowerarm_twist_01_r {parent=lowerarm_r, loc=<-18.316 0 0>, rot=<0 0 0 -1>}
 deform_elbow_in_base_r {parent=lowerarm_r, loc=<-0.1834 -3.6062 0.2572>, rot=<0 0 0 -1>}
 deform_elbow_in_r {parent=deform_elbow_in_base_r, loc=<0 0 0>, rot=<0 0 0 1>}
 deform_elbow_out_base_r {parent=lowerarm_r, loc=<0 4 -0.0001>, rot=<0 0 0 -1>}
 deform_elbow_out_r {parent=deform_elbow_out_base_r, loc=<0 0 0>, rot=<0 0 0 1>}
 hand_r_null {parent=lowerarm_r, type=Null, loc=<-22.7355 0.3203 0.6106>, rot=<-0.6784 -0.0223 0.0735 -0.7307>}
 upperarm_twist_01_r {parent=upperarm_r, loc=<-7.2135 0.1134 0>, rot=<0 0 0 -1>}
 deform_upperarm_bk_base_r {parent=upperarm_twist_01_r, loc=<6.4053 4.3746 -0.314>, rot=<0 0 0 -1>}
 deform_upperarm_bk_r {parent=deform_upperarm_bk_base_r, loc=<0 0 0>, rot=<0 0 0 1>}
 deform_upperarm_fr_base_r {parent=upperarm_twist_01_r, loc=<5.6312 -5.092 -0.3354>, rot=<0 0 0 -1>}
 deform_upperarm_fr_r {parent=deform_upperarm_fr_base_r, loc=<0 0 0>, rot=<0 0 0 1>}
 deform_upperarm_up_base_r {parent=upperarm_twist_01_r, loc=<8.891 -0.4825 -3.9517>, rot=<0 0 0 -1>}
 deform_upperarm_up_r {parent=deform_upperarm_up_base_r, loc=<0 0 0>, rot=<0 0 0 1>}
 upperarm_twist_02_r {parent=upperarm_r, loc=<-19.7998 0.1134 0>, rot=<0 0 0 -1>}
 upperarm_r_null {parent=clavicle_r, type=Null, loc=<-9.8658 1.2136 0.4105>, rot=<0.0164 -0.3245 -0.0269 -0.9454>}
 attach_backpack {parent=spine_05, loc=<-0.123 -13.473 0>, rot=<-0.0093 -0.707 0.0093 -0.707>}
 Backpack_BR {parent=attach_backpack, type=Null, loc=<0 3.6944 0>, rot=<0 0 0 1>}
 attach_cape {parent=spine_05, loc=<9.017 -11.714 0>, rot=<0.0616 -0.7044 -0.0616 -0.7044>}
 cape {parent=attach_cape, type=Null, loc=<0 0 0>, rot=<0 0 0 1>}
 base_loop {parent=spine_05, loc=<2.9685 9.0091 0>, rot=<0 0 0.9385 0.3453>}
 dyn_loop_01 {parent=base_loop, loc=<0 0 0>, rot=<0 0 0 1>}
 dyn_loop_02 {parent=dyn_loop_01, loc=<3.6609 0 0>, rot=<0 0 0 1>}
 base_lapel_A_l {parent=spine_05, loc=<7.4004 5.5133 -7.2216>, rot=<-0.6326 -0.4269 -0.09 -0.6398>}
 dyn_lapel_A_l_01 {parent=base_lapel_A_l, loc=<0 0 0>, rot=<0 0 0 1>}
 dyn_lapel_A_l_02 {parent=dyn_lapel_A_l_01, loc=<3 0 0>, rot=<0 0 0.2172 -0.9761>}
 dyn_lapel_A_l_03 {parent=dyn_lapel_A_l_02, loc=<2.437 0 0>, rot=<0 0 0 1>}
 base_lapel_B_l {parent=spine_05, loc=<10.8713 0.9328 -6.9969>, rot=<-0.5513 -0.3044 -0.0194 -0.7766>}
 dyn_lapel_B_l_01 {parent=base_lapel_B_l, loc=<0 0 0>, rot=<0 0 -0.1736 -0.9848>}
 dyn_lapel_B_l_02 {parent=dyn_lapel_B_l_01, loc=<2.81 0 0>, rot=<0 0 0.254 -0.9672>}
 dyn_lapel_B_l_03 {parent=dyn_lapel_B_l_02, loc=<1.8271 0 0>, rot=<0 0 0 1>}
 base_lapelBuckle_l {parent=dyn_lapel_B_l_01, loc=<1.4856 -0.2166 -1.2493>, rot=<0.0354 -0.7363 0.0162 -0.6756>}
 dyn_lapelBuckle_l {parent=base_lapelBuckle_l, loc=<0 0 0>, rot=<0 0 0 1>}
 dyn_lapelBuckle_l_end {parent=dyn_lapelBuckle_l, loc=<3.991 0 0>, rot=<0 0 0 1>}
 base_lapel_C_l {parent=spine_05, loc=<12.1658 -3.5236 -4.5556>, rot=<-0.3369 -0.1351 0.0022 -0.9318>}
 dyn_lapel_C_l_01 {parent=base_lapel_C_l, loc=<0 0 0>, rot=<0 0 0 1>}
 dyn_lapel_C_l_02 {parent=dyn_lapel_C_l_01, loc=<2.32 0 0>, rot=<0 0 0.254 -0.9672>}
 dyn_lapel_C_l_03 {parent=dyn_lapel_C_l_02, loc=<2.2608 0 0>, rot=<0 0 0 1>}
 base_lapel_A_r {parent=spine_05, loc=<7.1453 5.8136 7.1312>, rot=<0.6107 0.5095 0.0081 -0.6061>}
 dyn_lapel_A_r_01 {parent=base_lapel_A_r, loc=<0 0 0>, rot=<0 0 0 1>}
 dyn_lapel_A_r_02 {parent=dyn_lapel_A_r_01, loc=<3.0779 0 0>, rot=<-0.0755 -0.0285 0.3275 -0.9414>}
 dyn_lapel_A_r_03 {parent=dyn_lapel_A_r_02, loc=<2.4369 0 0>, rot=<0 0 0 1>}
 base_lapel_B_r {parent=spine_05, loc=<10.8712 0.9328 6.9969>, rot=<0.658 0.304 0.025 -0.6885>}
 dyn_lapel_B_r_01 {parent=base_lapel_B_r, loc=<0 0 0>, rot=<0 0 0 1>}
 dyn_lapel_B_r_02 {parent=dyn_lapel_B_r_01, loc=<2.8098 0 0>, rot=<-0.1731 -0.0281 0.2525 -0.9516>}
 dyn_lapel_B_r_03 {parent=dyn_lapel_B_r_02, loc=<1.827 0 0>, rot=<0 0 0 1>}
 base_lapelBuckle_r {parent=dyn_lapel_B_r_01, loc=<1.5896 -0.7102 1.1093>, rot=<-0.1061 0.732 0.1821 -0.6479>}
 dyn_lapelBuckle_r {parent=base_lapelBuckle_r, loc=<0 0 0>, rot=<0 0 0 1>}
 dyn_lapelBuckle_r_end {parent=dyn_lapelBuckle_r, loc=<3.9912 0 0>, rot=<0 0 0 1>}
 base_lapel_C_r {parent=spine_05, loc=<12.1655 -3.5237 4.5556>, rot=<0.3916 0.1347 0.0102 -0.9102>}
 dyn_lapel_C_r_01 {parent=base_lapel_C_r, loc=<0 0 0>, rot=<0 0 0 1>}
 dyn_lapel_C_r_02 {parent=dyn_lapel_C_r_01, loc=<2.3207 0 0>, rot=<-0.2809 -0.107 0.2305 -0.9255>}
 dyn_lapel_C_r_03 {parent=dyn_lapel_C_r_02, loc=<2.2603 0 0>, rot=<0 0 0 1>}
 base_lapel_c {parent=spine_05, loc=<11.9617 -5.6297 -0.1136>, rot=<0 0 -0.0401 -0.9992>}
 dyn_lapel_c_01 {parent=base_lapel_c, loc=<0 0 0>, rot=<0 0 0 1>}
 dyn_lapel_c_02 {parent=dyn_lapel_c_01, loc=<2.4446 0 0>, rot=<0 0 0.3257 -0.9455>}
 dyn_lapel_c_03 {parent=dyn_lapel_c_02, loc=<2.4463 0 0>, rot=<0 0 0 1>}
 base_collar_A_l {parent=spine_05, loc=<6.0521 5.5006 -2.7475>, rot=<-0.6365 -0.3584 -0.0408 -0.6818>}
 dyn_collar_A_l_01 {parent=base_collar_A_l, loc=<0 0 0>, rot=<0 0 0 1>}
 dyn_collar_A_l_02 {parent=dyn_collar_A_l_01, loc=<6.9254 0 0>, rot=<0 0 0 1>}
 base_collar_B_l {parent=spine_05, loc=<10.3733 0.3543 -4.3897>, rot=<-0.6635 -0.2352 -0.041 -0.7091>}
 dyn_collar_B_l_01 {parent=base_collar_B_l, loc=<0 -0.6315 0>, rot=<0 0 0 1>}
 dyn_collar_B_l_02 {parent=dyn_collar_B_l_01, loc=<5.6596 0 0>, rot=<0 0 0 1>}
 base_collar_c {parent=spine_05, loc=<12.8638 -3.1067 0>, rot=<0 0 0.0014 -1>}
 dyn_collar_c_01 {parent=base_collar_c, loc=<0 0 0>, rot=<0 0 0 1>}
 dyn_collar_c_02 {parent=dyn_collar_c_01, loc=<5.4533 0 0>, rot=<0 0 0 1>}
 base_collar_B_r {parent=spine_05, loc=<10.373 0.3542 4.3897>, rot=<0.5653 0.2278 -0.0713 -0.7896>}
 dyn_collar_B_r_01 {parent=base_collar_B_r, loc=<0 0 0>, rot=<0 0 0 1>}
 dyn_collar_B_r_02 {parent=dyn_collar_B_r_01, loc=<5.6602 0 0>, rot=<0 0 0 1>}
 base_collar_A_r {parent=spine_05, loc=<5.7884 6.6358 2.3641>, rot=<0.557 0.3668 0.0626 -0.7424>}
 dyn_collar_A_r_01 {parent=base_collar_A_r, loc=<0 0 0>, rot=<0 0 0 1>}
 dyn_collar_A_r_02 {parent=dyn_collar_A_r_01, loc=<6.9256 0 0>, rot=<-0.0441 0.0682 -0.0019 -0.9967>}
 head_null {parent=spine_05, type=Null, loc=<20.6573 3.5489 0>, rot=<0 0 -0.036 -0.9994>}
 clavicle_l_null {parent=spine_05, type=Null, loc=<-0.8685 -0.335 -6.1823>, rot=<0.4566 0.3027 0.3141 0.7754>}
 base_jacket_l {parent=spine_04, loc=<0.1255 5.0995 -11.1659>, rot=<0.0281 0.6362 0.7709 -0.0136>}
 dyn_jacket_l_01 {parent=base_jacket_l, loc=<0 0 0>, rot=<0 0 0 1>}
 dyn_jacket_l_02 {parent=dyn_jacket_l_01, loc=<4.3301 0 0>, rot=<0 0 0 1>}
 dyn_jacket_l_03 {parent=dyn_jacket_l_02, loc=<4.3301 0 0>, rot=<0 0 0 1>}
 base_jacket_r {parent=spine_04, loc=<0.1251 5.0995 11.1659>, rot=<-0.0301 -0.4868 0.8729 -0.0083>}
 dyn_jacket_r_01 {parent=base_jacket_r, loc=<0 0 0>, rot=<0 0 0 1>}
 dyn_jacket_r_02 {parent=dyn_jacket_r_01, loc=<4.3297 0 0>, rot=<0.0001 0 0 -1>}
 dyn_jacket_r_03 {parent=dyn_jacket_r_02, loc=<4.3297 0 0>, rot=<0 0 0 1>}
 base_tail_l {parent=spine_04, loc=<0.041 -4.19 -9.27>, rot=<0.1461 -0.9178 -0.3664 -0.0462>}
 dyn_tail_l_01 {parent=base_tail_l, loc=<0 0 0>, rot=<0 0 0 1>}
 dyn_tail_l_02 {parent=dyn_tail_l_01, loc=<5.7468 0 0>, rot=<0 0 0 1>}
 dyn_tail_l_03 {parent=dyn_tail_l_02, loc=<5.7468 0 0>, rot=<0 0 0 1>}
 base_tail_r {parent=spine_04, loc=<0.0413 -4.1902 9.2698>, rot=<0.1491 -0.8895 0.4305 0.0357>}
 dyn_tail_r_01 {parent=base_tail_r, loc=<0 0 0>, rot=<0 0 0 -1>}
 dyn_tail_r_02 {parent=dyn_tail_r_01, loc=<5.746 0 0>, rot=<0 0 0 -1>}
 dyn_tail_r_03 {parent=dyn_tail_r_02, loc=<5.747 0 0>, rot=<0 0 0 1>}
 base_tail_c {parent=spine_04, loc=<1.3366 -9.0925 -0.449>, rot=<0.0262 -0.9995 0.0115 0.0126>}
 dyn_tail_c_01 {parent=base_tail_c, loc=<0 0 0>, rot=<0 0 0 1>}
 dyn_tail_c_02 {parent=dyn_tail_c_01, loc=<4.7926 0 0>, rot=<0 0 0 1>}
 dyn_tail_c_03 {parent=dyn_tail_c_02, loc=<3.7049 0 0>, rot=<0 0 0 1>}
 deform_pec_aim_base_l {parent=spine_04, loc=<3.2456 9.3309 -1.2237>, rot=<0.0603 -0.4033 0.3318 -0.8507>}
 deform_pec_aim_l {parent=deform_pec_aim_base_l, loc=<0 0 0>, rot=<0 0 0 1>}
 deform_pec_base_l {parent=deform_pec_aim_l, loc=<9.3665 0 0>, rot=<0 0 0 1>}
 deform_pec_l {parent=deform_pec_base_l, loc=<0 0 0>, rot=<0 0 0 1>}
 deform_pec_aim_base_r {parent=spine_04, loc=<3.2452 9.3309 1.2237>, rot=<-0.4033 -0.0603 0.8507 0.3318>}
 deform_pec_aim_r {parent=deform_pec_aim_base_r, loc=<0 0 0>, rot=<0 0 0 1>}
 deform_pec_base_r {parent=deform_pec_aim_r, loc=<-9.3667 -0.0003 -0.0004>, rot=<0 0 0 1>}
 deform_pec_r {parent=deform_pec_base_r, loc=<0 0 0>, rot=<0 0 0 1>}
 base_buckle_up_l {parent=spine_03, loc=<11.5543 8.1448 -11.11>, rot=<0.2242 0.363 0.8897 0.1626>}
 dyn_buckle_up_l_01 {parent=base_buckle_up_l, loc=<0 0 0>, rot=<0 0 0 1>}
 dyn_buckle_up_l_02 {parent=dyn_buckle_up_l_01, loc=<3.946 0 0>, rot=<0 0 -0.0808 -0.9967>}
 dyn_buckle_up_l_03 {parent=dyn_buckle_up_l_02, loc=<4.8632 -0.0695 0>, rot=<0 0 0 1>}
 base_buckle_dw_l {parent=spine_03, loc=<9.7751 4.4798 -11.796>, rot=<0.2006 0.5487 0.8006 0.133>}
 dyn_buckle_dw_l_01 {parent=base_buckle_dw_l, loc=<0 0 0>, rot=<0 0 0 1>}
 dyn_buckle_dw_l_02 {parent=dyn_buckle_dw_l_01, loc=<6.9 0 0>, rot=<0 0 -0.0497 -0.9988>}
 dyn_buckle_dw_l_03 {parent=dyn_buckle_dw_l_02, loc=<5.7386 -0.1299 0>, rot=<0 0 0 1>}
 base_buckle_up_r {parent=spine_03, loc=<11.541 8.0879 11.0478>, rot=<-0.363 0.2242 0.1626 -0.8897>}
 base_buckle_up_00_r {parent=base_buckle_up_r, loc=<0 0 0>, rot=<0 -0.1666 0.986 0>}
 dyn_buckle_up_r_01 {parent=base_buckle_up_00_r, loc=<0 0 0>, rot=<0 0 0 1>}
 dyn_buckle_up_r_02 {parent=dyn_buckle_up_r_01, loc=<3.9463 0 0>, rot=<-0.1062 0.0182 -0.0787 -0.9911>}
 dyn_buckle_up_r_03 {parent=dyn_buckle_up_r_02, loc=<4.4372 0 0>, rot=<0 0 0 1>}
 base_buckle_dw_r {parent=spine_03, loc=<9.6575 4.5513 11.8091>, rot=<-0.5487 0.2006 0.133 -0.8006>}
 base_buckle_dw_00_r {parent=base_buckle_dw_r, loc=<0 0 0>, rot=<0 -0.0293 0.9996 0>}
 dyn_buckle_dw_r_01 {parent=base_buckle_dw_00_r, loc=<0 0 0>, rot=<0 0 0 1>}
 dyn_buckle_dw_r_02 {parent=dyn_buckle_dw_r_01, loc=<6.8995 0 0>, rot=<-0.0743 -0.0007 -0.0497 -0.996>}
 dyn_buckle_dw_r_03 {parent=dyn_buckle_dw_r_02, loc=<4.4383 0 0>, rot=<0 0 0 1>}
 thigh_l {parent=pelvis, loc=<-1.0855 -0.2109 -8.7053>, rot=<0.0009 0.0355 0.0155 -0.9993>}
 calf_l {parent=thigh_l, loc=<-42.5284 -0.0019 -0.7297>, rot=<0.0006 -0.0161 -0.0391 -0.9991>}
 foot_l {parent=calf_l, loc=<-46.7401 -2.5014 -0.592>, rot=<0.0427 -0.0238 0.0353 -0.9982>}
 ball_l {parent=foot_l, loc=<-3.5658 10.8066 -0.1544>, rot=<0.0001 -0.0001 -0.7063 -0.7079>}
 calf_twist_02_l {parent=calf_l, loc=<-19.1546 -1.7943 -0.2638>, rot=<0.0237 0.0095 -0.0035 -0.9997>}
 calf_twist_01_l {parent=calf_l, loc=<-35.8483 -1.9197 -0.5794>, rot=<0.0237 0.0095 -0.0035 -0.9997>}
 deform_knee_in_base_l {parent=calf_l, loc=<0.8031 -5.0231 0.2456>, rot=<0 0 0 -1>}
 deform_knee_in_l {parent=deform_knee_in_base_l, loc=<0 0 0>, rot=<0 0 0 1>}
 deform_knee_out_base_l {parent=calf_l, loc=<0 4 0>, rot=<0 0 0 -1>}
 deform_knee_out_l {parent=deform_knee_out_base_l, loc=<0 0 0>, rot=<0 0 0 1>}
 foot_l_null {parent=calf_l, type=Null, loc=<-46.7401 -2.5014 -0.592>, rot=<0.0427 -0.0238 0.0353 -0.9982>}
 thigh_twist_01_l {parent=thigh_l, loc=<-27.952 0 0>, rot=<0.0013 0 0 -1>}
 deform_groin_base_l {parent=thigh_l, loc=<0.4801 5.6776 -0.2486>, rot=<0 0 0 -1>}
 deform_groin_l {parent=deform_groin_base_l, loc=<0 0 0>, rot=<0 0 0 1>}
 thigh_r {parent=pelvis, loc=<-1.0855 -0.2109 8.7053>, rot=<0.0355 -0.0009 0.9993 0.0155>}
 calf_r {parent=thigh_r, loc=<42.5284 0.0019 0.7297>, rot=<0.0006 -0.0161 -0.0391 -0.9991>}
 foot_r {parent=calf_r, loc=<46.7401 2.5014 0.592>, rot=<0.0427 -0.0238 0.0353 -0.9982>}
 ball_r {parent=foot_r, loc=<3.5658 -10.8066 0.1544>, rot=<0.0001 -0.0001 -0.7063 -0.7079>}
 calf_twist_02_r {parent=calf_r, loc=<19.1546 1.7943 0.2638>, rot=<0.0237 0.0095 -0.0035 -0.9997>}
 calf_twist_01_r {parent=calf_r, loc=<35.8483 1.9197 0.5794>, rot=<0.0237 0.0095 -0.0035 -0.9997>}
 deform_knee_in_base_r {parent=calf_r, loc=<-0.8031 5.0231 -0.2457>, rot=<0 0 0 -1>}
 deform_knee_in_r {parent=deform_knee_in_base_r, loc=<0 0 0>, rot=<0 0 0 1>}
 deform_knee_out_base_r {parent=calf_r, loc=<0 -4 0>, rot=<0 0 0 -1>}
 deform_knee_out_r {parent=deform_knee_out_base_r, loc=<0 0 0>, rot=<0 0 0 1>}
 foot_r_null {parent=calf_r, type=Null, loc=<46.7401 2.5014 0.592>, rot=<0.0427 -0.0238 0.0353 -0.9982>}
 thigh_twist_01_r {parent=thigh_r, loc=<27.952 0 0>, rot=<0.0013 0 0 -1>}
 deform_groin_base_r {parent=thigh_r, loc=<-0.4801 -5.6776 0.2486>, rot=<0 0 0 -1>}
 deform_groin_r {parent=deform_groin_base_r, loc=<0 0 0>, rot=<0 0 0 1>}
 base_shirt_l {parent=pelvis, loc=<8.5934 6.6574 -7.7341>, rot=<-0.0385 0.0019 0.9968 0.0705>}
 dyn_shirt_l_01 {parent=base_shirt_l, loc=<0 0 0>, rot=<0 0 0 1>}
 dyn_shirt_l_02 {parent=dyn_shirt_l_01, loc=<4.7183 0 0>, rot=<0 0 -0.0356 -0.9994>}
 dyn_shirt_l_03 {parent=dyn_shirt_l_02, loc=<4.718 0 0>, rot=<0 0 0 1>}
 base_shirt_r {parent=pelvis, loc=<8.5573 7.1726 6.0902>, rot=<0.0144 -0.0064 0.9974 0.0707>}
 dyn_shirt_r_01 {parent=base_shirt_r, loc=<0 0 0>, rot=<0 0 0 1>}
 dyn_shirt_r_02 {parent=dyn_shirt_r_01, loc=<4.7179 0 0>, rot=<0.0976 -0.0064 -0.035 -0.9946>}
 dyn_shirt_r_03 {parent=dyn_shirt_r_02, loc=<4.718 0 0>, rot=<0 0 0 1>}
 deform_glute_base_l {parent=pelvis, loc=<-7.3466 -9.0213 -9.1245>, rot=<0.0172 0.031 -0.4477 -0.8935>}
 deform_glute_l {parent=deform_glute_base_l, loc=<0 0 0>, rot=<0 0 0 1>}
 deform_glute_base_r {parent=pelvis, loc=<-7.3464 -9.0213 9.1245>, rot=<0.031 -0.0172 0.8935 -0.4477>}
 deform_glute_r {parent=deform_glute_base_r, loc=<0 0 0>, rot=<0 0 0 1>}
 thigh_r_null {parent=pelvis, type=Null, loc=<-1.0855 -0.2109 8.7053>, rot=<0.0355 -0.0009 0.9993 0.0155>}
 thigh_l_null {parent=pelvis, type=Null, loc=<-1.0855 -0.2109 -8.7053>, rot=<0.0009 0.0355 0.0155 -0.9993>}
 base_shirtStrap_l {parent=pelvis, loc=<9.8085 1.7486 -12.2645>, rot=<-0.1066 0.6392 0.7495 0.135>}
 dyn_shirtStrap_l_01 {parent=base_shirtStrap_l, loc=<0 0 0>, rot=<0 0 0 1>}
 dyn_shirtStrap_l_end {parent=dyn_shirtStrap_l_01, loc=<8.5295 0.2717 0>, rot=<0 0 -0.0356 -0.9994>}
 base_shirtStrap_r {parent=pelvis, loc=<9.8058 1.7064 12.2353>, rot=<0.0884 -0.6445 0.7501 0.1194>}
 dyn_shirtStrap_r_01 {parent=base_shirtStrap_r, loc=<0 0 0>, rot=<0 0 0 1>}
 dyn_shirtStrap_r_end {parent=dyn_shirtStrap_r_01, loc=<8.5292 0.2612 -0.075>, rot=<0.0976 -0.0064 -0.035 -0.9946>}
 ik_hand_root {parent=root, loc=<0 0 0>, rot=<0 0 0 1>}
 ik_hand_gun {parent=ik_hand_root, loc=<-48.1743 3.3545 104.4381>, rot=<0.6276 0.1309 0.4183 -0.6434>}
 pelvis_ctrl {controlType=EULER_TRANSFORM, type=Control, loc=<0 0 0>, rot=<0 0 0 1>}
 spine_01_ctrl {controlType=EULER_TRANSFORM, parent=pelvis_ctrl, type=Control, loc=<0 0 0>, rot=<0 0 0 1>}
 spine_02_ctrl {controlType=EULER_TRANSFORM, parent=spine_01_ctrl, type=Control, loc=<0 0 0>, rot=<0 0 0 1>}
 spine_03_ctrl {controlType=EULER_TRANSFORM, parent=spine_02_ctrl, type=Control, loc=<0 0 0>, rot=<0 0 0 1>}
 spine_04_ctrl {controlType=EULER_TRANSFORM, parent=spine_03_ctrl, type=Control, loc=<0 0 0>, rot=<0 0 0 1>}
 spine_05_ctrl {controlType=EULER_TRANSFORM, parent=spine_04_ctrl, type=Control, loc=<0 0 0>, rot=<0 0 0 1>}
 neck_01_ctrl {controlType=EULER_TRANSFORM, parent=spine_05_ctrl, type=Control, loc=<0 0 0>, rot=<0 0 0 1>}
 head_ctrl {controlType=EULER_TRANSFORM, parent=neck_01_ctrl, type=Control, loc=<0 0 0>, rot=<0 0 0 1>}
 clavicle_l_ctrl {controlType=EULER_TRANSFORM, parent=spine_05_ctrl, type=Control, loc=<0 0 0>, rot=<0 0 0 1>}
 upperarm_l_ctrl {controlType=EULER_TRANSFORM, parent=clavicle_l_ctrl, type=Control, loc=<0 0 0>, rot=<0 0 0 1>}
 lowerarm_l_ctrl {controlType=EULER_TRANSFORM, parent=upperarm_l_ctrl, type=Control, loc=<0 0 0>, rot=<0 0 0 1>}
 clavicle_r_ctrl {controlType=EULER_TRANSFORM, parent=spine_05_ctrl, type=Control, loc=<0 0 0>, rot=<0 0 0 1>}
 upperarm_r_ctrl {controlType=EULER_TRANSFORM, parent=clavicle_r_ctrl, type=Control, loc=<0 0 0>, rot=<0 0 0 1>}
 lowerarm_r_ctrl {controlType=EULER_TRANSFORM, parent=upperarm_r_ctrl, type=Control, loc=<0 0 0>, rot=<0 0 0 1>}
 thigh_l_ctrl {controlType=EULER_TRANSFORM, parent=pelvis_ctrl, type=Control, loc=<0 0 0>, rot=<0 0 0 1>}
 calf_l_ctrl {controlType=EULER_TRANSFORM, parent=thigh_l_ctrl, type=Control, loc=<0 0 0>, rot=<0 0 0 1>}
 foot_l_ctrl {controlType=EULER_TRANSFORM, parent=calf_l_ctrl, type=Control, loc=<0 0 0>, rot=<0 0 0 1>}
 thigh_r_ctrl {controlType=EULER_TRANSFORM, parent=pelvis_ctrl, type=Control, loc=<0 0 0>, rot=<0 0 0 1>}
 calf_r_ctrl {controlType=EULER_TRANSFORM, parent=thigh_r_ctrl, type=Control, loc=<0 0 0>, rot=<0 0 0 1>}
 foot_r_ctrl {controlType=EULER_TRANSFORM, parent=calf_r_ctrl, type=Control, loc=<0 0 0>, rot=<0 0 0 1>}
]
variables [
 L_Arm_Up_SPR {type=double, default=0}
 L_Clav_SPR {type=double, default=0}
 Large_ADS_down {type=double, default=0}
 Large_ADS_up {type=double, default=0}
 Pistol ADS up {type=double, default=0}
 R_Arm_Front_SPR {type=double, default=0}
 R_Arm_Up_SPR {type=double, default=0}
 R_Clav_SPR {type=double, default=0}
 isDBNO {type=double, default=0}
 isEmoting {type=double, default=0}
 isLargeWeapon {type=double, default=0}
 isNotSkydiving {type=double, default=0}
 isSkydiving {type=double, default=0}
]
