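# Physics asset for the female medium TwinGinger cosmetic body, covering left and right jacket chains and left, center and right tail chains.
# Each three-bone dyn chain hangs from a small base body tied to spine_04 by a free constraint.
# Chain links use narrow swing limits and zero twist, except each tail's first link, which is fully locked to its base.
# Capsules on the pelvis and spine_01 to 04 form the torso, linked by free constraints.
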PhysicsAsset /BRCosmetics/Characters/Player/Female/Medium/Bodies/F_MED_TwinGinger/Meshes/F_MED_TwinGinger_Physics
bodies {
 base_jacket_l [{shapeType=Capsule, length=1, radius=1, shapeName=base_jacket_l_capsule}]
 base_jacket_r [{shapeType=Capsule, length=1, radius=1, shapeName=base_jacket_r_capsule}]
 base_tail_c [{shapeType=Capsule, length=1, radius=1, shapeName=base_tail_c_capsule}]
 base_tail_l [{shapeType=Capsule, length=1, radius=1, shapeName=base_tail_l_capsule}]
 base_tail_r [{shapeType=Capsule, length=1, radius=1, shapeName=base_tail_r_capsule}]
 dyn_jacket_l_01 [{shapeType=Capsule, length=9.1761, radius=1.6194, shapeName=dyn_jacket_l_01_capsule, center=<2.4762 -0.3006 0>}]
 dyn_jacket_l_02 [{shapeType=Capsule, length=8.3666, radius=1.6194, shapeName=dyn_jacket_l_02_capsule, center=<1.6253 0 -0.2699>}]
 dyn_jacket_l_03 [{shapeType=Capsule, length=3.5788, radius=0.7207, shapeName=dyn_jacket_l_03_capsule, center=<1.9773 0 0>, rotation=<rot 90 0 0>}]
 dyn_jacket_r_01 [{shapeType=Capsule, length=9.1761, radius=1.6194, shapeName=dyn_jacket_r_01_capsule, center=<2.4762 0.3006 0>}]
 dyn_jacket_r_02 [{shapeType=Capsule, length=8.3666, radius=1.6194, shapeName=dyn_jacket_r_02_capsule, center=<1.6253 0 0.2699>}]
 dyn_jacket_r_03 [{shapeType=Capsule, length=3.5788, radius=0.7207, shapeName=dyn_jacket_r_03_capsule, center=<1.9773 0 0>, rotation=<rot 90 0 0>}]
 dyn_tail_c_01 [{shapeType=Capsule, length=13.1041, radius=1.4344, shapeName=dyn_tail_c_01_capsule, center=<2.6126 0 0>}]
 dyn_tail_c_02 [{shapeType=Capsule, length=6.9991, radius=1.6044, shapeName=dyn_tail_c_02_capsule, center=<1.7611 0 0>}]
 dyn_tail_c_03 [{shapeType=Capsule, length=3.5069, radius=0.9232, shapeName=dyn_tail_c_03_capsule, center=<2.483 0 0>, rotation=<rot 90 0 0>}]
 dyn_tail_l_01 [{shapeType=Capsule, length=5.2456, radius=2.016, shapeName=dyn_tail_l_01_capsule, center=<3.099 0 0>}]
 dyn_tail_l_02 [{shapeType=Capsule, length=6.4098, radius=1.684, shapeName=dyn_tail_l_02_capsule, center=<1.92 0 -1.1929>}]
 dyn_tail_l_03 [
  {shapeType=Capsule, length=5.7212, radius=1.3803, shapeName=dyn_tail_l_03_capsule, center=<1.5403 -0.5294 0>, rotation=<rot 90 -5.7843 0>}
  {shapeType=Capsule, length=3.9853, radius=1.3636, shapeName=dyn_tail_l_03_capsule, center=<-0.7261 -0.2466 -0.7383>}
 ]
 dyn_tail_r_01 [{shapeType=Capsule, length=5.2456, radius=2.016, shapeName=dyn_tail_r_01_capsule, center=<3.099 0 0>}]
 dyn_tail_r_02 [{shapeType=Capsule, length=6.4098, radius=1.684, shapeName=dyn_tail_r_02_capsule, center=<1.92 0 1.1929>}]
 dyn_tail_r_03 [
  {shapeType=Capsule, length=5.7212, radius=1.3803, shapeName=dyn_tail_r_03_capsule, center=<1.5403 0.5294 0>, rotation=<rot 90 -5.7842 0>}
  {shapeType=Capsule, length=3.9853, radius=1.3636, shapeName=dyn_tail_r_03_capsule, center=<-0.7261 0.2466 0.7383>}
 ]
 pelvis [{shapeType=Capsule, length=10.0041, radius=12.0171, shapeName=pelvis_capsule, center=<-2.0951 -0.9707 0>}]
 spine_01 [{shapeType=Capsule, length=9.4295, radius=8.7294, shapeName=spine_01_capsule, center=<0.6701 0.1954 0>}]
 spine_02 [
  {shapeType=Capsule, length=5.3989, radius=7.999, shapeName=spine_02_capsule, center=<-0.4181 1.2013 0>}
  {
   shapeType=Capsule
   length=3.3324
   radius=2.9479
   shapeName=spine_02_capsule
   center=<-0.6878 -6.3105 -7.5914>
   rotation=<rot -27.9653 -53.0621 74.9385>
  }
 ]
 spine_03 [{shapeType=Capsule, length=2.6399, radius=8.3385, shapeName=spine_03_capsule, center=<-0.285 1.827 0>}]
 spine_04 [{shapeType=Capsule, length=5.5828, radius=9.1633, shapeName=spine_04_capsule, center=<0 1.5259 0>}]
}
constraints [
 {bone1Name=dyn_jacket_l_01, bone2Name=base_jacket_l, swing1LimitDegrees=20, swing2LimitDegrees=17.5, twistLimitDegrees=0}
 {bone1Name=dyn_jacket_l_02, bone2Name=dyn_jacket_l_01, swing1LimitDegrees=5, swing2LimitDegrees=17.5, twistLimitDegrees=0}
 {bone1Name=dyn_jacket_l_03, bone2Name=dyn_jacket_l_02, swing1LimitDegrees=5, swing2LimitDegrees=17.5, twistLimitDegrees=0}
 {bone1Name=dyn_jacket_r_01, bone2Name=base_jacket_r, swing1LimitDegrees=20, swing2LimitDegrees=17.5, twistLimitDegrees=0}
 {bone1Name=dyn_jacket_r_02, bone2Name=dyn_jacket_r_01, swing1LimitDegrees=5, swing2LimitDegrees=17.5, twistLimitDegrees=0}
 {bone1Name=dyn_jacket_r_03, bone2Name=dyn_jacket_r_02, swing1LimitDegrees=5, swing2LimitDegrees=17.5, twistLimitDegrees=0}
 {
  bone1Name=dyn_tail_c_01
  bone2Name=base_tail_c
  swing1LimitDegrees=5
  swing1Motion=Locked
  swing2LimitDegrees=10
  swing2Motion=Locked
  twistLimitDegrees=0
  twistMotion=Locked
 }
 {bone1Name=dyn_tail_c_02, bone2Name=dyn_tail_c_01, swing1LimitDegrees=15, swing2LimitDegrees=10, twistLimitDegrees=0}
 {bone1Name=dyn_tail_c_03, bone2Name=dyn_tail_c_02, swing1LimitDegrees=15, swing2LimitDegrees=10, twistLimitDegrees=0}
 {
  bone1Name=dyn_tail_l_01
  bone2Name=base_tail_l
  swing1LimitDegrees=5
  swing1Motion=Locked
  swing2LimitDegrees=10
  swing2Motion=Locked
  twistLimitDegrees=0
  twistMotion=Locked
 }
 {bone1Name=dyn_tail_l_02, bone2Name=dyn_tail_l_01, swing1LimitDegrees=10, swing2LimitDegrees=10, twistLimitDegrees=0}
 {bone1Name=dyn_tail_l_03, bone2Name=dyn_tail_l_02, swing1LimitDegrees=15, swing2LimitDegrees=15, twistLimitDegrees=0}
 {
  bone1Name=dyn_tail_r_01
  bone2Name=base_tail_r
  swing1LimitDegrees=5
  swing1Motion=Locked
  swing2LimitDegrees=10
  swing2Motion=Locked
  twistLimitDegrees=0
  twistMotion=Locked
 }
 {bone1Name=dyn_tail_r_02, bone2Name=dyn_tail_r_01, swing1LimitDegrees=10, swing2LimitDegrees=10, twistLimitDegrees=0}
 {bone1Name=dyn_tail_r_03, bone2Name=dyn_tail_r_02, swing1LimitDegrees=15, swing2LimitDegrees=15, twistLimitDegrees=0}
 {
  bone1Name=base_jacket_l
  bone2Name=spine_04
  swing1LimitDegrees=45
  swing1Motion=Free
  swing2LimitDegrees=45
  swing2Motion=Free
  twistLimitDegrees=45
  twistMotion=Free
 }
 {
  bone1Name=base_jacket_r
  bone2Name=spine_04
  swing1LimitDegrees=45
  swing1Motion=Free
  swing2LimitDegrees=45
  swing2Motion=Free
  twistLimitDegrees=45
  twistMotion=Free
 }
 {
  bone1Name=base_tail_l
  bone2Name=spine_04
  swing1LimitDegrees=45
  swing1Motion=Free
  swing2LimitDegrees=45
  swing2Motion=Free
  twistLimitDegrees=45
  twistMotion=Free
 }
 {
  bone1Name=base_tail_r
  bone2Name=spine_04
  swing1LimitDegrees=45
  swing1Motion=Free
  swing2LimitDegrees=45
  swing2Motion=Free
  twistLimitDegrees=45
  twistMotion=Free
 }
 {
  bone1Name=base_tail_c
  bone2Name=spine_04
  swing1LimitDegrees=45
  swing1Motion=Free
  swing2LimitDegrees=45
  swing2Motion=Free
  twistLimitDegrees=45
  twistMotion=Free
 }
 {
  bone1Name=spine_03
  bone2Name=spine_02
  swing1LimitDegrees=45
  swing1Motion=Free
  swing2LimitDegrees=45
  swing2Motion=Free
  twistLimitDegrees=45
  twistMotion=Free
 }
 {
  bone1Name=spine_04
  bone2Name=spine_03
  swing1LimitDegrees=45
  swing1Motion=Free
  swing2LimitDegrees=45
  swing2Motion=Free
  twistLimitDegrees=45
  twistMotion=Free
 }
 {
  bone1Name=spine_02
  bone2Name=spine_01
  swing1LimitDegrees=45
  swing1Motion=Free
  swing2LimitDegrees=45
  swing2Motion=Free
  twistLimitDegrees=45
  twistMotion=Free
 }
 {
  bone1Name=spine_01
  bone2Name=pelvis
  swing1LimitDegrees=45
  swing1Motion=Free
  swing2LimitDegrees=45
  swing2Motion=Free
  twistLimitDegrees=45
  twistMotion=Free
 }
]
