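# Animation Blueprint for the F_MED_TwinGinger female body cosmetic, adding secondary motion to its clothing.
# It copies the attached parent mesh's pose into a Control Rig fed weapon, emote, skydiving and DBNO values.
# Pelvis-space RigidBody, buckle-to-jacket constraints and many single-bone AnimDynamics springs then move collar, lapels, buckles and shirt.
# Cached poses link the stages, ending in additive lapel bone adjustments before output.
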AnimBlueprint /BRCosmetics/Characters/Player/Female/Medium/Bodies/F_MED_TwinGinger/Meshes/F_MED_TwinGinger_AnimBP
target_skeleton=&/F_MED_TwinGinger_Skeleton
graphs [
 AnimGraph {nodes [
   Output Pose(Result=Use cached pose 'CRCache_Output'()) {layerGroup=None}
   n1 = "Transform (Modify) Bone - Bone: base_lapel_c"!(ComponentPose=Local To Component(LocalPose=Use cached pose 'CRCache_Dyn'()), bAlphaBoolEnabled:bool=True, Translation="-2,0,0") -> (Pose) {
    alphaInputType=Bool
    translationMode=BMM_Additive
    translationSpace=BCS_BoneSpace
    alphaBoolBlend {blendInTime=0.1, blendOption=Linear, blendOutTime=0.1}
    boneToModify {boneName=base_lapel_c}
    translation {x=-2}
   }
   n2 = "Transform (Modify) Bone - Bone: base_lapel_C_l"!(ComponentPose=@n1, bAlphaBoolEnabled:bool=True, Translation="0,0,0") -> (Pose) {
    alphaInputType=Bool
    translationMode=BMM_Additive
    translationSpace=BCS_BoneSpace
    alphaBoolBlend {blendInTime=0.1, blendOption=Linear, blendOutTime=0.1}
    boneToModify {boneName=base_lapel_C_l}
   }
   n3 = "Transform (Modify) Bone - Bone: base_lapel_C_r"!(ComponentPose=@n2, bAlphaBoolEnabled:bool=True, Translation="0,0,0") -> (Pose) {
    alphaInputType=Bool
    translationMode=BMM_Additive
    translationSpace=BCS_BoneSpace
    alphaBoolBlend {blendInTime=0.1, blendOption=Linear, blendOutTime=0.1}
    boneToModify {boneName=base_lapel_C_r}
   }
   Save cached pose 'CRCache_Output'(Pose=Component To Local(ComponentPose=@n3)) {cacheName=CRCache_Output}
   n4 = Control Rig!(
    Source=Copy Pose From Mesh() {bUseAttachedParent=true}
    Alpha
    isLargeWeapon:double=Get isLargeWeapon() -> (isLargeWeapon:double)
    isEmoting:double=Get isPlayingEmoteFloat() -> (isPlayingEmoteFloat:double)
    isSkydiving:double=Get f_Sky_Diving() -> (f_Sky_Diving:double)
    isDBNO:double=Get isDBNOFloat() -> (isDBNOFloat:double)
   ) -> (Pose) {
    alphaBoolBlend {blendOption=Linear}
    controlRigAssetReference {blueprintRigClass=&/F_MED_TwinGinger_CtrlRig.F_MED_TwinGinger_CtrlRig_C}
   }
   Save cached pose 'CRCache_Input'(Pose=@n4) {cacheName=CRCache_Input}
   n5 = Anim Dynamics - Bone: dyn_loop_01(ComponentPose=Local To Component(LocalPose=Use cached pose 'CRCache_Input'()), Alpha=1, ExternalForce="0,0,0") {
    angularDampingOverride=1
    angularSpringConstant=3500
    bAngularSpring=true
    bOverrideAngularDamping=true
    bUseGravityOverride=true
    gravityScale=0.35
    alphaBoolBlend {blendOption=Linear}
    boundBone {boneName=dyn_loop_01}
    componentAppliedLinearAccClamp=<10000 10000 10000>
    componentLinearAccScale=<1 1 1>
    componentLinearVelScale=<1 1 1>
    gravityOverride {z=-980}
    physicsBodyDefinitions [{
      boundBone {boneName=dyn_loop_01}
      boxExtents=<10 10 8>
      constraintSetup {angularLimitsMax=<5 5 0>, angularLimitsMin=<-5 -5 -100>, angularTarget=<1 0 0>}
      localJointOffset=<4 0 0>
     }]
   }
   n6 = RigidBody(ComponentPose=@n5, Alpha=0.8) {
    overridePhysicsAsset=&/F_MED_TwinGinger_Physics
    simulationSpace=BaseBoneSpace
    alphaBoolBlend {blendOption=Linear}
    baseBoneRef {boneName=pelvis}
    overrideWorldGravity {z=-980}
    simSpaceSettings {
     maxAngularAcceleration=300
     maxAngularVelocity=10
     maxLinearAcceleration=300
     maxLinearVelocity=500
     velocityScaleZ=0.5
     worldAlpha=1
     externalLinearDragV {z=0.2}
    }
   }
   n7 = "Constraint - dyn_buckle_up_l_02 (P)"(ComponentPose=@n6, Alpha=1, ConstraintWeights_0=1) {
    alphaBoolBlend {blendOption=Linear}
    boneToModify {boneName=dyn_buckle_up_l_02}
    constraintSetup [{perAxis {bX=false, bY=false}, targetBone {boneName=dyn_jacket_l_01}}]
    constraintWeights [1]
   }
   n8 = "Constraint - dyn_buckle_dw_l_02 (P)"(ComponentPose=@n7, Alpha=1, ConstraintWeights_0=1) {
    alphaBoolBlend {blendOption=Linear}
    boneToModify {boneName=dyn_buckle_dw_l_02}
    constraintSetup [{perAxis {bX=false, bY=false}, targetBone {boneName=dyn_jacket_l_02}}]
    constraintWeights [1]
   }
   n9 = "Constraint - dyn_buckle_up_r_02 (P)"(ComponentPose=@n8, Alpha=1, ConstraintWeights_0=1) {
    alphaBoolBlend {blendOption=Linear}
    boneToModify {boneName=dyn_buckle_up_r_02}
    constraintSetup [{perAxis {bX=false, bY=false}, targetBone {boneName=dyn_jacket_r_01}}]
    constraintWeights [1]
   }
   n10 = "Constraint - dyn_buckle_dw_r_01 (P)"(ComponentPose=@n9, Alpha=1, ConstraintWeights_0=1) {
    alphaBoolBlend {blendOption=Linear}
    boneToModify {boneName=dyn_buckle_dw_r_01}
    constraintSetup [{perAxis {bX=false, bY=false}, targetBone {boneName=dyn_jacket_r_02}}]
    constraintWeights [1]
   }
   n11 = Anim Dynamics - Bone: dyn_collar_A_l_01!(
    ComponentPose="Transform (Modify) Bone - Bone: base_collar_c"!(ComponentPose=@n10, Alpha=1, Rotation="0,0,0") -> (Pose) {rotationMode=BMM_Additive, alphaBoolBlend {blendOption=Linear}, boneToModify {boneName=base_collar_c}}
    ExternalForce="0,0,0"
   ) -> (Pose) {
    alpha=0.4
    angularDampingOverride=1
    angularSpringConstant=3500
    bAngularSpring=true
    bOverrideAngularDamping=true
    bUseGravityOverride=true
    gravityScale=0.35
    alphaBoolBlend {blendOption=Linear}
    boundBone {boneName=dyn_collar_A_l_01}
    componentAppliedLinearAccClamp=<10000 10000 10000>
    componentLinearAccScale=<1 1 1>
    componentLinearVelScale=<1 1 1>
    gravityOverride {z=-980}
    physicsBodyDefinitions [{
      boundBone {boneName=dyn_collar_A_l_01}
      boxExtents=<15 20 15>
      constraintSetup {angularLimitsMax=<0 10 20>, angularLimitsMin=<0 -10 -30>, angularTarget=<1 0 0>}
      localJointOffset=<1 0 0>
     }]
   }
   n12 = Anim Dynamics - Bone: dyn_collar_B_l_01!(ComponentPose=@n11, ExternalForce="0,0,0") -> (Pose) {
    alpha=0.4
    angularDampingOverride=1
    angularSpringConstant=3500
    bAngularSpring=true
    bOverrideAngularDamping=true
    bUseGravityOverride=true
    gravityScale=0.35
    simulationSpace=RootRelative
    alphaBoolBlend {blendOption=Linear}
    boundBone {boneName=dyn_collar_B_l_01}
    componentAppliedLinearAccClamp=<10000 10000 10000>
    componentLinearAccScale=<1 1 1>
    componentLinearVelScale=<1 1 1>
    gravityOverride {z=-980}
    physicsBodyDefinitions [{
      boundBone {boneName=dyn_collar_B_l_01}
      boxExtents=<10 20 10>
      constraintSetup {angularLimitsMax=<0 10 5>, angularLimitsMin=<0 -10 -25>, angularTarget=<1 0 0>}
      localJointOffset=<1 0 0>
     }]
   }
   n13 = Anim Dynamics - Bone: dyn_collar_c_01!(ComponentPose=@n12, ExternalForce="0,0,0") -> (Pose) {
    alpha=0.4
    angularDampingOverride=1
    angularSpringConstant=3500
    bAngularSpring=true
    bOverrideAngularDamping=true
    bUseGravityOverride=true
    gravityScale=0.35
    simulationSpace=RootRelative
    alphaBoolBlend {blendOption=Linear}
    boundBone {boneName=dyn_collar_c_01}
    componentAppliedLinearAccClamp=<10000 10000 10000>
    componentLinearAccScale=<1 1 1>
    componentLinearVelScale=<1 1 1>
    gravityOverride {z=-980}
    physicsBodyDefinitions [{
      boundBone {boneName=dyn_collar_c_01}
      boxExtents=<10 20 10>
      constraintSetup {angularLimitsMax=<0 5 0>, angularLimitsMin=<0 -5 -25>, angularTarget=<1 0 0>}
      localJointOffset=<1 0 0>
     }]
   }
   n14 = Anim Dynamics - Bone: dyn_collar_A_r_01!(ComponentPose=@n13, ExternalForce="0,0,0") -> (Pose) {
    alpha=0.4
    angularDampingOverride=1
    angularSpringConstant=3500
    bAngularSpring=true
    bOverrideAngularDamping=true
    bUseGravityOverride=true
    gravityScale=0.35
    alphaBoolBlend {blendOption=Linear}
    boundBone {boneName=dyn_collar_A_r_01}
    componentAppliedLinearAccClamp=<10000 10000 10000>
    componentLinearAccScale=<1 1 1>
    componentLinearVelScale=<1 1 1>
    gravityOverride {z=-980}
    physicsBodyDefinitions [{
      boundBone {boneName=dyn_collar_A_r_01}
      boxExtents=<15 20 15>
      constraintSetup {angularLimitsMax=<0 10 20>, angularLimitsMin=<0 -10 -30>, angularTarget=<1 0 0>}
      localJointOffset=<1 0 0>
     }]
   }
   n15 = Anim Dynamics - Bone: dyn_collar_B_r_01!(ComponentPose=@n14, ExternalForce="0,0,0") -> (Pose) {
    alpha=0.4
    angularDampingOverride=1
    angularSpringConstant=3500
    bAngularSpring=true
    bOverrideAngularDamping=true
    bUseGravityOverride=true
    gravityScale=0.35
    simulationSpace=RootRelative
    alphaBoolBlend {blendOption=Linear}
    boundBone {boneName=dyn_collar_B_r_01}
    componentAppliedLinearAccClamp=<10000 10000 10000>
    componentLinearAccScale=<1 1 1>
    componentLinearVelScale=<1 1 1>
    gravityOverride {z=-980}
    physicsBodyDefinitions [{
      boundBone {boneName=dyn_collar_B_r_01}
      boxExtents=<10 20 10>
      constraintSetup {angularLimitsMax=<0 10 5>, angularLimitsMin=<0 -10 -25>, angularTarget=<1 0 0>}
      localJointOffset=<1 0 0>
     }]
   }
   n16 = Anim Dynamics - Bone: dyn_lapel_A_l_01(ComponentPose=@n15, Alpha=0.8, ExternalForce="0,0,0") {
    angularDampingOverride=1
    angularSpringConstant=5500
    bAngularSpring=true
    bOverrideAngularDamping=true
    bUseGravityOverride=true
    gravityScale=0.35
    alphaBoolBlend {blendOption=Linear}
    boundBone {boneName=dyn_lapel_A_l_01}
    componentAppliedLinearAccClamp=<10000 10000 10000>
    componentLinearAccScale=<1 1 1>
    componentLinearVelScale=<1 1 1>
    gravityOverride {z=-980}
    physicsBodyDefinitions [{
      boundBone {boneName=dyn_lapel_A_l_01}
      boxExtents=<10 18 10>
      constraintSetup {angularLimitsMax=<0 10 25>, angularLimitsMin=<0 -10 -25>, angularTarget=<1 0 0>}
      localJointOffset=<0.5 0 0>
     }]
   }
   n17 = Anim Dynamics - Bone: dyn_lapel_A_l_02(ComponentPose=@n16, Alpha=0.8, ExternalForce="0,0,0") {
    angularDampingOverride=1
    angularSpringConstant=5500
    bAngularSpring=true
    bOverrideAngularDamping=true
    bUseGravityOverride=true
    gravityScale=0.35
    alphaBoolBlend {blendOption=Linear}
    boundBone {boneName=dyn_lapel_A_l_02}
    componentAppliedLinearAccClamp=<10000 10000 10000>
    componentLinearAccScale=<1 1 1>
    componentLinearVelScale=<1 1 1>
    gravityOverride {z=-980}
    physicsBodyDefinitions [{
      boundBone {boneName=dyn_lapel_A_l_02}
      boxExtents=<10 18 10>
      constraintSetup {angularLimitsMax=<0 10 25>, angularLimitsMin=<0 -10 -25>, angularTarget=<1 0 0>}
      localJointOffset=<0.5 0 0>
     }]
   }
   n18 = Anim Dynamics - Bone: dyn_lapel_A_l_03(ComponentPose=@n17, Alpha=0.8, ExternalForce="0,0,0") {
    angularDampingOverride=1
    angularSpringConstant=5500
    bAngularSpring=true
    bOverrideAngularDamping=true
    bUseGravityOverride=true
    gravityScale=0.35
    alphaBoolBlend {blendOption=Linear}
    boundBone {boneName=dyn_lapel_A_l_03}
    componentAppliedLinearAccClamp=<10000 10000 10000>
    componentLinearAccScale=<1 1 1>
    componentLinearVelScale=<1 1 1>
    gravityOverride {z=-980}
    physicsBodyDefinitions [{
      boundBone {boneName=dyn_lapel_A_l_03}
      boxExtents=<10 18 10>
      constraintSetup {angularLimitsMax=<0 10 25>, angularLimitsMin=<0 -10 -25>, angularTarget=<1 0 0>}
      localJointOffset=<0.5 0 0>
     }]
   }
   n19 = Anim Dynamics - Bone: dyn_lapel_B_l_01(ComponentPose=@n18, Alpha=0.8, ExternalForce="0,0,0") {
    angularDampingOverride=1
    angularSpringConstant=5500
    bAngularSpring=true
    bOverrideAngularDamping=true
    bUseGravityOverride=true
    gravityScale=0.35
    alphaBoolBlend {blendOption=Linear}
    boundBone {boneName=dyn_lapel_B_l_01}
    componentAppliedLinearAccClamp=<10000 10000 10000>
    componentLinearAccScale=<1 1 1>
    componentLinearVelScale=<1 1 1>
    gravityOverride {z=-980}
    physicsBodyDefinitions [{
      boundBone {boneName=dyn_lapel_B_l_01}
      boxExtents=<10 18 10>
      constraintSetup {angularLimitsMax=<0 10 25>, angularLimitsMin=<0 -10 -25>, angularTarget=<1 0 0>}
      localJointOffset=<0.5 0 0>
     }]
   }
   n20 = Anim Dynamics - Bone: dyn_lapel_B_l_02(ComponentPose=@n19, Alpha=0.8, ExternalForce="0,0,0") {
    angularDampingOverride=1
    angularSpringConstant=5500
    bAngularSpring=true
    bOverrideAngularDamping=true
    bUseGravityOverride=true
    gravityScale=0.35
    alphaBoolBlend {blendOption=Linear}
    boundBone {boneName=dyn_lapel_B_l_02}
    componentAppliedLinearAccClamp=<10000 10000 10000>
    componentLinearAccScale=<1 1 1>
    componentLinearVelScale=<1 1 1>
    gravityOverride {z=-980}
    physicsBodyDefinitions [{
      boundBone {boneName=dyn_lapel_B_l_02}
      boxExtents=<10 18 10>
      constraintSetup {angularLimitsMax=<0 10 25>, angularLimitsMin=<0 -10 -25>, angularTarget=<1 0 0>}
      localJointOffset=<0.5 0 0>
     }]
   }
   n21 = Anim Dynamics - Bone: dyn_lapel_B_l_03(ComponentPose=@n20, Alpha=0.8, ExternalForce="0,0,0") {
    angularDampingOverride=1
    angularSpringConstant=5500
    bAngularSpring=true
    bOverrideAngularDamping=true
    bUseGravityOverride=true
    gravityScale=0.35
    alphaBoolBlend {blendOption=Linear}
    boundBone {boneName=dyn_lapel_B_l_03}
    componentAppliedLinearAccClamp=<10000 10000 10000>
    componentLinearAccScale=<1 1 1>
    componentLinearVelScale=<1 1 1>
    gravityOverride {z=-980}
    physicsBodyDefinitions [{
      boundBone {boneName=dyn_lapel_B_l_03}
      boxExtents=<10 18 10>
      constraintSetup {angularLimitsMax=<0 10 25>, angularLimitsMin=<0 -10 -25>, angularTarget=<1 0 0>}
      localJointOffset=<0.5 0 0>
     }]
   }
   n22 = Anim Dynamics - Bone: dyn_lapel_C_l_01(ComponentPose=@n21, Alpha=1, ExternalForce="0,0,0") {
    angularDampingOverride=1
    angularSpringConstant=5500
    bAngularSpring=true
    bOverrideAngularDamping=true
    bUseGravityOverride=true
    gravityScale=0.35
    alphaBoolBlend {blendOption=Linear}
    boundBone {boneName=dyn_lapel_C_l_01}
    componentAppliedLinearAccClamp=<10000 10000 10000>
    componentLinearAccScale=<1 1 1>
    componentLinearVelScale=<1 1 1>
    gravityOverride {z=-980}
    physicsBodyDefinitions [{
      boundBone {boneName=dyn_lapel_C_l_01}
      boxExtents=<10 18 10>
      constraintSetup {angularLimitsMax=<0 10 15>, angularLimitsMin=<0 -10 -20>, angularTarget=<1 0 0>}
      localJointOffset=<0.5 0 0>
     }]
   }
   n23 = Anim Dynamics - Bone: dyn_lapel_C_l_02(ComponentPose=@n22, Alpha=1, ExternalForce="0,0,0") {
    angularDampingOverride=1
    angularSpringConstant=5500
    bAngularSpring=true
    bOverrideAngularDamping=true
    bUseGravityOverride=true
    gravityScale=0.35
    alphaBoolBlend {blendOption=Linear}
    boundBone {boneName=dyn_lapel_C_l_02}
    componentAppliedLinearAccClamp=<10000 10000 10000>
    componentLinearAccScale=<1 1 1>
    componentLinearVelScale=<1 1 1>
    gravityOverride {z=-980}
    physicsBodyDefinitions [{
      boundBone {boneName=dyn_lapel_C_l_02}
      boxExtents=<10 18 10>
      constraintSetup {angularLimitsMax=<0 10 15>, angularLimitsMin=<0 -10 -20>, angularTarget=<1 0 0>}
      localJointOffset=<0.5 0 0>
     }]
   }
   n24 = Anim Dynamics - Bone: dyn_lapel_C_l_03(ComponentPose=@n23, Alpha=1, ExternalForce="0,0,0") {
    angularDampingOverride=1
    angularSpringConstant=5500
    bAngularSpring=true
    bOverrideAngularDamping=true
    bUseGravityOverride=true
    gravityScale=0.35
    alphaBoolBlend {blendOption=Linear}
    boundBone {boneName=dyn_lapel_C_l_03}
    componentAppliedLinearAccClamp=<10000 10000 10000>
    componentLinearAccScale=<1 1 1>
    componentLinearVelScale=<1 1 1>
    gravityOverride {z=-980}
    physicsBodyDefinitions [{
      boundBone {boneName=dyn_lapel_C_l_03}
      boxExtents=<10 18 10>
      constraintSetup {angularLimitsMax=<0 10 15>, angularLimitsMin=<0 -10 -20>, angularTarget=<1 0 0>}
      localJointOffset=<0.5 0 0>
     }]
   }
   n25 = Anim Dynamics - Bone: dyn_lapel_c_01(ComponentPose=@n24, Alpha=1, ExternalForce="0,0,0") {
    angularDampingOverride=1
    angularSpringConstant=5500
    bAngularSpring=true
    bOverrideAngularDamping=true
    bUseGravityOverride=true
    gravityScale=0.35
    alphaBoolBlend {blendOption=Linear}
    boundBone {boneName=dyn_lapel_c_01}
    componentAppliedLinearAccClamp=<10000 10000 10000>
    componentLinearAccScale=<1 1 1>
    componentLinearVelScale=<1 1 1>
    gravityOverride {z=-980}
    physicsBodyDefinitions [{
      boundBone {boneName=dyn_lapel_c_01}
      boxExtents=<10 18 10>
      constraintSetup {angularLimitsMax=<0 5 10>, angularLimitsMin=<0 -5 -15>, angularTarget=<1 0 0>}
      localJointOffset=<0.5 0 0>
     }]
   }
   n26 = Anim Dynamics - Bone: dyn_lapel_c_02(ComponentPose=@n25, Alpha=1, ExternalForce="0,0,0") {
    angularDampingOverride=1
    angularSpringConstant=5500
    bAngularSpring=true
    bOverrideAngularDamping=true
    bUseGravityOverride=true
    gravityScale=0.35
    alphaBoolBlend {blendOption=Linear}
    boundBone {boneName=dyn_lapel_c_02}
    componentAppliedLinearAccClamp=<10000 10000 10000>
    componentLinearAccScale=<1 1 1>
    componentLinearVelScale=<1 1 1>
    gravityOverride {z=-980}
    physicsBodyDefinitions [{
      boundBone {boneName=dyn_lapel_c_02}
      boxExtents=<10 18 10>
      constraintSetup {angularLimitsMax=<0 5 10>, angularLimitsMin=<0 -5 -15>, angularTarget=<1 0 0>}
      localJointOffset=<0.5 0 0>
     }]
   }
   n27 = Anim Dynamics - Bone: dyn_lapel_c_03(ComponentPose=@n26, Alpha=1, ExternalForce="0,0,0") {
    angularDampingOverride=1
    angularSpringConstant=5500
    bAngularSpring=true
    bOverrideAngularDamping=true
    bUseGravityOverride=true
    gravityScale=0.35
    alphaBoolBlend {blendOption=Linear}
    boundBone {boneName=dyn_lapel_c_03}
    componentAppliedLinearAccClamp=<10000 10000 10000>
    componentLinearAccScale=<1 1 1>
    componentLinearVelScale=<1 1 1>
    gravityOverride {z=-980}
    physicsBodyDefinitions [{
      boundBone {boneName=dyn_lapel_c_03}
      boxExtents=<10 18 10>
      constraintSetup {angularLimitsMax=<0 5 10>, angularLimitsMin=<0 -5 -15>, angularTarget=<1 0 0>}
      localJointOffset=<0.5 0 0>
     }]
   }
   n28 = Anim Dynamics - Bone: dyn_lapel_A_r_01(ComponentPose=@n27, Alpha=1, ExternalForce="0,0,0") {
    angularDampingOverride=1
    angularSpringConstant=5500
    bAngularSpring=true
    bOverrideAngularDamping=true
    bUseGravityOverride=true
    gravityScale=0.35
    alphaBoolBlend {blendOption=Linear}
    boundBone {boneName=dyn_lapel_A_r_01}
    componentAppliedLinearAccClamp=<10000 10000 10000>
    componentLinearAccScale=<1 1 1>
    componentLinearVelScale=<1 1 1>
    gravityOverride {z=-980}
    physicsBodyDefinitions [{
      boundBone {boneName=dyn_lapel_A_r_01}
      boxExtents=<10 18 10>
      constraintSetup {angularLimitsMax=<0 10 25>, angularLimitsMin=<0 -10 -25>, angularTarget=<1 0 0>}
      localJointOffset=<0.5 0 0>
     }]
   }
   n29 = Anim Dynamics - Bone: dyn_lapel_A_r_02(ComponentPose=@n28, Alpha=1, ExternalForce="0,0,0") {
    angularDampingOverride=1
    angularSpringConstant=5500
    bAngularSpring=true
    bOverrideAngularDamping=true
    bUseGravityOverride=true
    gravityScale=0.35
    alphaBoolBlend {blendOption=Linear}
    boundBone {boneName=dyn_lapel_A_r_02}
    componentAppliedLinearAccClamp=<10000 10000 10000>
    componentLinearAccScale=<1 1 1>
    componentLinearVelScale=<1 1 1>
    gravityOverride {z=-980}
    physicsBodyDefinitions [{
      boundBone {boneName=dyn_lapel_A_r_02}
      boxExtents=<10 18 10>
      constraintSetup {angularLimitsMax=<0 10 25>, angularLimitsMin=<0 -10 -25>, angularTarget=<1 0 0>}
      localJointOffset=<0.5 0 0>
     }]
   }
   n30 = Anim Dynamics - Bone: dyn_lapel_A_r_03(ComponentPose=@n29, Alpha=1, ExternalForce="0,0,0") {
    angularDampingOverride=1
    angularSpringConstant=5500
    bAngularSpring=true
    bOverrideAngularDamping=true
    bUseGravityOverride=true
    gravityScale=0.35
    alphaBoolBlend {blendOption=Linear}
    boundBone {boneName=dyn_lapel_A_r_03}
    componentAppliedLinearAccClamp=<10000 10000 10000>
    componentLinearAccScale=<1 1 1>
    componentLinearVelScale=<1 1 1>
    gravityOverride {z=-980}
    physicsBodyDefinitions [{
      boundBone {boneName=dyn_lapel_A_r_03}
      boxExtents=<10 18 10>
      constraintSetup {angularLimitsMax=<0 10 25>, angularLimitsMin=<0 -10 -25>, angularTarget=<1 0 0>}
      localJointOffset=<0.5 0 0>
     }]
   }
   n31 = Anim Dynamics - Bone: dyn_lapel_B_r_01(ComponentPose=@n30, Alpha=1, ExternalForce="0,0,0") {
    angularDampingOverride=1
    angularSpringConstant=5500
    bAngularSpring=true
    bOverrideAngularDamping=true
    bUseGravityOverride=true
    gravityScale=0.35
    alphaBoolBlend {blendOption=Linear}
    boundBone {boneName=dyn_lapel_B_r_01}
    componentAppliedLinearAccClamp=<10000 10000 10000>
    componentLinearAccScale=<1 1 1>
    componentLinearVelScale=<1 1 1>
    gravityOverride {z=-980}
    physicsBodyDefinitions [{
      boundBone {boneName=dyn_lapel_B_r_01}
      boxExtents=<10 18 10>
      constraintSetup {angularLimitsMax=<0 10 25>, angularLimitsMin=<0 -10 -25>, angularTarget=<1 0 0>}
      localJointOffset=<0.5 0 0>
     }]
   }
   n32 = Anim Dynamics - Bone: dyn_lapel_B_r_02(ComponentPose=@n31, Alpha=1, ExternalForce="0,0,0") {
    angularDampingOverride=1
    angularSpringConstant=5500
    bAngularSpring=true
    bOverrideAngularDamping=true
    bUseGravityOverride=true
    gravityScale=0.35
    alphaBoolBlend {blendOption=Linear}
    boundBone {boneName=dyn_lapel_B_r_02}
    componentAppliedLinearAccClamp=<10000 10000 10000>
    componentLinearAccScale=<1 1 1>
    componentLinearVelScale=<1 1 1>
    gravityOverride {z=-980}
    physicsBodyDefinitions [{
      boundBone {boneName=dyn_lapel_B_r_02}
      boxExtents=<10 18 10>
      constraintSetup {angularLimitsMax=<0 10 25>, angularLimitsMin=<0 -10 -25>, angularTarget=<1 0 0>}
      localJointOffset=<0.5 0 0>
     }]
   }
   n33 = Anim Dynamics - Bone: dyn_lapel_B_r_03(ComponentPose=@n32, Alpha=1, ExternalForce="0,0,0") {
    angularDampingOverride=1
    angularSpringConstant=5500
    bAngularSpring=true
    bOverrideAngularDamping=true
    bUseGravityOverride=true
    gravityScale=0.35
    alphaBoolBlend {blendOption=Linear}
    boundBone {boneName=dyn_lapel_B_r_03}
    componentAppliedLinearAccClamp=<10000 10000 10000>
    componentLinearAccScale=<1 1 1>
    componentLinearVelScale=<1 1 1>
    gravityOverride {z=-980}
    physicsBodyDefinitions [{
      boundBone {boneName=dyn_lapel_B_r_03}
      boxExtents=<10 18 10>
      constraintSetup {angularLimitsMax=<0 10 25>, angularLimitsMin=<0 -10 -25>, angularTarget=<1 0 0>}
      localJointOffset=<0.5 0 0>
     }]
   }
   n34 = Anim Dynamics - Bone: dyn_lapel_C_r_01(ComponentPose=@n33, Alpha=1, ExternalForce="0,0,0") {
    angularDampingOverride=1
    angularSpringConstant=5500
    bAngularSpring=true
    bOverrideAngularDamping=true
    bUseGravityOverride=true
    gravityScale=0.35
    alphaBoolBlend {blendOption=Linear}
    boundBone {boneName=dyn_lapel_C_r_01}
    componentAppliedLinearAccClamp=<10000 10000 10000>
    componentLinearAccScale=<1 1 1>
    componentLinearVelScale=<1 1 1>
    gravityOverride {z=-980}
    physicsBodyDefinitions [{
      boundBone {boneName=dyn_lapel_C_r_01}
      boxExtents=<10 18 10>
      constraintSetup {angularLimitsMax=<0 10 15>, angularLimitsMin=<0 -10 -20>, angularTarget=<1 0 0>}
      localJointOffset=<0.5 0 0>
     }]
   }
   n35 = Anim Dynamics - Bone: dyn_lapel_C_r_02(ComponentPose=@n34, Alpha=1, ExternalForce="0,0,0") {
    angularDampingOverride=1
    angularSpringConstant=5500
    bAngularSpring=true
    bOverrideAngularDamping=true
    bUseGravityOverride=true
    gravityScale=0.35
    alphaBoolBlend {blendOption=Linear}
    boundBone {boneName=dyn_lapel_C_r_02}
    componentAppliedLinearAccClamp=<10000 10000 10000>
    componentLinearAccScale=<1 1 1>
    componentLinearVelScale=<1 1 1>
    gravityOverride {z=-980}
    physicsBodyDefinitions [{
      boundBone {boneName=dyn_lapel_C_r_02}
      boxExtents=<10 18 10>
      constraintSetup {angularLimitsMax=<0 10 15>, angularLimitsMin=<0 -10 -20>, angularTarget=<1 0 0>}
      localJointOffset=<0.5 0 0>
     }]
   }
   n36 = Anim Dynamics - Bone: dyn_lapel_C_r_03(ComponentPose=@n35, Alpha=1, ExternalForce="0,0,0") {
    angularDampingOverride=1
    angularSpringConstant=5500
    bAngularSpring=true
    bOverrideAngularDamping=true
    bUseGravityOverride=true
    gravityScale=0.35
    alphaBoolBlend {blendOption=Linear}
    boundBone {boneName=dyn_lapel_C_r_03}
    componentAppliedLinearAccClamp=<10000 10000 10000>
    componentLinearAccScale=<1 1 1>
    componentLinearVelScale=<1 1 1>
    gravityOverride {z=-980}
    physicsBodyDefinitions [{
      boundBone {boneName=dyn_lapel_C_r_03}
      boxExtents=<10 18 10>
      constraintSetup {angularLimitsMax=<0 10 15>, angularLimitsMin=<0 -10 -20>, angularTarget=<1 0 0>}
      localJointOffset=<0.5 0 0>
     }]
   }
   n37 = Anim Dynamics - Bone: dyn_lapelBuckle_l(ComponentPose=@n36, Alpha=1, ExternalForce="0,0,0") {
    angularDampingOverride=1
    angularSpringConstant=5000
    bAngularSpring=true
    bOverrideAngularDamping=true
    bUseGravityOverride=true
    gravityScale=0.35
    simulationSpace=RootRelative
    alphaBoolBlend {blendOption=Linear}
    boundBone {boneName=dyn_lapelBuckle_l}
    componentAppliedLinearAccClamp=<10000 10000 10000>
    componentLinearAccScale=<1 1 1>
    componentLinearVelScale=<1 1 1>
    gravityOverride {z=-980}
    physicsBodyDefinitions [{
      boundBone {boneName=dyn_lapelBuckle_l}
      boxExtents=<10 10 10>
      constraintSetup {angularLimitsMax=<0 0 2>, angularLimitsMin=<0 0 -5>, angularTarget=<1 0 0>}
      localJointOffset=<1 0 0>
     }]
   }
   n38 = Anim Dynamics - Bone: dyn_lapelBuckle_r(ComponentPose=@n37, Alpha=1, ExternalForce="0,0,0") {
    angularDampingOverride=1
    angularSpringConstant=5000
    bAngularSpring=true
    bOverrideAngularDamping=true
    bUseGravityOverride=true
    gravityScale=0.35
    simulationSpace=RootRelative
    alphaBoolBlend {blendOption=Linear}
    boundBone {boneName=dyn_lapelBuckle_r}
    componentAppliedLinearAccClamp=<10000 10000 10000>
    componentLinearAccScale=<1 1 1>
    componentLinearVelScale=<1 1 1>
    gravityOverride {z=-980}
    physicsBodyDefinitions [{
      boundBone {boneName=dyn_lapelBuckle_r}
      boxExtents=<10 10 10>
      constraintSetup {angularLimitsMax=<0 0 2>, angularLimitsMin=<0 0 -5>, angularTarget=<1 0 0>}
      localJointOffset=<1 0 0>
     }]
   }
   n39 = Anim Dynamics - Bone: dyn_buckle_up_l_01!(ComponentPose=@n38, ExternalForce="0,0,0") -> (Pose) {
    alpha=0.8
    angularDampingOverride=1
    angularSpringConstant=4000
    bAngularSpring=true
    bOverrideAngularDamping=true
    bUseGravityOverride=true
    bUsePlanarLimit=false
    gravityScale=0.35
    alphaBoolBlend {blendOption=Linear}
    boundBone {boneName=dyn_buckle_up_l_01}
    componentAppliedLinearAccClamp=<10000 10000 10000>
    componentLinearAccScale=<1 1 1>
    componentLinearVelScale=<1 1 1>
    gravityOverride {z=-980}
    physicsBodyDefinitions [{
      boundBone {boneName=dyn_buckle_up_l_01}
      boxExtents=<15 20 20>
      constraintSetup {angularLimitsMax=<0 10 0>, angularLimitsMin=<0 -15 -25>, angularTarget=<1 0 0>}
      localJointOffset=<4 0 0>
     }]
    relativeSpaceBone {boneName=spine_01}
   }
   n40 = Anim Dynamics - Bone: dyn_buckle_up_l_02!(ComponentPose=@n39, ExternalForce="0,0,0") -> (Pose) {
    alpha=0.8
    angularDampingOverride=1
    angularSpringConstant=4000
    bAngularSpring=true
    bOverrideAngularDamping=true
    bUseGravityOverride=true
    bUsePlanarLimit=false
    gravityScale=0.35
    alphaBoolBlend {blendOption=Linear}
    boundBone {boneName=dyn_buckle_up_l_02}
    componentAppliedLinearAccClamp=<10000 10000 10000>
    componentLinearAccScale=<1 1 1>
    componentLinearVelScale=<1 1 1>
    gravityOverride {z=-980}
    physicsBodyDefinitions [{
      boundBone {boneName=dyn_buckle_up_l_02}
      boxExtents=<15 20 20>
      constraintSetup {angularLimitsMax=<0 10 2>, angularLimitsMin=<0 -15 -25>, angularTarget=<1 0 0>}
      localJointOffset=<4 0 0>
     }]
    relativeSpaceBone {boneName=spine_01}
   }
   n41 = Anim Dynamics - Bone: dyn_buckle_up_l_03!(ComponentPose=@n40, ExternalForce="0,0,0") -> (Pose) {
    alpha=0.8
    angularDampingOverride=1
    angularSpringConstant=4000
    bAngularSpring=true
    bOverrideAngularDamping=true
    bUseGravityOverride=true
    bUsePlanarLimit=false
    gravityScale=0.35
    alphaBoolBlend {blendOption=Linear}
    boundBone {boneName=dyn_buckle_up_l_03}
    componentAppliedLinearAccClamp=<10000 10000 10000>
    componentLinearAccScale=<1 1 1>
    componentLinearVelScale=<1 1 1>
    gravityOverride {z=-980}
    physicsBodyDefinitions [{
      boundBone {boneName=dyn_buckle_up_l_03}
      boxExtents=<15 20 20>
      constraintSetup {angularLimitsMax=<0 10 2>, angularLimitsMin=<0 -15 -25>, angularTarget=<1 0 0>}
      localJointOffset=<4 0 0>
     }]
    relativeSpaceBone {boneName=spine_01}
   }
   n42 = Anim Dynamics - Bone: dyn_buckle_dw_l_01!(ComponentPose=@n41, ExternalForce="0,0,0") -> (Pose) {
    alpha=0.8
    angularDampingOverride=1
    angularSpringConstant=4000
    bAngularSpring=true
    bOverrideAngularDamping=true
    bUseGravityOverride=true
    bUsePlanarLimit=false
    gravityScale=0.35
    alphaBoolBlend {blendOption=Linear}
    boundBone {boneName=dyn_buckle_dw_l_01}
    componentAppliedLinearAccClamp=<10000 10000 10000>
    componentLinearAccScale=<1 1 1>
    componentLinearVelScale=<1 1 1>
    gravityOverride {z=-980}
    physicsBodyDefinitions [{
      boundBone {boneName=dyn_buckle_dw_l_01}
      boxExtents=<15 20 20>
      constraintSetup {angularLimitsMax=<0 10 0>, angularLimitsMin=<0 -15 -25>, angularTarget=<1 0 0>}
      localJointOffset=<4 0 0>
     }]
    relativeSpaceBone {boneName=spine_01}
   }
   n43 = Anim Dynamics - Bone: dyn_buckle_dw_l_02!(ComponentPose=@n42, ExternalForce="0,0,0") -> (Pose) {
    alpha=0.8
    angularDampingOverride=1
    angularSpringConstant=4000
    bAngularSpring=true
    bOverrideAngularDamping=true
    bUseGravityOverride=true
    bUsePlanarLimit=false
    gravityScale=0.35
    alphaBoolBlend {blendOption=Linear}
    boundBone {boneName=dyn_buckle_dw_l_02}
    componentAppliedLinearAccClamp=<10000 10000 10000>
    componentLinearAccScale=<1 1 1>
    componentLinearVelScale=<1 1 1>
    gravityOverride {z=-980}
    physicsBodyDefinitions [{
      boundBone {boneName=dyn_buckle_dw_l_02}
      boxExtents=<15 20 20>
      constraintSetup {angularLimitsMax=<0 10 2>, angularLimitsMin=<0 -15 -25>, angularTarget=<1 0 0>}
      localJointOffset=<4 0 0>
     }]
    relativeSpaceBone {boneName=spine_01}
   }
   n44 = Anim Dynamics - Bone: dyn_buckle_dw_l_03!(ComponentPose=@n43, ExternalForce="0,0,0") -> (Pose) {
    alpha=0.8
    angularDampingOverride=1
    angularSpringConstant=4000
    bAngularSpring=true
    bOverrideAngularDamping=true
    bUseGravityOverride=true
    bUsePlanarLimit=false
    gravityScale=0.35
    alphaBoolBlend {blendOption=Linear}
    boundBone {boneName=dyn_buckle_dw_l_03}
    componentAppliedLinearAccClamp=<10000 10000 10000>
    componentLinearAccScale=<1 1 1>
    componentLinearVelScale=<1 1 1>
    gravityOverride {z=-980}
    physicsBodyDefinitions [{
      boundBone {boneName=dyn_buckle_dw_l_03}
      boxExtents=<15 20 20>
      constraintSetup {angularLimitsMax=<0 10 2>, angularLimitsMin=<0 -15 -25>, angularTarget=<1 0 0>}
      localJointOffset=<4 0 0>
     }]
    relativeSpaceBone {boneName=spine_01}
   }
   n45 = Anim Dynamics - Bone: dyn_buckle_up_r_01!(ComponentPose=@n44, ExternalForce="0,0,0") -> (Pose) {
    alpha=0.8
    angularDampingOverride=1
    angularSpringConstant=4000
    bAngularSpring=true
    bOverrideAngularDamping=true
    bUseGravityOverride=true
    bUsePlanarLimit=false
    gravityScale=0.35
    alphaBoolBlend {blendOption=Linear}
    boundBone {boneName=dyn_buckle_up_r_01}
    componentAppliedLinearAccClamp=<10000 10000 10000>
    componentLinearAccScale=<1 1 1>
    componentLinearVelScale=<1 1 1>
    gravityOverride {z=-980}
    physicsBodyDefinitions [{
      boundBone {boneName=dyn_buckle_up_r_01}
      boxExtents=<15 20 20>
      constraintSetup {angularLimitsMax=<0 10 0>, angularLimitsMin=<0 -15 -25>, angularTarget=<1 0 0>}
      localJointOffset=<4 0 0>
     }]
    relativeSpaceBone {boneName=spine_01}
   }
   n46 = Anim Dynamics - Bone: dyn_buckle_up_r_02!(ComponentPose=@n45, ExternalForce="0,0,0") -> (Pose) {
    alpha=0.8
    angularDampingOverride=1
    angularSpringConstant=4000
    bAngularSpring=true
    bOverrideAngularDamping=true
    bUseGravityOverride=true
    bUsePlanarLimit=false
    gravityScale=0.35
    alphaBoolBlend {blendOption=Linear}
    boundBone {boneName=dyn_buckle_up_r_02}
    componentAppliedLinearAccClamp=<10000 10000 10000>
    componentLinearAccScale=<1 1 1>
    componentLinearVelScale=<1 1 1>
    gravityOverride {z=-980}
    physicsBodyDefinitions [{
      boundBone {boneName=dyn_buckle_up_r_02}
      boxExtents=<15 20 20>
      constraintSetup {angularLimitsMax=<0 10 2>, angularLimitsMin=<0 -15 -25>, angularTarget=<1 0 0>}
      localJointOffset=<4 0 0>
     }]
    relativeSpaceBone {boneName=spine_01}
   }
   n47 = Anim Dynamics - Bone: dyn_buckle_dw_r_01!(ComponentPose=@n46, ExternalForce="0,0,0") -> (Pose) {
    alpha=0.8
    angularDampingOverride=1
    angularSpringConstant=4000
    bAngularSpring=true
    bOverrideAngularDamping=true
    bUseGravityOverride=true
    bUsePlanarLimit=false
    gravityScale=0.35
    alphaBoolBlend {blendOption=Linear}
    boundBone {boneName=dyn_buckle_dw_r_01}
    componentAppliedLinearAccClamp=<10000 10000 10000>
    componentLinearAccScale=<1 1 1>
    componentLinearVelScale=<1 1 1>
    gravityOverride {z=-980}
    physicsBodyDefinitions [{
      boundBone {boneName=dyn_buckle_dw_r_01}
      boxExtents=<15 20 20>
      constraintSetup {angularLimitsMax=<0 10 0>, angularLimitsMin=<0 -15 -25>, angularTarget=<1 0 0>}
      localJointOffset=<4 0 0>
     }]
    relativeSpaceBone {boneName=spine_01}
   }
   n48 = Anim Dynamics - Bone: dyn_buckle_dw_r_02!(ComponentPose=@n47, ExternalForce="0,0,0") -> (Pose) {
    alpha=0.8
    angularDampingOverride=1
    angularSpringConstant=4000
    bAngularSpring=true
    bOverrideAngularDamping=true
    bUseGravityOverride=true
    bUsePlanarLimit=false
    gravityScale=0.35
    alphaBoolBlend {blendOption=Linear}
    boundBone {boneName=dyn_buckle_dw_r_02}
    componentAppliedLinearAccClamp=<10000 10000 10000>
    componentLinearAccScale=<1 1 1>
    componentLinearVelScale=<1 1 1>
    gravityOverride {z=-980}
    physicsBodyDefinitions [{
      boundBone {boneName=dyn_buckle_dw_r_02}
      boxExtents=<15 20 20>
      constraintSetup {angularLimitsMax=<0 10 2>, angularLimitsMin=<0 -15 -25>, angularTarget=<1 0 0>}
      localJointOffset=<4 0 0>
     }]
    relativeSpaceBone {boneName=spine_01}
   }
   n49 = Anim Dynamics - Bone: dyn_shirt_l_01(ComponentPose=@n48, Alpha=1, ExternalForce="0,0,0") {
    angularDampingOverride=1
    angularSpringConstant=3500
    bAngularSpring=true
    bOverrideAngularDamping=true
    bUseGravityOverride=true
    gravityScale=0.35
    simulationSpace=RootRelative
    alphaBoolBlend {blendOption=Linear}
    alphaScaleBiasClamp {bInterpResult=true}
    boundBone {boneName=dyn_shirt_l_01}
    componentAppliedLinearAccClamp=<10000 10000 10000>
    componentLinearAccScale=<1 1 1>
    componentLinearVelScale=<1 1 1>
    gravityOverride {z=-980}
    physicsBodyDefinitions [{
      boundBone {boneName=dyn_shirt_l_01}
      boxExtents=<10 10 10>
      constraintSetup {angularLimitsMax=<0 4 0>, angularLimitsMin=<0 -10 -20>, angularTarget=<1 0 0>}
      localJointOffset=<1 0 1>
     }]
   }
   n50 = Anim Dynamics - Bone: dyn_shirt_l_02(ComponentPose=@n49, Alpha=1, ExternalForce="0,0,0") {
    angularDampingOverride=1
    angularSpringConstant=3500
    bAngularSpring=true
    bOverrideAngularDamping=true
    bUseGravityOverride=true
    gravityScale=0.35
    simulationSpace=RootRelative
    alphaBoolBlend {blendOption=Linear}
    alphaScaleBiasClamp {bInterpResult=true}
    boundBone {boneName=dyn_shirt_l_02}
    componentAppliedLinearAccClamp=<10000 10000 10000>
    componentLinearAccScale=<1 1 1>
    componentLinearVelScale=<1 1 1>
    gravityOverride {z=-980}
    physicsBodyDefinitions [{
      boundBone {boneName=dyn_shirt_l_02}
      boxExtents=<10 10 10>
      constraintSetup {angularLimitsMax=<0 4 0>, angularLimitsMin=<0 -10 -20>, angularTarget=<1 0 0>}
      localJointOffset=<1 0 1>
     }]
   }
   n51 = Anim Dynamics - Bone: dyn_shirt_l_03(ComponentPose=@n50, Alpha=1, ExternalForce="0,0,0") {
    angularDampingOverride=1
    angularSpringConstant=3500
    bAngularSpring=true
    bOverrideAngularDamping=true
    bUseGravityOverride=true
    gravityScale=0.35
    simulationSpace=RootRelative
    alphaBoolBlend {blendOption=Linear}
    alphaScaleBiasClamp {bInterpResult=true}
    boundBone {boneName=dyn_shirt_l_03}
    componentAppliedLinearAccClamp=<10000 10000 10000>
    componentLinearAccScale=<1 1 1>
    componentLinearVelScale=<1 1 1>
    gravityOverride {z=-980}
    physicsBodyDefinitions [{
      boundBone {boneName=dyn_shirt_l_03}
      boxExtents=<10 10 10>
      constraintSetup {angularLimitsMax=<0 4 0>, angularLimitsMin=<0 -10 -20>, angularTarget=<1 0 0>}
      localJointOffset=<1 0 1>
     }]
   }
   n52 = Anim Dynamics - Bone: dyn_shirt_r_01(ComponentPose=@n51, Alpha=1, ExternalForce="0,0,0") {
    angularDampingOverride=1
    angularSpringConstant=3500
    bAngularSpring=true
    bOverrideAngularDamping=true
    bUseGravityOverride=true
    gravityScale=0.35
    simulationSpace=RootRelative
    alphaBoolBlend {blendOption=Linear}
    alphaScaleBiasClamp {bInterpResult=true}
    boundBone {boneName=dyn_shirt_r_01}
    componentAppliedLinearAccClamp=<10000 10000 10000>
    componentLinearAccScale=<1 1 1>
    componentLinearVelScale=<1 1 1>
    gravityOverride {z=-980}
    physicsBodyDefinitions [{
      boundBone {boneName=dyn_shirt_r_01}
      boxExtents=<10 10 10>
      constraintSetup {angularLimitsMax=<0 4 0>, angularLimitsMin=<0 -10 -20>, angularTarget=<1 0 0>}
      localJointOffset=<1 0 -1>
     }]
   }
   n53 = Anim Dynamics - Bone: dyn_shirt_r_02(ComponentPose=@n52, Alpha=1, ExternalForce="0,0,0") {
    angularDampingOverride=1
    angularSpringConstant=3500
    bAngularSpring=true
    bOverrideAngularDamping=true
    bUseGravityOverride=true
    gravityScale=0.35
    simulationSpace=RootRelative
    alphaBoolBlend {blendOption=Linear}
    alphaScaleBiasClamp {bInterpResult=true}
    boundBone {boneName=dyn_shirt_r_02}
    componentAppliedLinearAccClamp=<10000 10000 10000>
    componentLinearAccScale=<1 1 1>
    componentLinearVelScale=<1 1 1>
    gravityOverride {z=-980}
    physicsBodyDefinitions [{
      boundBone {boneName=dyn_shirt_r_02}
      boxExtents=<10 10 10>
      constraintSetup {angularLimitsMax=<0 4 0>, angularLimitsMin=<0 -10 -20>, angularTarget=<1 0 0>}
      localJointOffset=<1 0 -1>
     }]
   }
   n54 = Anim Dynamics - Bone: dyn_shirt_r_03(ComponentPose=@n53, Alpha=1, ExternalForce="0,0,0") {
    angularDampingOverride=1
    angularSpringConstant=3500
    bAngularSpring=true
    bOverrideAngularDamping=true
    bUseGravityOverride=true
    gravityScale=0.35
    simulationSpace=RootRelative
    alphaBoolBlend {blendOption=Linear}
    alphaScaleBiasClamp {bInterpResult=true}
    boundBone {boneName=dyn_shirt_r_03}
    componentAppliedLinearAccClamp=<10000 10000 10000>
    componentLinearAccScale=<1 1 1>
    componentLinearVelScale=<1 1 1>
    gravityOverride {z=-980}
    physicsBodyDefinitions [{
      boundBone {boneName=dyn_shirt_r_03}
      boxExtents=<10 10 10>
      constraintSetup {angularLimitsMax=<0 4 0>, angularLimitsMin=<0 -10 -20>, angularTarget=<1 0 0>}
      localJointOffset=<1 0 -1>
     }]
   }
   Save cached pose 'CRCache_Dyn'(Pose=Component To Local(ComponentPose=@n54)) {cacheName=CRCache_Dyn}
  ]}
 EventGraph {nodes [
   Event OnEndCrouching()
   > Set SkirtAlpha(SkirtAlpha:double=1) -> (Output_Get:double)
   Event OnBeginCrouching()
   > Set SkirtAlpha(SkirtAlpha=0)
   n1 = Event BlueprintUpdateAnimation()
   n2 = Get bIsPlayingEmote()
   > Set isPlayingEmoteFloat(isPlayingEmoteFloat:double=FInterpTo(Current=Get isPlayingEmoteFloat(), Target="To Float (Boolean)"(InBool=@n2), DeltaTime=@n1.DeltaTimeX, InterpSpeed=15)) -> (Output_Get:double)
   n3 = "To Float (Boolean)"(InBool=InvertBool!(bValue:bool=Get bIsDBNOCarried()) -> (ReturnValue:bool))
   n4 = "To Float (Boolean)"(InBool=Get bIsDBNO())
   > Set isDBNOFloat(isDBNOFloat:double=FInterpTo(Current=Get isDBNOFloat(), Target=float * float(A=@n4, B=@n3) {is=title K2Node_PromotableOperator}, DeltaTime=@n1.DeltaTimeX, InterpSpeed=15)) -> (Output_Get:double)
   Event OnBeginSkydiving()
   > n5 = Set f_Sky_Diving(f_Sky_Diving:double=1) -> (Output_Get:double)
   Set CollarFixAlpha(execute=[@n7.then, @n8.then, @n5.then], CollarFixAlpha:bool=OR Boolean(A=Get bIsSkydiveDiveMode(), B=Get bIsSkydiveFloating())) -> (Output_Get:bool)
   Event OnEndSkydiving()
   > Set f_Sky_Diving(f_Sky_Diving=0)
   Event OnWeaponChanged()
   > Set isLargeWeapon(isLargeWeapon:double=0) -> (Output_Get:double)
   > n6 = Switch on EFortWeaponCoreAnimation!(execute, Selection:EFortWeaponCoreAnimation Enum=Get WeaponCoreAnim!() -> (WeaponCoreAnim)) -> (
    Melee:Exec
    Pistol:Exec
    Shotgun:Exec
    PaperBlueprint:Exec
    Rifle:Exec
    MeleeOneHand:Exec
    MachinePistol:Exec
    RocketLauncher:Exec
    GrenadeLauncher:Exec
    GoingCommando:Exec
    AssaultRifle:Exec
    TacticalShotgun:Exec
    SniperRifle:Exec
    TrapPlacement:Exec
    ShoulderLauncher:Exec
    AbilityDecoTool:Exec
    Crossbow:Exec
    C4:Exec
    RemoteControl:Exec
    DualWield:Exec
    AR_BullPup:Exec
    AR_ForwardGrip:Exec
    MedPackPaddles:Exec
    SMG_P90:Exec
    AR_DrumGun:Exec
    Consumable_Small:Exec
    Consumable_Large:Exec
    Balloon:Exec
    MountedTurret:Exec
    CreativeTool:Exec
    ExplosiveBow:Exec
    AshtonIndigo:Exec
    AshtonChicago:Exec
    MeleeDualWield:Exec
    Unarmed:Exec
   )
   Set isLargeWeapon(execute=[@n6.AssaultRifle, @n6.Rifle, @n6.RocketLauncher, @n6.SMG_P90, @n6.TacticalShotgun], isLargeWeapon=1)
   > PrintText!(
    execute
    InText="LOCTEXT(\"Large weapon equipped\")"
    bPrintToScreen
    bPrintToLog=false
    TextColor=(R=1,G=0.2983,B=0.1364,A=1)
    Duration
    Key
   ) -> (then)
   n7 = Event OnParachuteOpened()
   n8 = Event OnParachuteClosed()
  ]}
]
variables [
 CollarFixAlpha {type=boolean}
 f_Sky_Diving {type=number}
 isDBNOFloat {type=number}
 isLargeWeapon {type=number}
 isPlayingEmoteFloat {type=number}
 LapelAlpha {type=number, default=0.2}
 SkirtAlpha {type=number, default=1}
]
