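# Physics asset for the F_MED_TwinGinger female face accessory mesh, defining capsule bodies for a left hair strand, two left lapels, head and neck.
# The four dynamic hair bodies chain from a base body with narrow per-link swing limits and near-locked twist.
# Hair base-to-head and head-to-neck joints leave swing and twist free.
# The two lapel capsules have no constraints in this asset.
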PhysicsAsset /BRCosmetics/Characters/Player/Female/Medium/Bodies/F_MED_TwinGinger/Meshes/Parts/F_MED_TwinGinger_FaceAcc_Physics
bodies {
 base_leftHair_A [{shapeType=Capsule, length=0.505, radius=0.505, shapeName=base_leftHair_A_capsule, rotation=<rot 0 -90 0>}]
 dyn_lapel_A_l_01 [{
   shapeType=Capsule
   length=9.495
   radius=4.6519
   shapeName=dyn_lapel_A_l_01_capsule
   center=<-0.2573 -1.2638 3.0414>
   rotation=<rot 12.3153 -67.5459 -16.105>
  }]
 dyn_lapel_B_l_01 [{
   shapeType=Capsule
   length=8.156
   radius=3.2257
   shapeName=dyn_lapel_B_l_01_capsule
   center=<0.0568 -2.3378 3.9617>
   rotation=<rot 14.2373 -84.5646 -15.0385>
  }]
 dyn_leftHair_A_01 [{
   shapeType=Capsule
   length=3.1556
   radius=1.3224
   shapeName=dyn_leftHair_A_01_capsule
   center=<2.0312 0.5573 -0.4941>
   rotation=<rot 0.0545 88.3947 -88.1957>
  }]
 dyn_leftHair_A_02 [{
   shapeType=Capsule
   length=3.7738
   radius=1.3902
   shapeName=dyn_leftHair_A_02_capsule
   center=<2.1179 0.3567 -0.154>
   rotation=<rot 84.5467 180 180>
  }]
 dyn_leftHair_A_03 [{
   shapeType=Capsule
   length=3.379
   radius=1.2616
   shapeName=dyn_leftHair_A_03_capsule
   center=<1.975 0.2093 -0.0345>
   rotation=<rot -86.7018 0 0>
  }]
 dyn_leftHair_A_04 [{
   shapeType=Capsule
   length=3.6262
   radius=1.153
   shapeName=dyn_leftHair_A_04_capsule
   center=<2.2408 0.0742 0.14>
   rotation=<rot 0 90.2248 -89.0265>
  }]
 head [{shapeType=Capsule, length=6.6898, radius=5.4539, center=<1.6636 3.3619 -0.0094>, rotation=<rot 4.5021 90.1711 -88.4286>}]
 neck_01 [{shapeType=Capsule, length=2.6397, radius=5.0338, shapeName=neck_01_capsule, center=<4.3571 0 -0.4038>, rotation=<rot 0 -90 95.2944>}]
}
constraints [
 {
  bone1Name=base_leftHair_A
  bone2Name=head
  swing1LimitDegrees=45
  swing1Motion=Free
  swing2LimitDegrees=45
  swing2Motion=Free
  twistLimitDegrees=45
  twistMotion=Free
 }
 {bone1Name=dyn_leftHair_A_01, bone2Name=base_leftHair_A, swing1LimitDegrees=10, swing2LimitDegrees=10, twistLimitDegrees=1}
 {bone1Name=dyn_leftHair_A_02, bone2Name=dyn_leftHair_A_01, swing1LimitDegrees=15, swing2LimitDegrees=10, twistLimitDegrees=1}
 {
  bone1Name=head
  bone2Name=neck_01
  swing1LimitDegrees=45
  swing1Motion=Free
  swing2LimitDegrees=45
  swing2Motion=Free
  twistLimitDegrees=45
  twistMotion=Free
 }
 {bone1Name=dyn_leftHair_A_03, bone2Name=dyn_leftHair_A_02, swing1LimitDegrees=20, swing2LimitDegrees=15, twistLimitDegrees=1}
 {bone1Name=dyn_leftHair_A_04, bone2Name=dyn_leftHair_A_03, swing1LimitDegrees=10, swing2LimitDegrees=5, twistLimitDegrees=1}
]
