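# Corrective Control Rig for the TwinGinger female character's hair part, rotating the base_leftHair_A bone as the head turns against the neck.
# The forward solve runs Hair_fix, which calls a reusable DeltaRotation v4 function on head relative to neck_02.
# That function splits swing from twist, remaps one Euler axis (here swing Y) to a clamped weight, and can print the rotation for debugging.
# The weight scales a fixed rotation applied through Modify Transforms.
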ControlRigBlueprint /BRCosmetics/Characters/Player/Female/Medium/Bodies/F_MED_TwinGinger/Meshes/Parts/F_MED_TwinGinger_FaceAcc_hair_CtrlRig
graphs [
 New Function_ContainedGraph {nodes [
   n1 = RigVMFunctionEntryNode()
   RigVMFunction_Sequence = &RigVMFunction_Sequence(ExecuteContext=@n1.ExecuteContext)
   DeltaRotation v4_1 = RigVMFunctionReferenceNode(
    Rotation Order=ZYX
    child=(Type=Bone,Name="head")
    debugRotation=false
    outputMinMax=(X=0,Y=1)
    parent=(Type=Bone,Name="neck_02")
    premult=1
    rotationMinMax=(X=0,Y=5)
    triggerAxis=Y
    twistAxis=(X=1,Y=0,Z=0)
    ExecuteContext=@RigVMFunction_Sequence.A
   )
   > &RigUnit_ModifyTransforms(
    ItemToModify=((Item=(Type=Bone,Name="base_leftHair_A"),Transform=(Rotation=(X=0.1294,Y=0.1294,Z=-0.017,W=0.983),Translation=(X=0,Y=0,Z=0),Scale3D=(X=1,Y=1,Z=1))))
    WeightMinimum=0
    Weight=@"DeltaRotation v4_1".outputWeight
   )
   RigVMFunctionReturnNode(ExecuteContext=@RigVMFunction_Sequence.B)
  ]}
 New Function_ContainedGraph {nodes [
   n1 = RigVMFunctionEntryNode()
   RigVMFunction_MathBoolToInteger = &RigVMFunction_MathBoolToInteger(Value=@n1.onlyTwist)
   RigVMFunction_MathQuaternionSwingTwist = &RigVMFunction_MathQuaternionSwingTwist(
    twistAxis=@n1.twistAxis
    Input=&RigUnit_GetRelativeTransformForItem(bChildInitial=False, bParentInitial=False, child=@n1.child, parent=@n1.parent).RelativeTransform.Rotation
   )
   n2 = RigVMDispatchNode(
    Values=((),())
    Index=@RigVMFunction_MathBoolToInteger.Result
    Values.0=@RigVMFunction_MathQuaternionSwingTwist.Swing
    Values.1=@RigVMFunction_MathQuaternionSwingTwist.Twist
   ) {template="DISPATCH_RigVMDispatch_SelectInt32(in Index,in Values,out Result)"}
   RigVMFunction_MathQuaternionToEuler = &RigVMFunction_MathQuaternionToEuler(Value=@n2.Result, RotationOrder=@n1.Rotation Order)
   > DISPATCH_RigVMDispatch_Print = RigVMDispatchNode(ScreenColor=(R=1,G=1,B=1,A=1), ScreenDuration=1.5, Enabled=@n1.debugRotation, Value=@RigVMFunction_MathQuaternionToEuler.Result) {template="DISPATCH_RigVMDispatch_Print(in Prefix,in Value,in Enabled,in ScreenDuration,in ScreenColor)"}
   Cast = RigVMDispatchNode(Value=@n1.triggerAxis) {template="DISPATCH_RigVMDispatch_CastEnumToInt(in Value,out Result)"}
   DISPATCH_RigVMDispatch_SelectInt32 = RigVMDispatchNode(
    Values=(0,0,0)
    Index=@Cast.Result
    Values.0=@RigVMFunction_MathQuaternionToEuler.Result.X
    Values.1=@RigVMFunction_MathQuaternionToEuler.Result.Y
    Values.2=@RigVMFunction_MathQuaternionToEuler.Result.Z
   ) {template="DISPATCH_RigVMDispatch_SelectInt32(in Index,in Values,out Result)"}
   Remap = &RigVMFunction_MathDoubleRemap(
    bClamp=False
    Value=@DISPATCH_RigVMDispatch_SelectInt32.Result
    TargetMinimum=@n1.outputMinMax.X
    TargetMaximum=@n1.outputMinMax.Y
    SourceMinimum=@n1.rotationMinMax.X
    SourceMaximum=@n1.rotationMinMax.Y
   )
   Absolute = &RigVMFunction_MathDoubleAbs(Value=@Remap.Result)
   Multiply = &RigVMFunction_MathDoubleMul(
    A=&RigVMFunction_MathDoubleClamp(Value=RigVMDispatchNode(True=@Absolute.Result, Condition=@n1.absoluteValue, False=@Remap.Result) {template="DISPATCH_RigVMDispatch_If(in Condition,in True,in False,out Result)"}.Result).Result
    B=@n1.premult
   )
   > RigVMFunctionReturnNode(outputQuat=@n2.Result, outputWeight=@Multiply.Result)
  ]}
 RigVMFunctionLibrary {nodes [
   DeltaRotation v4 = RigVMCollapseNode(
    ExecuteContext=()
    Rotation Order=ZYX
    child=(Type=Bone,Name="None")
    outputMinMax=(X=0,Y=1)
    parent=(Type=Bone,Name="None")
    premult=1
    rotationMinMax=(X=0,Y=0)
    twistAxis=(X=1,Y=0,Z=0)
   )
   Hair_fix = RigVMCollapseNode(ExecuteContext=())
  ]}
 RigVMModel {nodes [
   RigUnit_BeginExecution = &RigUnit_BeginExecution()
   Hair_fix = RigVMFunctionReferenceNode(ExecuteContext=@RigUnit_BeginExecution.ExecutePin)
  ]}
]
hierarchy [
 root {loc=<0 0 0>, rot=<0 0 0 1>}
 pelvis {parent=root, loc=<0 -0.5301 98.0571>, rot=<0.0127 0.707 -0.0127 -0.707>}
 spine_01 {parent=pelvis, loc=<6.1208 -0.9328 0>, rot=<0 0 0.034 -0.9994>}
 spine_02 {parent=spine_01, loc=<6.5997 -0.2095 0>, rot=<0 0 0.0238 -0.9997>}
 spine_03 {parent=spine_02, loc=<6.4914 -0.1237 0>, rot=<0 0 0.0149 -0.9999>}
 spine_04 {parent=spine_03, loc=<6.9598 0.0794 0>, rot=<0 0 -0.0132 -0.9999>}
 spine_05 {parent=spine_04, loc=<6.9987 0 0>, rot=<0 0 -0.0015 -1>}
 neck_01 {parent=spine_05, loc=<11.3138 0.1529 0>, rot=<0 0 -0.2003 -0.9797>}
 neck_02 {parent=neck_01, loc=<4.5627 0 0>, rot=<0 0 0.0504 -0.9987>}
 head {parent=neck_02, loc=<5.3914 0 0>, rot=<0 0 0.115 -0.9934>}
 base_rightHair_D {parent=head, loc=<3.9896 -2.4948 7.0211>, rot=<0.0862 0.3754 0.9189 -0.0854>}
 dyn_rightHair_D {parent=base_rightHair_D, loc=<0 0 0>, rot=<0 0 0 1>}
 base_rightHair_C {parent=head, loc=<8.4407 -0.4349 7.2156>, rot=<0.246 0.4464 0.7601 -0.403>}
 dyn_rightHair_C {parent=base_rightHair_C, loc=<0 0 0>, rot=<0 0 0 1>}
 base_backHair_K {parent=head, loc=<-0.2282 -3.8002 4.975>, rot=<0.0334 0.1278 0.9775 -0.1645>}
 dyn_backHair_K {parent=base_backHair_K, loc=<0 0 0>, rot=<0 0 0 1>}
 base_backHair_J {parent=head, loc=<-0.0568 -5.5087 2.1745>, rot=<0.158 0.1255 0.9646 -0.1696>}
 dyn_backHair_J {parent=base_backHair_J, loc=<0 0 0>, rot=<0 0 0 1>}
 base_backHair_I {parent=head, loc=<10.4705 -5.1729 5.0675>, rot=<0.266 0.2367 0.6162 -0.7025>}
 dyn_backHair_I {parent=base_backHair_I, loc=<0 0 0>, rot=<0 0 0 1>}
 base_backHair_H {parent=head, loc=<12.2935 -5.5207 3.4321>, rot=<0.3281 0.2553 0.5942 -0.6886>}
 dyn_backHair_H {parent=base_backHair_H, loc=<0 0 0>, rot=<0 0 0 1>}
 base_backHair_G {parent=head, loc=<4.6336 -6.765 5.2393>, rot=<-0.0304 0.2246 0.8959 -0.382>}
 dyn_backHair_G {parent=base_backHair_G, loc=<0 0 0>, rot=<0 0 0 1>}
 base_backHair_F {parent=head, loc=<3.0164 -7.7468 0.6495>, rot=<0.1621 0.0707 0.9596 -0.2189>}
 dyn_backHair_F {parent=base_backHair_F, loc=<0 0 0>, rot=<0 0 0 1>}
 base_backHair_E {parent=head, loc=<2.9316 -4.9711 -5.6372>, rot=<0.0575 -0.2912 0.8916 -0.3419>}
 dyn_backHair_E {parent=base_backHair_E, loc=<0 0 0>, rot=<0 0 0 1>}
 base_leftHair_H {parent=head, loc=<11.4542 -4.8775 -5.0794>, rot=<-0.2418 -0.2565 0.6391 -0.6836>}
 dyn_leftHair_H {parent=base_leftHair_H, loc=<0 0 0>, rot=<0 0 0 1>}
 base_backHair_D {parent=head, loc=<5.7706 -9.0643 -1.8943>, rot=<0.1402 -0.0233 0.8761 -0.4607>}
 dyn_backHair_D {parent=base_backHair_D, loc=<0 0 0>, rot=<0 0 0 1>}
 base_backHair_C {parent=head, loc=<10.1406 -7.7741 1.961>, rot=<0.2695 0.2669 0.6705 -0.6377>}
 dyn_backHair_C {parent=base_backHair_C, loc=<0 0 0>, rot=<0 0 0 1>}
 base_backHair_B {parent=head, loc=<7.2526 -8.6683 1.1855>, rot=<0.1379 0.1091 0.7859 -0.5929>}
 dyn_backHair_B {parent=base_backHair_B, loc=<0 0 0>, rot=<0 0 0 1>}
 base_backHair_A {parent=head, loc=<9.4745 -7.8452 -2.6905>, rot=<0.074 -0.0477 0.7758 -0.6248>}
 dyn_backHair_A {parent=base_backHair_A, loc=<0 0 0>, rot=<0 0 0 1>}
 base_leftHair_G {parent=head, loc=<13.3932 -0.5247 -4.4903>, rot=<0.0751 -0.5047 0.1195 -0.8517>}
 dyn_leftHair_G {parent=base_leftHair_G, loc=<0 0 0>, rot=<0 0 0 1>}
 base_leftHair_F {parent=head, loc=<2.2954 -2.6147 -6.0567>, rot=<0.1316 -0.9698 0.0259 -0.2035>}
 dyn_leftHair_F {parent=base_leftHair_F, loc=<0 0 0>, rot=<-0.6687 0 0 -0.7435>}
 base_leftHair_E {parent=head, loc=<9.8549 -1.4005 -6.6981>, rot=<0.3175 -0.8289 0.1783 -0.4246>}
 dyn_leftHair_E {parent=base_leftHair_E, loc=<0 0 0>, rot=<0 0 0 1>}
 base_leftHair_D {parent=head, loc=<6.4926 1.2029 -7.0865>, rot=<-0.1424 -0.4952 0.7766 -0.3625>}
 dyn_leftHair_D {parent=base_leftHair_D, loc=<0 0 0>, rot=<0 0 0 1>}
 base_leftHair_C {parent=head, loc=<9.2593 1.0447 -7.264>, rot=<-0.3695 -0.4234 0.5774 -0.5923>}
 dyn_leftHair_C {parent=base_leftHair_C, loc=<0 0 0>, rot=<0 0 0 1>}
 base_leftHair_B {parent=head, loc=<10.1417 4.9015 -6.4128>, rot=<-0.453 -0.6801 0.4818 -0.3165>}
 dyn_leftHair_B {parent=base_leftHair_B, loc=<0 0 0>, rot=<0 0 0 1>}
 base_rightHair_B {parent=head, loc=<10.0022 6.2588 4.5437>, rot=<-0.1071 -0.73 -0.6417 0.2095>}
 dyn_rightHair_B_01 {parent=base_rightHair_B, loc=<0 0 0>, rot=<0 0 0 1>}
 dyn_rightHair_B_02 {parent=dyn_rightHair_B_01, loc=<4.4974 0 0>, rot=<-0.0007 -0.0302 0.1972 -0.9799>}
 base_rightHair_A {parent=head, loc=<8.0356 2.3381 6.1295>, rot=<-0.0164 0.6731 0.7256 -0.1421>}
 dyn_rightHair_A_01 {parent=base_rightHair_A, loc=<0 0 0>, rot=<0 0 0 1>}
 dyn_rightHair_A_02 {parent=dyn_rightHair_A_01, loc=<6.0154 -0.3861 0.6407>, rot=<-0.0204 0.2116 0.1803 -0.9604>}
 base_earring_L {parent=head, loc=<-0.4524 -0.0958 -5.7985>, rot=<0.0022 -0.5287 0.8488 -0.0035>}
 dyn_earring_L {parent=base_earring_L, loc=<0 0 0>, rot=<0 0 0 1>}
 base_leftHair_A {parent=head, loc=<4.8554 4.4969 -6.5347>, rot=<0.0218 -0.7054 0.7046 -0.0741>}
 dyn_leftHair_A_01 {parent=base_leftHair_A, loc=<0 0 0>, rot=<0 0 0 1>}
 dyn_leftHair_A_02 {parent=dyn_leftHair_A_01, loc=<4.5956 0.1008 0.1365>, rot=<-0.0047 -0.1173 0.0732 -0.9904>}
 base_frontHair_C {parent=head, loc=<9.5245 7.5841 3.761>, rot=<0.1342 -0.5166 0.8347 0.1356>}
 dyn_frontHair_C_01 {parent=base_frontHair_C, loc=<0 0 0>, rot=<0.9977 0 0 -0.0673>}
 dyn_frontHair_C_02 {parent=dyn_frontHair_C_01, loc=<2.9645 0 0>, rot=<0.0298 -0.1996 -0.0026 -0.9794>}
 base_frontHair_B {parent=head, loc=<10.0873 9.474 -0.5041>, rot=<-0.1555 -0.6373 0.6238 -0.4249>}
 dyn_frontHair_B_01 {parent=base_frontHair_B, loc=<0 0 0>, rot=<0 0 0 1>}
 dyn_frontHair_B_02 {parent=dyn_frontHair_B_01, loc=<4.6751 0 -0.0752>, rot=<0 0.0645 0 -0.9979>}
 base_frontHair_A {parent=head, loc=<9.7766 8.3712 1.8975>, rot=<-0.1285 -0.9561 -0.0372 -0.2607>}
 dyn_frontHair_A {parent=base_frontHair_A, loc=<0 0 0>, rot=<0 0 0 1>}
 base_rightHair_F {parent=head, loc=<8.1328 -5.2445 5.7462>, rot=<0.1192 0.2899 0.8608 -0.4009>}
 dyn_rightHair_F {parent=base_rightHair_F, loc=<0 0 0>, rot=<0 0 0 1>}
 base_rightHair_E {parent=head, loc=<10.7467 -2.4088 6.7551>, rot=<0.3936 0.451 0.5535 -0.5791>}
 dyn_rightHair_E {parent=base_rightHair_E, loc=<0 0 0>, rot=<0 0 0 1>}
 base_backHair_M {parent=head, loc=<6.8377 -5.1029 -6.4857>, rot=<-0.0071 -0.2537 0.8349 -0.4884>}
 dyn_backHair_M {parent=base_backHair_M, loc=<0 0 0>, rot=<0 0 0 1>}
 base_backHair_L {parent=head, loc=<3.1864 -7.273 -4.141>, rot=<0.3933 -0.8964 -0.0805 0.1876>}
 dyn_backHair_L {parent=base_backHair_L, loc=<0 0.0169 -0.1631>, rot=<0.997 -0.0053 0.0298 0.0717>}
 base_leftHair_J {parent=head, loc=<0.0665 -4.4521 -4.8703>, rot=<-0.04 -0.5951 0.7932 -0.1232>}
 dyn_leftHair_J {parent=base_leftHair_J, loc=<0 0 0>, rot=<0 0 0 1>}
 base_leftHair_I {parent=head, loc=<-0.0734 -2.6476 -5.6768>, rot=<-0.1737 -0.6793 0.7105 0.0598>}
 dyn_leftHair_I {parent=base_leftHair_I, loc=<0 0 0>, rot=<0 0 0 1>}
 base_earring_R {parent=head, loc=<-0.4524 -0.0943 5.7985>, rot=<-0.0096 -0.2488 0.9685 -0.0015>}
 dyn_earring_R {parent=base_earring_R, loc=<0 0 0>, rot=<0 0 0 1>}
 base_neckCollar {parent=neck_01, loc=<1.9305 0.5839 0>, rot=<0.9948 0.1015 0 0>}
 dyn_neckCollar {parent=base_neckCollar, loc=<0 0 0>, rot=<0 0 0 1>}
 base_lapel_A_l {parent=spine_05, loc=<7.4004 5.5133 -7.2216>, rot=<-0.6326 -0.4269 -0.09 -0.6398>}
 dyn_lapel_A_l_01 {parent=base_lapel_A_l, loc=<0 0 0>, rot=<0 0 0 1>}
 base_lapel_B_l {parent=spine_05, loc=<10.8713 0.9328 -6.9969>, rot=<-0.5513 -0.3044 -0.0194 -0.7766>}
 dyn_lapel_B_l_01 {parent=base_lapel_B_l, loc=<0 0 0>, rot=<0 0 0 1>}
 base_collar_A_l {parent=spine_05, loc=<6.0521 5.5006 -2.7475>, rot=<-0.6365 -0.3584 -0.0408 -0.6818>}
 dyn_collar_A_l_01 {parent=base_collar_A_l, loc=<0 0 0>, rot=<0 0 0 1>}
 base_collar_B_l {parent=spine_05, loc=<10.3733 0.3543 -4.3897>, rot=<-0.6635 -0.2352 -0.041 -0.7091>}
 dyn_collar_B_l_01 {parent=base_collar_B_l, loc=<0 0 0>, rot=<0 0 0 1>}
]
