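# Animation Blueprint for the TwinGinger face accessory mesh, animating hair, earrings and collar.
# It layers left collar and lapel bones from the body mesh onto the parent pose, then runs a hair Control Rig.
# Bone overrides, fed by body socket transforms, and a weighted neck-collar constraint precede chained AnimDynamics, head-space RigidBody and a final Control Rig.
# The event graph caches the body mesh, checks attachment and switches hair simulation space for the front end.
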AnimBlueprint /BRCosmetics/Characters/Player/Female/Medium/Bodies/F_MED_TwinGinger/Meshes/Parts/F_MED_TwinGinger_FaceAcc_AnimBP
target_skeleton=&/F_MED_TwinGinger_FaceAcc_Skeleton
graphs [
 AnimGraph {nodes [
   n1 = Control Rig(Source=Use cached pose 'CRCache_dynamics'()) {
    alphaBoolBlend {blendOption=Linear}
    controlRigAssetReference {blueprintRigClass=&/F_MED_TwinGinger_FaceAcc_CtrlRig.F_MED_TwinGinger_FaceAcc_CtrlRig_C}
   }
   Output Pose(Result=@n1) {layerGroup=None}
   n2 = Layered blend per bone(BasePose=Copy Pose From Mesh() {bUseAttachedParent=true}, BlendPoses_0=Copy Pose From Mesh(SourceMeshComponent=Get Body_Mesh())) {layerSetup [{branchFilters [
       {boneName=base_collar_A_l}
       {boneName=dyn_collar_A_l_01}
       {boneName=base_collar_B_l}
       {boneName=dyn_collar_B_l_01}
       {boneName=base_lapel_A_l}
       {boneName=base_lapel_B_l}
       {boneName=dyn_lapel_B_l_01}
       {boneName=dyn_lapel_A_l_01}
      ]}]}
   n3 = Control Rig(Source=@n2) {
    alphaBoolBlend {blendOption=Linear}
    controlRigAssetReference {blueprintRigClass=&/F_MED_TwinGinger_FaceAcc_hair_CtrlRig.F_MED_TwinGinger_FaceAcc_hair_CtrlRig_C}
   }
   Save cached pose 'CRCache_inputPose'(Pose=@n3) {cacheName=CRCache_inputPose}
   n4 = "Transform (Modify) Bone - Bone: base_collar_A_l"!(
    ComponentPose=Local To Component(LocalPose=Use cached pose 'CRCache_inputPose'())
    bAlphaBoolEnabled:bool=True
    Translation="0,0,0"
    Rotation="0,0,0"
   ) -> (Pose) {
    alphaInputType=Bool
    rotationMode=BMM_Replace
    rotationSpace=BCS_ParentBoneSpace
    translationMode=BMM_Replace
    translationSpace=BCS_ParentBoneSpace
    alphaBoolBlend {blendOption=Linear}
    boneToModify {boneName=base_collar_A_l}
   }
   n5 = "Transform (Modify) Bone - Bone: dyn_collar_A_l_01"!(ComponentPose=@n4, bAlphaBoolEnabled:bool=True, Translation="0,0,0", Rotation="0,0,0") -> (Pose) {
    alphaInputType=Bool
    rotationMode=BMM_Replace
    rotationSpace=BCS_ParentBoneSpace
    translationMode=BMM_Replace
    translationSpace=BCS_ParentBoneSpace
    alphaBoolBlend {blendOption=Linear}
    boneToModify {boneName=dyn_collar_A_l_01}
   }
   n6 = "Transform (Modify) Bone - Bone: base_collar_B_l"!(ComponentPose=@n5, bAlphaBoolEnabled:bool=True, Translation="0,0,0", Rotation="0,0,0") -> (Pose) {
    alphaInputType=Bool
    rotationMode=BMM_Replace
    rotationSpace=BCS_ParentBoneSpace
    translationMode=BMM_Replace
    translationSpace=BCS_ParentBoneSpace
    alphaBoolBlend {blendOption=Linear}
    boneToModify {boneName=base_collar_B_l}
   }
   n7 = "Transform (Modify) Bone - Bone: dyn_collar_B_l_01"!(ComponentPose=@n6, bAlphaBoolEnabled:bool=True, Translation="0,0,0", Rotation="0,0,0") -> (Pose) {
    alphaInputType=Bool
    rotationMode=BMM_Replace
    rotationSpace=BCS_ParentBoneSpace
    translationMode=BMM_Replace
    translationSpace=BCS_ParentBoneSpace
    alphaBoolBlend {blendOption=Linear}
    boneToModify {boneName=dyn_collar_B_l_01}
   }
   n8 = "Transform (Modify) Bone - Bone: base_lapel_A_l"!(ComponentPose=@n7, bAlphaBoolEnabled:bool=True, Translation="0,0,0", Rotation="0,0,0") -> (Pose) {
    alphaInputType=Bool
    rotationMode=BMM_Replace
    rotationSpace=BCS_ParentBoneSpace
    translationMode=BMM_Replace
    translationSpace=BCS_ParentBoneSpace
    alphaBoolBlend {blendOption=Linear}
    boneToModify {boneName=base_lapel_A_l}
   }
   n9 = "Transform (Modify) Bone - Bone: dyn_lapel_A_l_01"!(ComponentPose=@n8, bAlphaBoolEnabled:bool=True, Translation="0,0,0", Rotation="0,0,0") -> (Pose) {
    alphaInputType=Bool
    rotationMode=BMM_Replace
    rotationSpace=BCS_ParentBoneSpace
    translationMode=BMM_Replace
    translationSpace=BCS_ParentBoneSpace
    alphaBoolBlend {blendOption=Linear}
    boneToModify {boneName=dyn_lapel_A_l_01}
   }
   n10 = "Transform (Modify) Bone - Bone: base_lapel_B_l"!(ComponentPose=@n9, bAlphaBoolEnabled:bool=True, Translation="0,0,0", Rotation="0,0,0") -> (Pose) {
    alphaInputType=Bool
    rotationMode=BMM_Replace
    rotationSpace=BCS_ParentBoneSpace
    translationMode=BMM_Replace
    translationSpace=BCS_ParentBoneSpace
    alphaBoolBlend {blendOption=Linear}
    boneToModify {boneName=base_lapel_B_l}
   }
   n11 = "Transform (Modify) Bone - Bone: dyn_lapel_B_l_01"!(ComponentPose=@n10, bAlphaBoolEnabled:bool=True, Translation="0,0,0", Rotation="0,0,0") -> (Pose) {
    alphaInputType=Bool
    rotationMode=BMM_Replace
    rotationSpace=BCS_ParentBoneSpace
    translationMode=BMM_Replace
    translationSpace=BCS_ParentBoneSpace
    alphaBoolBlend {blendOption=Linear}
    boneToModify {boneName=dyn_lapel_B_l_01}
   }
   n12 = "Constraint - base_neckCollar (PPPP)"!(
    ComponentPose=@n11
    Alpha=1
    ConstraintWeights_0=0.3225
    ConstraintWeights_1:float=0.3194
    ConstraintWeights_2:float=0.2347
    ConstraintWeights_3:float=0.1234
   ) -> (Pose) {
    alphaBoolBlend {blendOption=Linear}
    boneToModify {boneName=base_neckCollar}
    constraintSetup [{targetBone {boneName=neck_02}}, {targetBone {boneName=spine_05}}, {targetBone {boneName=head}}, {targetBone {boneName=neck_01}}]
    constraintWeights [1,1,1,1]
   }
   Save cached pose 'CRCache_preDyn'(Pose=Component To Local(ComponentPose=@n12)) {cacheName=CRCache_preDyn}
   n13 = Anim Dynamics - Bone: base_rightHair_A(ComponentPose=Local To Component(LocalPose=Use cached pose 'CRCache_preDyn'()), Alpha=1, ExternalForce="0,0,0") {
    angularDampingOverride=1
    angularSpringConstant=20000
    bAngularSpring=true
    bOverrideAngularDamping=true
    bUseGravityOverride=true
    gravityScale=0.35
    simulationSpace=RootRelative
    alphaBoolBlend {blendOption=Linear}
    boundBone {boneName=base_rightHair_A}
    componentLinearAccScale=<1 1 1>
    componentLinearVelScale=<1 1 1>
    gravityOverride {z=-980}
    physicsBodyDefinitions [{
      boundBone {boneName=base_rightHair_A}
      boxExtents=<20 30 30>
      constraintSetup {angularLimitsMax=<15 15 15>, angularLimitsMin=<-1 -1 -5>, angularTarget=<1 0 0>}
      localJointOffset=<4 0 0>
     }]
   }
   n14 = Anim Dynamics - Bone: base_rightHair_B!(SimulationSpace:AnimPhysSimSpaceType Enum, ComponentPose=@n13, Alpha=1, ExternalForce="0,0,0") -> (Pose) {
    angularDampingOverride=1
    angularSpringConstant=20000
    bAngularSpring=true
    bOverrideAngularDamping=true
    bUseGravityOverride=true
    gravityScale=0.35
    simulationSpace=RootRelative
    alphaBoolBlend {blendOption=Linear}
    boundBone {boneName=base_rightHair_B}
    componentLinearAccScale=<1 1 1>
    componentLinearVelScale=<1 1 1>
    gravityOverride {z=-980}
    physicsBodyDefinitions [{
      boundBone {boneName=base_rightHair_B}
      boxExtents=<20 30 30>
      constraintSetup {angularLimitsMax=<5 5 10>, angularLimitsMin=<-1 -1 5>, angularTarget=<1 0 0>}
      localJointOffset=<4 0 0>
     }]
   }
   n15 = Anim Dynamics - Bone: base_rightHair_C!(SimulationSpace:AnimPhysSimSpaceType Enum, ComponentPose=@n14, Alpha=1, ExternalForce="0,0,0") -> (Pose) {
    angularDampingOverride=1
    angularSpringConstant=20000
    bAngularSpring=true
    bOverrideAngularDamping=true
    bUseGravityOverride=true
    gravityScale=0.35
    simulationSpace=RootRelative
    alphaBoolBlend {blendOption=Linear}
    boundBone {boneName=base_rightHair_C}
    componentLinearAccScale=<1 1 1>
    componentLinearVelScale=<1 1 1>
    gravityOverride {z=-980}
    physicsBodyDefinitions [{
      boundBone {boneName=base_rightHair_C}
      boxExtents=<20 30 30>
      constraintSetup {angularLimitsMax=<15 15 15>, angularLimitsMin=<-15 -15 -15>, angularTarget=<1 0 0>}
      localJointOffset=<4 0 0>
     }]
   }
   n16 = Anim Dynamics - Bone: base_rightHair_D!(SimulationSpace:AnimPhysSimSpaceType Enum, ComponentPose=@n15, Alpha=1, ExternalForce="0,0,0") -> (Pose) {
    angularDampingOverride=1
    angularSpringConstant=20000
    bAngularSpring=true
    bOverrideAngularDamping=true
    bUseGravityOverride=true
    gravityScale=0.35
    simulationSpace=RootRelative
    alphaBoolBlend {blendOption=Linear}
    boundBone {boneName=base_rightHair_D}
    componentLinearAccScale=<1 1 1>
    componentLinearVelScale=<1 1 1>
    gravityOverride {z=-980}
    physicsBodyDefinitions [{
      boundBone {boneName=base_rightHair_D}
      boxExtents=<20 30 30>
      constraintSetup {angularLimitsMax=<15 15 15>, angularLimitsMin=<-15 -15 -15>, angularTarget=<1 0 0>}
      localJointOffset=<4 0 0>
     }]
   }
   n17 = Anim Dynamics - Bone: base_rightHair_E!(SimulationSpace:AnimPhysSimSpaceType Enum, ComponentPose=@n16, Alpha=1, ExternalForce="0,0,0") -> (Pose) {
    angularDampingOverride=1
    angularSpringConstant=20000
    bAngularSpring=true
    bOverrideAngularDamping=true
    bUseGravityOverride=true
    gravityScale=0.35
    simulationSpace=RootRelative
    alphaBoolBlend {blendOption=Linear}
    boundBone {boneName=base_rightHair_E}
    componentLinearAccScale=<1 1 1>
    componentLinearVelScale=<1 1 1>
    gravityOverride {z=-980}
    physicsBodyDefinitions [{
      boundBone {boneName=base_rightHair_E}
      boxExtents=<20 30 30>
      constraintSetup {angularLimitsMax=<15 15 15>, angularLimitsMin=<-15 -15 -15>, angularTarget=<1 0 0>}
      localJointOffset=<4 0 0>
     }]
   }
   n18 = Anim Dynamics - Bone: base_rightHair_F!(SimulationSpace:AnimPhysSimSpaceType Enum, ComponentPose=@n17, Alpha=1, ExternalForce="0,0,0") -> (Pose) {
    angularDampingOverride=1
    angularSpringConstant=20000
    bAngularSpring=true
    bOverrideAngularDamping=true
    bUseGravityOverride=true
    gravityScale=0.35
    simulationSpace=RootRelative
    alphaBoolBlend {blendOption=Linear}
    boundBone {boneName=base_rightHair_F}
    componentLinearAccScale=<1 1 1>
    componentLinearVelScale=<1 1 1>
    gravityOverride {z=-980}
    physicsBodyDefinitions [{
      boundBone {boneName=base_rightHair_F}
      boxExtents=<20 30 30>
      constraintSetup {angularLimitsMax=<15 15 15>, angularLimitsMin=<-15 -15 -15>, angularTarget=<1 0 0>}
      localJointOffset=<4 0 0>
     }]
   }
   n19 = Anim Dynamics - Bone: base_leftHair_B!(SimulationSpace:AnimPhysSimSpaceType Enum=Component, ComponentPose=@n18, Alpha=1, ExternalForce="0,0,0") -> (Pose) {
    angularDampingOverride=1
    angularSpringConstant=20000
    bAngularSpring=true
    bOverrideAngularDamping=true
    bUseGravityOverride=true
    gravityScale=0.35
    simulationSpace=RootRelative
    alphaBoolBlend {blendOption=Linear}
    boundBone {boneName=base_leftHair_B}
    componentLinearAccScale=<1 1 1>
    componentLinearVelScale=<1 1 1>
    gravityOverride {z=-980}
    physicsBodyDefinitions [{
      boundBone {boneName=base_leftHair_B}
      boxExtents=<20 30 30>
      constraintSetup {angularLimitsMax=<15 15 15>, angularLimitsMin=<-15 -15 -15>, angularTarget=<1 0 0>}
      localJointOffset=<4 0 0>
     }]
   }
   n20 = Anim Dynamics - Bone: base_leftHair_C!(SimulationSpace:AnimPhysSimSpaceType Enum, ComponentPose=@n19, Alpha=1, ExternalForce="0,0,0") -> (Pose) {
    angularDampingOverride=1
    angularSpringConstant=20000
    bAngularSpring=true
    bOverrideAngularDamping=true
    bUseGravityOverride=true
    gravityScale=0.35
    simulationSpace=RootRelative
    alphaBoolBlend {blendOption=Linear}
    boundBone {boneName=base_leftHair_C}
    componentLinearAccScale=<1 1 1>
    componentLinearVelScale=<1 1 1>
    gravityOverride {z=-980}
    physicsBodyDefinitions [{
      boundBone {boneName=base_leftHair_C}
      boxExtents=<20 30 30>
      constraintSetup {angularLimitsMax=<15 15 15>, angularLimitsMin=<-15 -15 -15>, angularTarget=<1 0 0>}
      localJointOffset=<4 0 0>
     }]
   }
   n21 = Anim Dynamics - Bone: base_leftHair_D!(SimulationSpace:AnimPhysSimSpaceType Enum, ComponentPose=@n20, Alpha=1, ExternalForce="0,0,0") -> (Pose) {
    angularDampingOverride=1
    angularSpringConstant=20000
    bAngularSpring=true
    bOverrideAngularDamping=true
    bUseGravityOverride=true
    gravityScale=0.35
    simulationSpace=RootRelative
    alphaBoolBlend {blendOption=Linear}
    boundBone {boneName=base_leftHair_D}
    componentLinearAccScale=<1 1 1>
    componentLinearVelScale=<1 1 1>
    gravityOverride {z=-980}
    physicsBodyDefinitions [{
      boundBone {boneName=base_leftHair_D}
      boxExtents=<20 30 30>
      constraintSetup {angularLimitsMax=<15 15 15>, angularLimitsMin=<-15 -15 -15>, angularTarget=<1 0 0>}
      localJointOffset=<4 0 0>
     }]
   }
   n22 = Anim Dynamics - Bone: base_leftHair_E!(SimulationSpace:AnimPhysSimSpaceType Enum, ComponentPose=@n21, Alpha=1, ExternalForce="0,0,0") -> (Pose) {
    angularDampingOverride=1
    angularSpringConstant=20000
    bAngularSpring=true
    bOverrideAngularDamping=true
    bUseGravityOverride=true
    gravityScale=0.35
    simulationSpace=RootRelative
    alphaBoolBlend {blendOption=Linear}
    boundBone {boneName=base_leftHair_E}
    componentLinearAccScale=<1 1 1>
    componentLinearVelScale=<1 1 1>
    gravityOverride {z=-980}
    physicsBodyDefinitions [{
      boundBone {boneName=base_leftHair_E}
      boxExtents=<20 30 30>
      constraintSetup {angularLimitsMax=<15 15 15>, angularLimitsMin=<-15 -15 -15>, angularTarget=<1 0 0>}
      localJointOffset=<4 0 0>
     }]
   }
   n23 = Anim Dynamics - Bone: base_leftHair_F!(SimulationSpace:AnimPhysSimSpaceType Enum, ComponentPose=@n22, Alpha=1, ExternalForce="0,0,0") -> (Pose) {
    angularDampingOverride=1
    angularSpringConstant=20000
    bAngularSpring=true
    bOverrideAngularDamping=true
    bUseGravityOverride=true
    gravityScale=0.35
    simulationSpace=RootRelative
    alphaBoolBlend {blendOption=Linear}
    boundBone {boneName=base_leftHair_F}
    componentLinearAccScale=<1 1 1>
    componentLinearVelScale=<1 1 1>
    gravityOverride {z=-980}
    physicsBodyDefinitions [{
      boundBone {boneName=base_leftHair_F}
      boxExtents=<20 30 30>
      constraintSetup {angularLimitsMax=<15 15 15>, angularLimitsMin=<-15 -15 -15>, angularTarget=<1 0 0>}
      localJointOffset=<4 0 0>
     }]
   }
   n24 = Anim Dynamics - Bone: base_leftHair_G!(SimulationSpace:AnimPhysSimSpaceType Enum, ComponentPose=@n23, Alpha=1, ExternalForce="0,0,0") -> (Pose) {
    angularDampingOverride=1
    angularSpringConstant=20000
    bAngularSpring=true
    bOverrideAngularDamping=true
    bUseGravityOverride=true
    gravityScale=0.35
    simulationSpace=RootRelative
    alphaBoolBlend {blendOption=Linear}
    boundBone {boneName=base_leftHair_G}
    componentLinearAccScale=<1 1 1>
    componentLinearVelScale=<1 1 1>
    gravityOverride {z=-980}
    physicsBodyDefinitions [{
      boundBone {boneName=base_leftHair_G}
      boxExtents=<20 30 30>
      constraintSetup {angularLimitsMax=<15 15 15>, angularLimitsMin=<-15 -15 -15>, angularTarget=<1 0 0>}
      localJointOffset=<4 0 0>
     }]
   }
   n25 = Anim Dynamics - Bone: base_leftHair_H!(SimulationSpace:AnimPhysSimSpaceType Enum, ComponentPose=@n24, Alpha=1, ExternalForce="0,0,0") -> (Pose) {
    angularDampingOverride=1
    angularSpringConstant=20000
    bAngularSpring=true
    bOverrideAngularDamping=true
    bUseGravityOverride=true
    gravityScale=0.35
    simulationSpace=RootRelative
    alphaBoolBlend {blendOption=Linear}
    boundBone {boneName=base_leftHair_H}
    componentLinearAccScale=<1 1 1>
    componentLinearVelScale=<1 1 1>
    gravityOverride {z=-980}
    physicsBodyDefinitions [{
      boundBone {boneName=base_leftHair_H}
      boxExtents=<20 30 30>
      constraintSetup {angularLimitsMax=<15 15 15>, angularLimitsMin=<-15 -15 -15>, angularTarget=<1 0 0>}
      localJointOffset=<4 0 0>
     }]
   }
   n26 = Anim Dynamics - Bone: base_leftHair_H(ComponentPose=@n25, Alpha=1, ExternalForce="0,0,0") {
    angularDampingOverride=1
    angularSpringConstant=20000
    bAngularSpring=true
    bOverrideAngularDamping=true
    bUseGravityOverride=true
    gravityScale=0.35
    simulationSpace=RootRelative
    alphaBoolBlend {blendOption=Linear}
    boundBone {boneName=base_leftHair_H}
    componentLinearAccScale=<1 1 1>
    componentLinearVelScale=<1 1 1>
    gravityOverride {z=-980}
    physicsBodyDefinitions [{
      boundBone {boneName=base_leftHair_H}
      boxExtents=<20 30 30>
      constraintSetup {angularLimitsMax=<15 15 15>, angularLimitsMin=<-15 -15 -15>, angularTarget=<1 0 0>}
      localJointOffset=<4 0 0>
     }]
   }
   n27 = Anim Dynamics - Bone: base_leftHair_J!(SimulationSpace:AnimPhysSimSpaceType Enum, ComponentPose=@n26, Alpha=1, ExternalForce="0,0,0") -> (Pose) {
    angularDampingOverride=1
    angularSpringConstant=20000
    bAngularSpring=true
    bOverrideAngularDamping=true
    bUseGravityOverride=true
    gravityScale=0.35
    simulationSpace=RootRelative
    alphaBoolBlend {blendOption=Linear}
    boundBone {boneName=base_leftHair_J}
    componentLinearAccScale=<1 1 1>
    componentLinearVelScale=<1 1 1>
    gravityOverride {z=-980}
    physicsBodyDefinitions [{
      boundBone {boneName=base_leftHair_J}
      boxExtents=<20 30 30>
      constraintSetup {angularLimitsMax=<15 15 15>, angularLimitsMin=<-15 -15 -15>, angularTarget=<1 0 0>}
      localJointOffset=<4 0 0>
     }]
   }
   n28 = Anim Dynamics - Bone: base_backHair_A!(SimulationSpace:AnimPhysSimSpaceType Enum, ComponentPose=@n27, Alpha=1, ExternalForce="0,0,0") -> (Pose) {
    angularDampingOverride=1
    angularSpringConstant=20000
    bAngularSpring=true
    bOverrideAngularDamping=true
    bUseGravityOverride=true
    gravityScale=0.35
    simulationSpace=RootRelative
    alphaBoolBlend {blendOption=Linear}
    boundBone {boneName=base_backHair_A}
    componentLinearAccScale=<1 1 1>
    componentLinearVelScale=<1 1 1>
    gravityOverride {z=-980}
    physicsBodyDefinitions [{
      boundBone {boneName=base_backHair_A}
      boxExtents=<20 30 30>
      constraintSetup {angularLimitsMax=<15 15 15>, angularLimitsMin=<-15 -15 -15>, angularTarget=<1 0 0>}
      localJointOffset=<4 0 0>
     }]
   }
   n29 = Anim Dynamics - Bone: base_backHair_B!(SimulationSpace:AnimPhysSimSpaceType Enum, ComponentPose=@n28, Alpha=1, ExternalForce="0,0,0") -> (Pose) {
    angularDampingOverride=1
    angularSpringConstant=20000
    bAngularSpring=true
    bOverrideAngularDamping=true
    bUseGravityOverride=true
    gravityScale=0.35
    simulationSpace=RootRelative
    alphaBoolBlend {blendOption=Linear}
    boundBone {boneName=base_backHair_B}
    componentLinearAccScale=<1 1 1>
    componentLinearVelScale=<1 1 1>
    gravityOverride {z=-980}
    physicsBodyDefinitions [{
      boundBone {boneName=base_backHair_B}
      boxExtents=<20 30 30>
      constraintSetup {angularLimitsMax=<15 15 15>, angularLimitsMin=<-15 -15 -15>, angularTarget=<1 0 0>}
      localJointOffset=<4 0 0>
     }]
   }
   n30 = Anim Dynamics - Bone: base_backHair_C!(SimulationSpace:AnimPhysSimSpaceType Enum, ComponentPose=@n29, Alpha=1, ExternalForce="0,0,0") -> (Pose) {
    angularDampingOverride=1
    angularSpringConstant=20000
    bAngularSpring=true
    bOverrideAngularDamping=true
    bUseGravityOverride=true
    gravityScale=0.35
    simulationSpace=RootRelative
    alphaBoolBlend {blendOption=Linear}
    boundBone {boneName=base_backHair_C}
    componentLinearAccScale=<1 1 1>
    componentLinearVelScale=<1 1 1>
    gravityOverride {z=-980}
    physicsBodyDefinitions [{
      boundBone {boneName=base_backHair_C}
      boxExtents=<20 30 30>
      constraintSetup {angularLimitsMax=<15 15 15>, angularLimitsMin=<-15 -15 -15>, angularTarget=<1 0 0>}
      localJointOffset=<4 0 0>
     }]
   }
   n31 = Anim Dynamics - Bone: base_backHair_D!(SimulationSpace:AnimPhysSimSpaceType Enum, ComponentPose=@n30, Alpha=1, ExternalForce="0,0,0") -> (Pose) {
    angularDampingOverride=1
    angularSpringConstant=20000
    bAngularSpring=true
    bOverrideAngularDamping=true
    bUseGravityOverride=true
    gravityScale=0.35
    simulationSpace=RootRelative
    alphaBoolBlend {blendOption=Linear}
    boundBone {boneName=base_backHair_D}
    componentLinearAccScale=<1 1 1>
    componentLinearVelScale=<1 1 1>
    gravityOverride {z=-980}
    physicsBodyDefinitions [{
      boundBone {boneName=base_backHair_D}
      boxExtents=<20 30 30>
      constraintSetup {angularLimitsMax=<15 15 15>, angularLimitsMin=<-15 -15 -15>, angularTarget=<1 0 0>}
      localJointOffset=<4 0 0>
     }]
   }
   n32 = Anim Dynamics - Bone: base_backHair_E!(SimulationSpace:AnimPhysSimSpaceType Enum, ComponentPose=@n31, Alpha=1, ExternalForce="0,0,0") -> (Pose) {
    angularDampingOverride=1
    angularSpringConstant=20000
    bAngularSpring=true
    bOverrideAngularDamping=true
    bUseGravityOverride=true
    gravityScale=0.35
    simulationSpace=RootRelative
    alphaBoolBlend {blendOption=Linear}
    boundBone {boneName=base_backHair_E}
    componentLinearAccScale=<1 1 1>
    componentLinearVelScale=<1 1 1>
    gravityOverride {z=-980}
    physicsBodyDefinitions [{
      boundBone {boneName=base_backHair_E}
      boxExtents=<20 30 30>
      constraintSetup {angularLimitsMax=<15 15 15>, angularLimitsMin=<-15 -15 -15>, angularTarget=<1 0 0>}
      localJointOffset=<4 0 0>
     }]
   }
   n33 = Anim Dynamics - Bone: base_backHair_F!(SimulationSpace:AnimPhysSimSpaceType Enum, ComponentPose=@n32, Alpha=1, ExternalForce="0,0,0") -> (Pose) {
    angularDampingOverride=1
    angularSpringConstant=20000
    bAngularSpring=true
    bOverrideAngularDamping=true
    bUseGravityOverride=true
    gravityScale=0.35
    simulationSpace=RootRelative
    alphaBoolBlend {blendOption=Linear}
    boundBone {boneName=base_backHair_F}
    componentLinearAccScale=<1 1 1>
    componentLinearVelScale=<1 1 1>
    gravityOverride {z=-980}
    physicsBodyDefinitions [{
      boundBone {boneName=base_backHair_F}
      boxExtents=<20 30 30>
      constraintSetup {angularLimitsMax=<15 15 15>, angularLimitsMin=<-15 -15 -15>, angularTarget=<1 0 0>}
      localJointOffset=<4 0 0>
     }]
   }
   n34 = Anim Dynamics - Bone: base_backHair_G!(SimulationSpace:AnimPhysSimSpaceType Enum, ComponentPose=@n33, Alpha=1, ExternalForce="0,0,0") -> (Pose) {
    angularDampingOverride=1
    angularSpringConstant=20000
    bAngularSpring=true
    bOverrideAngularDamping=true
    bUseGravityOverride=true
    gravityScale=0.35
    simulationSpace=RootRelative
    alphaBoolBlend {blendOption=Linear}
    boundBone {boneName=base_backHair_G}
    componentLinearAccScale=<1 1 1>
    componentLinearVelScale=<1 1 1>
    gravityOverride {z=-980}
    physicsBodyDefinitions [{
      boundBone {boneName=base_backHair_G}
      boxExtents=<20 30 30>
      constraintSetup {angularLimitsMax=<15 15 15>, angularLimitsMin=<-15 -15 -15>, angularTarget=<1 0 0>}
      localJointOffset=<4 0 0>
     }]
   }
   n35 = Anim Dynamics - Bone: base_backHair_H!(SimulationSpace:AnimPhysSimSpaceType Enum, ComponentPose=@n34, Alpha=1, ExternalForce="0,0,0") -> (Pose) {
    angularDampingOverride=1
    angularSpringConstant=20000
    bAngularSpring=true
    bOverrideAngularDamping=true
    bUseGravityOverride=true
    gravityScale=0.35
    simulationSpace=RootRelative
    alphaBoolBlend {blendOption=Linear}
    boundBone {boneName=base_backHair_H}
    componentLinearAccScale=<1 1 1>
    componentLinearVelScale=<1 1 1>
    gravityOverride {z=-980}
    physicsBodyDefinitions [{
      boundBone {boneName=base_backHair_H}
      boxExtents=<20 30 30>
      constraintSetup {angularLimitsMax=<15 15 15>, angularLimitsMin=<-15 -15 -15>, angularTarget=<1 0 0>}
      localJointOffset=<4 0 0>
     }]
   }
   n36 = Anim Dynamics - Bone: base_backHair_I!(SimulationSpace:AnimPhysSimSpaceType Enum, ComponentPose=@n35, Alpha=1, ExternalForce="0,0,0") -> (Pose) {
    angularDampingOverride=1
    angularSpringConstant=20000
    bAngularSpring=true
    bOverrideAngularDamping=true
    bUseGravityOverride=true
    gravityScale=0.35
    simulationSpace=RootRelative
    alphaBoolBlend {blendOption=Linear}
    boundBone {boneName=base_backHair_I}
    componentLinearAccScale=<1 1 1>
    componentLinearVelScale=<1 1 1>
    gravityOverride {z=-980}
    physicsBodyDefinitions [{
      boundBone {boneName=base_backHair_I}
      boxExtents=<20 30 30>
      constraintSetup {angularLimitsMax=<15 15 15>, angularLimitsMin=<-15 -15 -15>, angularTarget=<1 0 0>}
      localJointOffset=<4 0 0>
     }]
   }
   n37 = Anim Dynamics - Bone: base_backHair_J!(SimulationSpace:AnimPhysSimSpaceType Enum, ComponentPose=@n36, Alpha=1, ExternalForce="0,0,0") -> (Pose) {
    angularDampingOverride=1
    angularSpringConstant=20000
    bAngularSpring=true
    bOverrideAngularDamping=true
    bUseGravityOverride=true
    gravityScale=0.35
    simulationSpace=RootRelative
    alphaBoolBlend {blendOption=Linear}
    boundBone {boneName=base_backHair_J}
    componentLinearAccScale=<1 1 1>
    componentLinearVelScale=<1 1 1>
    gravityOverride {z=-980}
    physicsBodyDefinitions [{
      boundBone {boneName=base_backHair_J}
      boxExtents=<20 30 30>
      constraintSetup {angularLimitsMax=<15 15 15>, angularLimitsMin=<-15 -15 -15>, angularTarget=<1 0 0>}
      localJointOffset=<4 0 0>
     }]
   }
   n38 = Anim Dynamics - Bone: base_backHair_K!(SimulationSpace:AnimPhysSimSpaceType Enum, ComponentPose=@n37, Alpha=1, ExternalForce="0,0,0") -> (Pose) {
    angularDampingOverride=1
    angularSpringConstant=20000
    bAngularSpring=true
    bOverrideAngularDamping=true
    bUseGravityOverride=true
    gravityScale=0.35
    simulationSpace=RootRelative
    alphaBoolBlend {blendOption=Linear}
    boundBone {boneName=base_backHair_K}
    componentLinearAccScale=<1 1 1>
    componentLinearVelScale=<1 1 1>
    gravityOverride {z=-980}
    physicsBodyDefinitions [{
      boundBone {boneName=base_backHair_K}
      boxExtents=<20 30 30>
      constraintSetup {angularLimitsMax=<15 15 15>, angularLimitsMin=<-15 -15 -15>, angularTarget=<1 0 0>}
      localJointOffset=<4 0 0>
     }]
   }
   n39 = Anim Dynamics - Bone: base_backHair_L!(SimulationSpace:AnimPhysSimSpaceType Enum, ComponentPose=@n38, Alpha=1, ExternalForce="0,0,0") -> (Pose) {
    angularDampingOverride=1
    angularSpringConstant=20000
    bAngularSpring=true
    bOverrideAngularDamping=true
    bUseGravityOverride=true
    gravityScale=0.35
    simulationSpace=RootRelative
    alphaBoolBlend {blendOption=Linear}
    boundBone {boneName=base_backHair_L}
    componentLinearAccScale=<1 1 1>
    componentLinearVelScale=<1 1 1>
    gravityOverride {z=-980}
    physicsBodyDefinitions [{
      boundBone {boneName=base_backHair_L}
      boxExtents=<20 30 30>
      constraintSetup {angularLimitsMax=<15 15 15>, angularLimitsMin=<-15 -15 -15>, angularTarget=<1 0 0>}
      localJointOffset=<4 0 0>
     }]
   }
   n40 = Anim Dynamics - Bone: base_backHair_M!(SimulationSpace:AnimPhysSimSpaceType Enum, ComponentPose=@n39, Alpha=1, ExternalForce="0,0,0") -> (Pose) {
    angularDampingOverride=1
    angularSpringConstant=20000
    bAngularSpring=true
    bOverrideAngularDamping=true
    bUseGravityOverride=true
    gravityScale=0.35
    simulationSpace=RootRelative
    alphaBoolBlend {blendOption=Linear}
    boundBone {boneName=base_backHair_M}
    componentLinearAccScale=<1 1 1>
    componentLinearVelScale=<1 1 1>
    gravityOverride {z=-980}
    physicsBodyDefinitions [{
      boundBone {boneName=base_backHair_M}
      boxExtents=<20 30 30>
      constraintSetup {angularLimitsMax=<15 15 15>, angularLimitsMin=<-15 -15 -15>, angularTarget=<1 0 0>}
      localJointOffset=<4 0 0>
     }]
   }
   n41 = Anim Dynamics - Bone: base_frontHair_A!(SimulationSpace:AnimPhysSimSpaceType Enum, ComponentPose=@n40, Alpha=1, ExternalForce="0,0,0") -> (Pose) {
    angularDampingOverride=1
    angularSpringConstant=20000
    bAngularSpring=true
    bOverrideAngularDamping=true
    bUseGravityOverride=true
    gravityScale=0.35
    simulationSpace=RootRelative
    alphaBoolBlend {blendOption=Linear}
    boundBone {boneName=base_frontHair_A}
    componentLinearAccScale=<1 1 1>
    componentLinearVelScale=<1 1 1>
    gravityOverride {z=-980}
    physicsBodyDefinitions [{
      boundBone {boneName=base_frontHair_A}
      boxExtents=<20 30 30>
      constraintSetup {angularLimitsMax=<15 15 15>, angularLimitsMin=<-15 -15 -15>, angularTarget=<1 0 0>}
      localJointOffset=<4 0 0>
     }]
   }
   n42 = Anim Dynamics - Bone: base_frontHair_B!(SimulationSpace:AnimPhysSimSpaceType Enum, ComponentPose=@n41, Alpha=1, ExternalForce="0,0,0") -> (Pose) {
    angularDampingOverride=1
    angularSpringConstant=10000
    bAngularSpring=true
    bOverrideAngularDamping=true
    bUseGravityOverride=true
    gravityScale=0.35
    simulationSpace=RootRelative
    alphaBoolBlend {blendOption=Linear}
    boundBone {boneName=base_frontHair_B}
    componentLinearAccScale=<1 1 1>
    componentLinearVelScale=<1 1 1>
    gravityOverride {z=-980}
    physicsBodyDefinitions [{
      boundBone {boneName=base_frontHair_B}
      boxExtents=<20 30 30>
      constraintSetup {angularLimitsMax=<2 1 5>, angularLimitsMin=<-2 -1 -0.1>, angularTarget=<1 0 0>}
      localJointOffset=<4 0 0>
     }]
   }
   n43 = Anim Dynamics - Bone: base_frontHair_C!(SimulationSpace:AnimPhysSimSpaceType Enum=BoneRelative, ComponentPose=@n42, Alpha=1, ExternalForce="0,0,0") -> (Pose) {
    angularDampingOverride=1
    angularSpringConstant=20000
    bAngularSpring=true
    bOverrideAngularDamping=true
    bUseGravityOverride=true
    gravityScale=0.35
    simulationSpace=BoneRelative
    alphaBoolBlend {blendOption=Linear}
    boundBone {boneName=base_frontHair_C}
    componentLinearAccScale=<1 1 1>
    componentLinearVelScale=<1 1 1>
    gravityOverride {z=-980}
    physicsBodyDefinitions [{
      boundBone {boneName=base_frontHair_C}
      boxExtents=<20 30 30>
      constraintSetup {angularLimitsMax=<5 5 0.1>, angularLimitsMin=<-1 -1 -5>, angularTarget=<1 0 0>}
      localJointOffset=<4 0 0>
     }]
   }
   n44 = Anim Dynamics - Bone: base_earring_L!(SimulationSpace:AnimPhysSimSpaceType Enum, ComponentPose=@n43, Alpha=1, ExternalForce="0,0,0") -> (Pose) {
    angularDampingOverride=1
    angularSpringConstant=20000
    bAngularSpring=true
    bOverrideAngularDamping=true
    bUseGravityOverride=true
    gravityScale=0.35
    simulationSpace=RootRelative
    alphaBoolBlend {blendOption=Linear}
    boundBone {boneName=base_earring_L}
    componentLinearAccScale=<1 1 1>
    componentLinearVelScale=<1 1 1>
    gravityOverride {z=-980}
    physicsBodyDefinitions [{
      boundBone {boneName=base_earring_L}
      boxExtents=<20 30 30>
      constraintSetup {angularLimitsMax=<15 15 15>, angularLimitsMin=<-15 -15 -15>, angularTarget=<1 0 0>}
      localJointOffset=<4 0 0>
     }]
   }
   n45 = Anim Dynamics - Bone: base_earring_R!(SimulationSpace:AnimPhysSimSpaceType Enum, ComponentPose=@n44, Alpha=1, ExternalForce="0,0,0") -> (Pose) {
    angularDampingOverride=1
    angularSpringConstant=20000
    bAngularSpring=true
    bOverrideAngularDamping=true
    bUseGravityOverride=true
    gravityScale=0.35
    simulationSpace=RootRelative
    alphaBoolBlend {blendOption=Linear}
    boundBone {boneName=base_earring_R}
    componentLinearAccScale=<1 1 1>
    componentLinearVelScale=<1 1 1>
    gravityOverride {z=-980}
    physicsBodyDefinitions [{
      boundBone {boneName=base_earring_R}
      boxExtents=<20 30 30>
      constraintSetup {angularLimitsMax=<15 15 15>, angularLimitsMin=<-15 -15 -15>, angularTarget=<1 0 0>}
      localJointOffset=<4 0 0>
     }]
   }
   n46 = RigidBody(ComponentPose=@n45) {
    overridePhysicsAsset=&/F_MED_TwinGinger_FaceAcc_Physics
    simulationSpace=BaseBoneSpace
    alphaBoolBlend {blendOption=Linear}
    baseBoneRef {boneName=head}
    componentAppliedLinearAccClamp=<3000 3000 3000>
    simSpaceSettings {
     dampingAlpha=0.2
     maxAngularAcceleration=1
     maxAngularVelocity=1
     maxLinearAcceleration=200
     maxLinearVelocity=100
     velocityScaleZ=0.5
     worldAlpha=0.5
     externalLinearDragV=<0.1 0.1 0.1>
    }
   }
   Save cached pose 'CRCache_dynamics'(Pose=Component To Local(ComponentPose=@n46)) {cacheName=CRCache_dynamics}
  ]}
 EventGraph {nodes [
   Event BlueprintUpdateAnimation()
   > n1 = Sequence()
   n2 = Get Body_Mesh()
   > Is Valid(InputObject=@n2)
   n3 = GetSocketBoneName!(self:Skinned Mesh Component Object Reference=@n2, InSocketName:Name=upperarm_r) -> (ReturnValue:Name=None)
   > n4 = Set isBodyAttached(isBodyAttached:bool="Not Equal Exactly (String)"(A="To String (Name)"(InName=@n3), B=None)) -> (Output_Get:bool)
   > Branch(Condition=@n4)
   > K2Node_Composite_0 = Joints_from_body!(execute) -> (then:Exec) {is=title K2Node_Composite}
   n5 = Branch(execute=@n1.then_1, Condition=Get bIsInFrontEnd())
   > Set Hair Simulation Space(Hair Simulation Space:AnimPhysSimSpaceType Enum=World) -> (Output_Get:AnimPhysSimSpaceType Enum)
   Set Hair Simulation Space(execute=@n5.else)
   n6 = Event OnCharacterCustomizationLoaded()
   > Set Body_Mesh(Body_Mesh=Get Array Item with EFort Custom Part Type(Array=@n6.CustomMeshParts))
  ]}
 Joints_from_body {
  path=EventGraph.K2Node_Composite_0.Joints_from_body
  nodes [
   Inputs()
   n1 = BreakTransform(InTransform=GetSocketTransform(self=Get Body_Mesh(), InSocketName=base_collar_A_l, TransformSpace=RTS_ParentBoneSpace))
   > Set base_collar_A_l_Translation(base_collar_A_l_Translation:Vector=@n1) -> (Output_Get:Vector)
   > Set base_collar_A_l_Rotation(base_collar_A_l_Rotation:Rotator=@n1.Rotation) -> (Output_Get:Rotator)
   n2 = BreakTransform(InTransform=GetSocketTransform(self=Get Body_Mesh(), InSocketName=dyn_collar_A_l_01, TransformSpace=RTS_ParentBoneSpace))
   > Set dyn_collar_A_l_01_Translation(dyn_collar_A_l_01_Translation:Vector=@n2) -> (Output_Get:Vector)
   > Set dyn_collar_A_l_01_Rotation(dyn_collar_A_l_01_Rotation:Rotator=@n2.Rotation) -> (Output_Get:Rotator)
   n3 = BreakTransform(InTransform=GetSocketTransform(self=Get Body_Mesh(), InSocketName=base_collar_B_l, TransformSpace=RTS_ParentBoneSpace))
   > Set base_collar_B_l_Translation(base_collar_B_l_Translation:Vector=@n3) -> (Output_Get:Vector)
   > Set base_collar_B_l_Rotation(base_collar_B_l_Rotation:Rotator=@n3.Rotation) -> (Output_Get:Rotator)
   n4 = BreakTransform(InTransform=GetSocketTransform(self=Get Body_Mesh(), InSocketName=dyn_collar_B_l_01, TransformSpace=RTS_ParentBoneSpace))
   > Set dyn_collar_B_l_01_Translation(dyn_collar_B_l_01_Translation:Vector=@n4) -> (Output_Get:Vector)
   > Set dyn_collar_B_l_01_Rotation(dyn_collar_B_l_01_Rotation:Rotator=@n4.Rotation) -> (Output_Get:Rotator)
   n5 = BreakTransform(InTransform=GetSocketTransform(self=Get Body_Mesh(), InSocketName=base_lapel_A_l, TransformSpace=RTS_ParentBoneSpace))
   > Set base_lapel_A_l_Translation(base_lapel_A_l_Translation:Vector=@n5) -> (Output_Get:Vector)
   > Set base_lapel_A_l_Rotation(base_lapel_A_l_Rotation:Rotator=@n5.Rotation) -> (Output_Get:Rotator)
   n6 = BreakTransform(InTransform=GetSocketTransform(self=Get Body_Mesh(), InSocketName=dyn_lapel_A_l_01, TransformSpace=RTS_ParentBoneSpace))
   > Set dyn_lapel_A_l_01_Translation(dyn_lapel_A_l_01_Translation:Vector=@n6) -> (Output_Get:Vector)
   > Set dyn_lapel_A_l_01_Rotation(dyn_lapel_A_l_01_Rotation:Rotator=@n6.Rotation) -> (Output_Get:Rotator)
   n7 = BreakTransform(InTransform=GetSocketTransform(self=Get Body_Mesh(), InSocketName=base_lapel_B_l, TransformSpace=RTS_ParentBoneSpace))
   > Set base_lapel_B_l_Translation(base_lapel_B_l_Translation:Vector=@n7) -> (Output_Get:Vector)
   > Set base_lapel_B_l_Rotation(base_lapel_B_l_Rotation:Rotator=@n7.Rotation) -> (Output_Get:Rotator)
   n8 = BreakTransform(InTransform=GetSocketTransform(self=Get Body_Mesh(), InSocketName=dyn_lapel_B_l_01, TransformSpace=RTS_ParentBoneSpace))
   > Set dyn_lapel_B_l_01_Translation(dyn_lapel_B_l_01_Translation:Vector=@n8) -> (Output_Get:Vector)
   > Set dyn_lapel_B_l_01_Rotation(dyn_lapel_B_l_01_Rotation:Rotator=@n8.Rotation) -> (Output_Get:Rotator)
   > Outputs!(then:Exec)
  ]
 }
]
variables [
 base_collar_A_l_Rotation {type=Rotator}
 base_collar_A_l_Translation {type=Vector}
 base_collar_B_l_Rotation {type=Rotator}
 base_collar_B_l_Translation {type=Vector}
 base_lapel_A_l_Rotation {type=Rotator}
 base_lapel_A_l_Translation {type=Vector}
 base_lapel_B_l_Rotation {type=Rotator}
 base_lapel_B_l_Translation {type=Vector}
 Body_Mesh {type=SkeletalMeshComponent}
 dyn_collar_A_l_01_Rotation {type=Rotator}
 dyn_collar_A_l_01_Translation {type=Vector}
 dyn_collar_B_l_01_Rotation {type=Rotator}
 dyn_collar_B_l_01_Translation {type=Vector}
 dyn_lapel_A_l_01_Rotation {type=Rotator}
 dyn_lapel_A_l_01_Translation {type=Vector}
 dyn_lapel_B_l_01_Rotation {type=Rotator}
 dyn_lapel_B_l_01_Translation {type=Vector}
 Hair Simulation Space {type=AnimPhysSimSpaceType, default=RootRelative}
 isBodyAttached {type=boolean}
]
